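; Cattus foundation manager that enables or disables a configured list of dynamic building foundations to match the event's state.
; On BeginPlay, the dedicated server runs PostEventJoin immediately with DebugSwapFoundations, otherwise once playlist data is ready and calendar event "PCVD" is active.
; PostEventJoin applies every foundation's stored state.
; LateJoin, uncalled here, applies only states whose change time is at or before JoinTime and debug-prints the flag.
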(event EventBeginPlay
  ;[1 1 1] Begin play
  (if (Networking|IsDedicatedServer)
    (if (Variables|Default|GetDebugSwapFoundations)
      (CallFunction|PostEventJoin)
      (else
        (Class|FortAsyncActionCurrentPlaylistReady|PlaylistDataReadyAsync
          (:OnReady
            (bind _biseventactive (CalendarEvents|GetCalendarEventInformation "PCVD"))
            (if _biseventactive
              (CallFunction|PostEventJoin))))))))

(event Custom|PostEventJoin
  ;[1 1 1] Post event join
  (for _array_element (Variables|Default|GetFoundations)
    (bind (_array_element_foundation_2_b0cfb7e045f01a33f797dbb7aadeaee4 _array_element_enablefoundation_10_100be4cb4682c7f06104fda4a069b50b) (Utilities|Struct|BreakCatFoundationStateData _array_element))
    (Class|BuildingFoundation|SetDynamicFoundationEnabled _array_element_foundation_2_b0cfb7e045f01a33f797dbb7aadeaee4 _array_element_enablefoundation_10_100be4cb4682c7f06104fda4a069b50b)))

(event Custom|LateJoin
  (params (JoinTime :type Float%20%28double-precision%29))
  ;[1 1 1] Late join
  (for _array_element (Variables|Default|GetFoundations)
    (bind (_array_element_foundation_2_b0cfb7e045f01a33f797dbb7aadeaee4 _array_element_enablefoundation_10_100be4cb4682c7f06104fda4a069b50b _array_element_statechangetime_9_d58b5f59492f4cb3c25b0983ab6a1185) (Utilities|Struct|BreakCatFoundationStateData _array_element))
    (if (<= _array_element_statechangetime_9_d58b5f59492f4cb3c25b0983ab6a1185 JoinTime)
      (Class|BuildingFoundation|SetDynamicFoundationEnabled _array_element_foundation_2_b0cfb7e045f01a33f797dbb7aadeaee4 _array_element_enablefoundation_10_100be4cb4682c7f06104fda4a069b50b)
      (Development|PrintString (Utilities|String|ToString(Boolean) _array_element_enablefoundation_10_100be4cb4682c7f06104fda4a069b50b)))))

(fn ConstructionScript)
