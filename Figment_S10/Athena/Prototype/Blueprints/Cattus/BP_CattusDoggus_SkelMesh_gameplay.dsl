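; Gameplay collision skeletal mesh actor for the Cattus/Doggus prototype, with overlap collision enabled only on the server.
; Mesh overlaps apply the GE_BeingMoved effect to players and register damageable buildings for clearing.
; A separate, self-rescheduling knockback event launches listed players still overlapping, using physics-asset surface normal plus point velocity, clamped, doubled pre-clamp on non-positive dot product.
; Nothing here fills that player list or first triggers the knockback.
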(event EventBeginPlay
  (if (Networking|IsServer)
    (Collision|SetCollisionProfileName (Variables|SkeletalMeshActor|GetSkeletalMeshComponent) "OverlapAllDynamic" true :!class /Script/Engine.PrimitiveComponent)
    (Collision|SetActorEnableCollision true)
    (else
      (Collision|SetCollisionProfileName (Variables|SkeletalMeshActor|GetSkeletalMeshComponent) "nocollision" true :!class /Script/Engine.PrimitiveComponent)
      (Collision|SetActorEnableCollision false))))

(event Collision|EventActorBeginOverlap
  (params (OtherActor :type Actor%20Object%20Reference)))

(event EventTick
  (params (DeltaSeconds :type Float%20%28single-precision%29)))

(event OnComponentBeginOverlap(SkeletalMeshComponent)
  (params (OverlappedComponent :type Primitive%20Component%20Object%20Reference) (OtherActor :type Actor%20Object%20Reference) (OtherComp :type Primitive%20Component%20Object%20Reference) (OtherBodyIndex :type Integer) (bFromSweep :type Boolean) (SweepResult :type Hit%20Result%20Structure%20%28by%20ref%29))
  (CallFunction|AttemptHandlePlayerOverlap OtherActor)
  (CallFunction|AttemptHandleBuildingOverlap OtherActor))

(event Custom|KnockBackOverlappingActors
  (bind _baseknockbackstrength (Variables|Default|GetBaseKnockbackStrength))
  (bind _knockedplayer (Variables|Default|GetKnockedPlayer))
  (bind _returnvalue (Transformation|GetActorLocation _knockedplayer))
  (bind _skeletalmeshcomponent (Variables|SkeletalMeshActor|GetSkeletalMeshComponent))
  (bind (_closestworldposition _normal _bonename) (Components|SkeletalMesh|GetClosestPointOnPhysicsAsset _skeletalmeshcomponent _returnvalue))
  (bind _returnvalue_2 (* _normal _baseknockbackstrength))
  (bind _knockedplayer_2 (Variables|Default|GetKnockedPlayer))
  (bind _maxknock (Variables|Default|GetMaxKnock))
  (bind _minknock (Variables|Default|GetMinKnock))
  (bind _knockbacktimer (Variables|Default|GetKnockbackTimer))
  (for _array_element (Variables|Default|GetPlayersToKnock)
    (Utilities|IsValid _array_element
      (:Is%20Valid
        (if (Collision|IsOverlappingActor (Variables|SkeletalMeshActor|GetSkeletalMeshComponent) _array_element :!class /Script/Engine.PrimitiveComponent)
          ;[1 1 1] Refreshable duration GE that handles the whiteout visuals for being knocked back
          (GameplayEffects|ApplyGameplayEffectToTarget :self (Ability|GetAbilitySystemComponent _array_element) :GameplayEffectClass "/Figment_S10/Athena/Prototype/Blueprints/Cattus/GE_BeingMoved.GE_BeingMoved_C" :Target (Ability|GetAbilitySystemComponent _array_element) :Level 0.0)
          (Variables|Default|SetKnockedPlayer _array_element)
          (bind _returnvalue_3 (Physics|GetPhysicsLinearVelocityatPoint (Variables|SkeletalMeshActor|GetSkeletalMeshComponent) _closestworldposition _bonename))
          (if (> (Math|Vector|DotProduct _normal (- (Transformation|GetActorLocation _knockedplayer_2) _closestworldposition)) 0.0)
            (bind _asfort_pawn (Utilities|Casting|CastToFortPawn _knockedplayer_2))
            (Pawn|Character|LaunchCharacterJump _asfort_pawn (Math|Vector|ClampVectorSize (+ _returnvalue_3 _returnvalue_2) _minknock _maxknock) true true true true)
            (else
              (bind _asfort_pawn_2 (Utilities|Casting|CastToFortPawn _knockedplayer_2))
              (Pawn|Character|LaunchCharacterJump _asfort_pawn_2 (Math|Vector|ClampVectorSize (* (+ _returnvalue_3 _returnvalue_2) 2.0) _minknock _maxknock) true true true true)))
          (else
            (Utilities|Array|RemoveItem (Variables|Default|GetPlayersToKnock) _array_element))))))
  (if (> (Utilities|Array|Length (Variables|Default|GetPlayersToKnock)) 0)
    (if (not (Utilities|Time|IsTimerActivebyHandle _knockbacktimer))
      (bind _returnvalue_4 (Utilities|Time|SetTimerbyEvent (EventDispatchers|CreateEvent :!event Custom|KnockBackOverlappingActors) 0.04))
      (Variables|Default|SetKnockbackTimer _returnvalue_4))
    (else
      (Utilities|Time|ClearandInvalidateTimerbyHandle _knockbacktimer))))

(fn AttemptHandleBuildingOverlap
  (params (Object :type Object%20Reference))
  (bind _asbuilding_static_mesh (Utilities|Casting|CastToBuildingSMActor Object))
  (if (Variables|/Script/Engine.Actor|GetbCanBeDamaged _asbuilding_static_mesh)
    (Helpers|AddActortoClear _asbuilding_static_mesh)))

(fn AttemptHandlePlayerOverlap
  (params (Object :type Object%20Reference))
  (bind _asfort_player_pawn_athena (Utilities|Casting|CastToFortPlayerPawnAthena Object))
  (GameplayEffects|ApplyGameplayEffectToTarget :self (Ability|GetAbilitySystemComponent _asfort_player_pawn_athena) :GameplayEffectClass "/Figment_S10/Athena/Prototype/Blueprints/Cattus/GE_BeingMoved.GE_BeingMoved_C" :Target (Ability|GetAbilitySystemComponent _asfort_player_pawn_athena) :Level 0.0))

(fn ConstructionScript)
