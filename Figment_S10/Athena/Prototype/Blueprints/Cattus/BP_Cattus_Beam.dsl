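; BP_Cattus_Beam is a prototype Cattus beam hazard that destroys building props and kills pawns overlapping it.
; Its cube collision is overlap-only on the server and disabled on clients.
; On overlap, damageable props get a short random lifespan, while other actors are cast to FortPawn, debug-printed and force-killed as AI eliminations.
; A self-rescheduling timer event damages queued props one by one, though nothing here starts it or fills its array.
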(event EventBeginPlay
  (bind _cube (Variables|Default|GetCube))
  (if (Networking|IsServer)
    (Collision|SetCollisionProfileName _cube "OverlapAllDynamic" true :!class /Script/Engine.PrimitiveComponent)
    (Collision|SetActorEnableCollision true)
    (else
      (Collision|SetCollisionProfileName _cube "nocollision" true :!class /Script/Engine.PrimitiveComponent)
      (Collision|SetActorEnableCollision false))))

(event Collision|EventActorBeginOverlap
  (params (OtherActor :type Actor%20Object%20Reference))
  (bind _asbuilding_prop (Utilities|Casting|CastToBuildingProp OtherActor)
    (:then
      (Utilities|IsValid _asbuilding_prop
        (:Is%20Valid
          (if (Variables|/Script/Engine.Actor|GetbCanBeDamaged _asbuilding_prop)
            (Actor|SetLifeSpan _asbuilding_prop (Math|Random|RandomFloatinRange 0.1 0.3))))))
    (:CastFailed
      (bind _asfort_pawn (Utilities|Casting|CastToFortPawn OtherActor))
      (Development|PrintString (Utilities|String|Append (Utilities|GetDisplayName _asfort_pawn) "IS OVERLAPPING"))
      (FortPawn|ForceKill _asfort_pawn "(TagName=\"DeathCause.EliminatedByAI\")" 0 self))))

(event EventTick
  (params (DeltaSeconds :type Float%20%28single-precision%29)))

(event Custom|DetroyProps
  (if (Math|Conversions|ToBoolean(Integer) (Utilities|Array|Length (Variables|Default|GetPropsToDestroy)))
    (Variables|Default|SetTargetProp (Utilities|Array|Get(acopy) (Variables|Default|GetPropsToDestroy)))
    (Utilities|IsValid (Variables|Default|GetTargetProp)
      (:Is%20Valid
        (Environment|ApplyGlobalEnvironmentDamagetoActor (Variables|Default|GetTargetProp) 2e+03))
      (:Is%20Not%20Valid
        _))
    (Utilities|Array|RemoveItem (Variables|Default|GetPropsToDestroy) (Variables|Default|GetTargetProp))
    (bind _returnvalue (Utilities|Time|SetTimerbyEvent (EventDispatchers|CreateEvent :!event Custom|DetroyProps) 0.04))
    (Variables|Default|SetPropDestroyTimer _returnvalue)
    (else
      (Utilities|Time|ClearandInvalidateTimerbyHandle (Variables|Default|GetPropDestroyTimer)))))

(fn ConstructionScript)
