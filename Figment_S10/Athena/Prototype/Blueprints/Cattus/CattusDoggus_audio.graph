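# Audio level sequence for the Cattus vs Doggus fight, covering music, character stems and one-shot effects from the water entrance to the final impact and moon exit.
# Each sound is an AmbientSound with spawn and audio tracks, most one-shots spawned only around their cue.
# Vocal and body stems stay spawned and attached for most of the timeline, while environmental impacts mostly sit at fixed world positions.
# A root Music track spans the fight.
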LevelSequence /Figment_S10/Athena/Prototype/Blueprints/Cattus/CattusDoggus_audio
bindings [
 Cattus Tackled Splash {
  id=#1
  possessed_class=AmbientSound
  tracks [
   {
    type=MovieSceneSpawnTrack
    sections [{
      type=MovieSceneSpawnSection
      channels [None {type=MovieSceneScriptingBoolChannel, default=false, keys [[4156,false],[4217,true],[5349,false]]}]
     }]
   }
   {
    type=MovieScene3DTransformTrack
    sections [{
      type=MovieScene3DTransformSection
      channels [
       Location.X {type=MovieSceneScriptingDoubleChannel, default=80095.9922}
       Location.Y {type=MovieSceneScriptingDoubleChannel, default=-163715.7812}
       Location.Z {type=MovieSceneScriptingDoubleChannel, default=-5120}
       Rotation.X {type=MovieSceneScriptingDoubleChannel, default=0}
       Rotation.Y {type=MovieSceneScriptingDoubleChannel, default=0}
       Rotation.Z {type=MovieSceneScriptingDoubleChannel, default=0}
       Scale.X {type=MovieSceneScriptingDoubleChannel, default=1}
       Scale.Y {type=MovieSceneScriptingDoubleChannel, default=1}
       Scale.Z {type=MovieSceneScriptingDoubleChannel, default=1}
      ]
     }]
   }
   {
    type=MovieSceneAudioTrack
    display_name=Audio
    sections [{
      type=MovieSceneAudioSection
      end_frame=5315
      start_frame=4217
      channels [
       Volume {type=MovieSceneScriptingFloatChannel, default=1}
       Pitch {type=MovieSceneScriptingFloatChannel, default=1}
      ]
     }]
   }
  ]
 }
 Cattus POI Smash {
  id=#2
  possessed_class=AmbientSound
  tracks [
   {
    type=MovieSceneSpawnTrack
    sections [{
      type=MovieSceneSpawnSection
      channels [None {type=MovieSceneScriptingBoolChannel, default=false, keys [[1423,false],[1494,true],[1858,false]]}]
     }]
   }
   {
    type=MovieScene3DTransformTrack
    sections [{
      type=MovieScene3DTransformSection
      channels [
       Location.X {type=MovieSceneScriptingDoubleChannel, default=111147.5781}
       Location.Y {type=MovieSceneScriptingDoubleChannel, default=-26250.3652}
       Location.Z {type=MovieSceneScriptingDoubleChannel, default=692.6227}
       Rotation.X {type=MovieSceneScriptingDoubleChannel, default=0}
       Rotation.Y {type=MovieSceneScriptingDoubleChannel, default=0}
       Rotation.Z {type=MovieSceneScriptingDoubleChannel, default=0}
       Scale.X {type=MovieSceneScriptingDoubleChannel, default=1}
       Scale.Y {type=MovieSceneScriptingDoubleChannel, default=1}
       Scale.Z {type=MovieSceneScriptingDoubleChannel, default=1}
      ]
     }]
   }
   {
    type=MovieSceneAudioTrack
    display_name=Audio
    sections [{
      type=MovieSceneAudioSection
      end_frame=1835
      start_frame=1492
      channels [
       Volume {type=MovieSceneScriptingFloatChannel, default=1}
       Pitch {type=MovieSceneScriptingFloatChannel, default=1}
      ]
     }]
   }
  ]
 }
 Gateway Impacts {
  id=#3
  possessed_class=AmbientSound
  tracks [
   {
    type=MovieSceneSpawnTrack
    sections [{
      type=MovieSceneSpawnSection
      channels [None {type=MovieSceneScriptingBoolChannel, default=false, keys [[2175,false],[2320,true],[2585,false],[6364,false],[6368.5038,true],[6721,false]]}]
     }]
   }
   {
    type=MovieScene3DTransformTrack
    sections [{
      type=MovieScene3DTransformSection
      channels [
       Location.X {type=MovieSceneScriptingDoubleChannel, default=66099.7891}
       Location.Y {type=MovieSceneScriptingDoubleChannel, default=-43117.0781}
       Location.Z {type=MovieSceneScriptingDoubleChannel, default=-2180.895}
       Rotation.X {type=MovieSceneScriptingDoubleChannel, default=0}
       Rotation.Y {type=MovieSceneScriptingDoubleChannel, default=0}
       Rotation.Z {type=MovieSceneScriptingDoubleChannel, default=0}
       Scale.X {type=MovieSceneScriptingDoubleChannel, default=1}
       Scale.Y {type=MovieSceneScriptingDoubleChannel, default=1}
       Scale.Z {type=MovieSceneScriptingDoubleChannel, default=1}
      ]
     }]
   }
   {
    type=MovieSceneAudioTrack
    display_name=Audio
    sections [
     {
      type=MovieSceneAudioSection
      end_frame=2551
      start_frame=2318
      channels [
       Volume {type=MovieSceneScriptingFloatChannel, default=1}
       Pitch {type=MovieSceneScriptingFloatChannel, default=1}
      ]
     }
     {
      type=MovieSceneAudioSection
      end_frame=6689
      start_frame=6363
      channels [
       Volume {type=MovieSceneScriptingFloatChannel, default=1}
       Pitch {type=MovieSceneScriptingFloatChannel, default=1}
      ]
     }
    ]
   }
  ]
 }
 Cattus Entrance Water {
  id=#4
  possessed_class=AmbientSound
  tracks [
   {
    type=MovieSceneSpawnTrack
    sections [{
      type=MovieSceneSpawnSection
      channels [None {type=MovieSceneScriptingBoolChannel, default=false, keys [[144,false],[204,true],[1128,false]]}]
     }]
   }
   {
    type=MovieScene3DTransformTrack
    sections [{
      type=MovieScene3DTransformSection
      channels [
       Location.X {type=MovieSceneScriptingDoubleChannel, default=147161.5156}
       Location.Y {type=MovieSceneScriptingDoubleChannel, default=-26231.3457}
       Location.Z {type=MovieSceneScriptingDoubleChannel, default=-5103.8174}
       Rotation.X {type=MovieSceneScriptingDoubleChannel, default=0}
       Rotation.Y {type=MovieSceneScriptingDoubleChannel, default=0}
       Rotation.Z {type=MovieSceneScriptingDoubleChannel, default=0}
       Scale.X {type=MovieSceneScriptingDoubleChannel, default=1}
       Scale.Y {type=MovieSceneScriptingDoubleChannel, default=1}
       Scale.Z {type=MovieSceneScriptingDoubleChannel, default=1}
      ]
     }]
   }
   {
    type=MovieSceneAudioTrack
    display_name=Audio
    sections [{
      type=MovieSceneAudioSection
      end_frame=1021
      start_frame=202
      channels [
       Volume {type=MovieSceneScriptingFloatChannel, default=1}
       Pitch {type=MovieSceneScriptingFloatChannel, default=1}
      ]
     }]
   }
  ]
 }
 Cattus Impact Final {
  id=#5
  possessed_class=AmbientSound
  tracks [
   {
    type=MovieSceneSpawnTrack
    sections [{
      type=MovieSceneSpawnSection
      channels [None {type=MovieSceneScriptingBoolChannel, default=false, keys [[7670,false],[7710,true],[8083,false]]}]
     }]
   }
   {
    type=MovieScene3DTransformTrack
    sections [{
      type=MovieScene3DTransformSection
      channels [
       Location.X {type=MovieSceneScriptingDoubleChannel, default=0}
       Location.Y {type=MovieSceneScriptingDoubleChannel, default=0}
       Location.Z {type=MovieSceneScriptingDoubleChannel, default=0}
       Rotation.X {type=MovieSceneScriptingDoubleChannel, default=-42.6268}
       Rotation.Y {type=MovieSceneScriptingDoubleChannel, default=-37.4921}
       Rotation.Z {type=MovieSceneScriptingDoubleChannel, default=-30.9415}
       Scale.X {type=MovieSceneScriptingDoubleChannel, default=0.01}
       Scale.Y {type=MovieSceneScriptingDoubleChannel, default=0.01}
       Scale.Z {type=MovieSceneScriptingDoubleChannel, default=0.01}
      ]
     }]
   }
   {type=MovieScene3DAttachTrack, sections [{type=MovieScene3DAttachSection, end_frame=9068, start_frame=7710}]}
   {
    type=MovieSceneAudioTrack
    display_name=Audio
    sections [{
      type=MovieSceneAudioSection
      end_frame=8062
      start_frame=7710
      channels [
       Volume {type=MovieSceneScriptingFloatChannel, default=1}
       Pitch {type=MovieSceneScriptingFloatChannel, default=1}
      ]
     }]
   }
  ]
 }
 Doggus Sword Activate {
  id=#6
  possessed_class=AmbientSound
  tracks [
   {type=MovieSceneSpawnTrack, sections [{type=MovieSceneSpawnSection, channels [None {type=MovieSceneScriptingBoolChannel, default=true}]}]}
   {
    type=MovieScene3DTransformTrack
    sections [{
      type=MovieScene3DTransformSection
      channels [
       Location.X {type=MovieSceneScriptingDoubleChannel, default=0}
       Location.Y {type=MovieSceneScriptingDoubleChannel, default=0}
       Location.Z {type=MovieSceneScriptingDoubleChannel, default=0}
       Rotation.X {type=MovieSceneScriptingDoubleChannel, default=177}
       Rotation.Y {type=MovieSceneScriptingDoubleChannel, default=-75.0346}
       Rotation.Z {type=MovieSceneScriptingDoubleChannel, default=165.2219}
       Scale.X {type=MovieSceneScriptingDoubleChannel, default=0.01}
       Scale.Y {type=MovieSceneScriptingDoubleChannel, default=0.01}
       Scale.Z {type=MovieSceneScriptingDoubleChannel, default=0.01}
      ]
     }]
   }
   {type=MovieScene3DAttachTrack, sections [{type=MovieScene3DAttachSection, end_frame=9156, start_frame=7277}]}
   {
    type=MovieSceneAudioTrack
    display_name=Audio
    sections [{
      type=MovieSceneAudioSection
      end_frame=8397
      start_frame=7277
      channels [
       Volume {type=MovieSceneScriptingFloatChannel, default=1}
       Pitch {type=MovieSceneScriptingFloatChannel, default=1}
      ]
     }]
   }
  ]
 }
 Doggus Vocal Stem {
  id=#7
  possessed_class=AmbientSound
  tracks [
   {type=MovieSceneSpawnTrack, sections [{type=MovieSceneSpawnSection, channels [None {type=MovieSceneScriptingBoolChannel, default=true}]}]}
   {
    type=MovieScene3DTransformTrack
    sections [{
      type=MovieScene3DTransformSection
      channels [
       Location.X {type=MovieSceneScriptingDoubleChannel, default=0}
       Location.Y {type=MovieSceneScriptingDoubleChannel, default=0}
       Location.Z {type=MovieSceneScriptingDoubleChannel, default=0}
       Rotation.X {type=MovieSceneScriptingDoubleChannel, default=-55.571}
       Rotation.Y {type=MovieSceneScriptingDoubleChannel, default=75.8386}
       Rotation.Z {type=MovieSceneScriptingDoubleChannel, default=67.882}
       Scale.X {type=MovieSceneScriptingDoubleChannel, default=0.01}
       Scale.Y {type=MovieSceneScriptingDoubleChannel, default=0.01}
       Scale.Z {type=MovieSceneScriptingDoubleChannel, default=0.01}
      ]
     }]
   }
   {type=MovieScene3DAttachTrack, sections [{type=MovieScene3DAttachSection, end_frame=9151, start_frame=216}]}
   {
    type=MovieSceneAudioTrack
    display_name=Audio
    sections [{
      type=MovieSceneAudioSection
      end_frame=9151
      start_frame=216
      channels [
       Volume {type=MovieSceneScriptingFloatChannel, default=1}
       Pitch {type=MovieSceneScriptingFloatChannel, default=1}
      ]
     }]
   }
  ]
 }
 Cattus Body Impact Large {
  id=#8
  possessed_class=AmbientSound
  tracks [
   {
    type=MovieSceneSpawnTrack
    sections [{
      type=MovieSceneSpawnSection
      channels [None {type=MovieSceneScriptingBoolChannel, default=false, keys [[5235,false],[5269,true],[5816,false],[6819,true],[7201,false]]}]
     }]
   }
   {
    type=MovieScene3DTransformTrack
    sections [{
      type=MovieScene3DTransformSection
      channels [
       Location.X {type=MovieSceneScriptingDoubleChannel, default=0}
       Location.Y {type=MovieSceneScriptingDoubleChannel, default=0}
       Location.Z {type=MovieSceneScriptingDoubleChannel, default=0}
       Rotation.X {type=MovieSceneScriptingDoubleChannel, default=0}
       Rotation.Y {type=MovieSceneScriptingDoubleChannel, default=0}
       Rotation.Z {type=MovieSceneScriptingDoubleChannel, default=0}
       Scale.X {type=MovieSceneScriptingDoubleChannel, default=0.01}
       Scale.Y {type=MovieSceneScriptingDoubleChannel, default=0.01}
       Scale.Z {type=MovieSceneScriptingDoubleChannel, default=0.01}
      ]
     }]
   }
   {type=MovieScene3DAttachTrack, sections [{type=MovieScene3DAttachSection, end_frame=7999, start_frame=5268}]}
   {
    type=MovieSceneAudioTrack
    display_name=Audio
    sections [
     {
      type=MovieSceneAudioSection
      end_frame=7172
      start_frame=6826
      channels [
       Volume {type=MovieSceneScriptingFloatChannel, default=1}
       Pitch {type=MovieSceneScriptingFloatChannel, default=1}
      ]
     }
     {
      type=MovieSceneAudioSection
      end_frame=5800
      start_frame=5685
      channels [
       Volume {type=MovieSceneScriptingFloatChannel, default=1}
       Pitch {type=MovieSceneScriptingFloatChannel, default=1}
      ]
     }
    ]
   }
  ]
 }
 Doggus Sword Pull {
  id=#9
  possessed_class=AmbientSound
  tracks [
   {type=MovieSceneSpawnTrack, sections [{type=MovieSceneSpawnSection, channels [None {type=MovieSceneScriptingBoolChannel, default=true}]}]}
   {
    type=MovieScene3DTransformTrack
    sections [{
      type=MovieScene3DTransformSection
      channels [
       Location.X {type=MovieSceneScriptingDoubleChannel, default=35840}
       Location.Y {type=MovieSceneScriptingDoubleChannel, default=-56448}
       Location.Z {type=MovieSceneScriptingDoubleChannel, default=315}
       Rotation.X {type=MovieSceneScriptingDoubleChannel, default=0}
       Rotation.Y {type=MovieSceneScriptingDoubleChannel, default=0}
       Rotation.Z {type=MovieSceneScriptingDoubleChannel, default=0}
       Scale.X {type=MovieSceneScriptingDoubleChannel, default=1}
       Scale.Y {type=MovieSceneScriptingDoubleChannel, default=1}
       Scale.Z {type=MovieSceneScriptingDoubleChannel, default=1}
      ]
     }]
   }
   {
    type=MovieSceneAudioTrack
    display_name=Audio
    sections [{
      type=MovieSceneAudioSection
      end_frame=8094
      start_frame=7208
      channels [
       Volume {type=MovieSceneScriptingFloatChannel, default=1}
       Pitch {type=MovieSceneScriptingFloatChannel, default=1}
      ]
     }]
   }
  ]
 }
 Doggus Body Stem {
  id=#10
  possessed_class=AmbientSound
  tracks [
   {type=MovieScene3DAttachTrack, sections [{type=MovieScene3DAttachSection, end_frame=9151, start_frame=216}]}
   {
    type=MovieScene3DTransformTrack
    sections [{
      type=MovieScene3DTransformSection
      channels [
       Location.X {type=MovieSceneScriptingDoubleChannel, default=0}
       Location.Y {type=MovieSceneScriptingDoubleChannel, default=0}
       Location.Z {type=MovieSceneScriptingDoubleChannel, default=0}
       Rotation.X {type=MovieSceneScriptingDoubleChannel, default=89.8915}
       Rotation.Y {type=MovieSceneScriptingDoubleChannel, default=0.4145}
       Rotation.Z {type=MovieSceneScriptingDoubleChannel, default=-93.244}
       Scale.X {type=MovieSceneScriptingDoubleChannel, default=0.01}
       Scale.Y {type=MovieSceneScriptingDoubleChannel, default=0.01}
       Scale.Z {type=MovieSceneScriptingDoubleChannel, default=0.01}
      ]
     }]
   }
   {type=MovieSceneSpawnTrack, sections [{type=MovieSceneSpawnSection, channels [None {type=MovieSceneScriptingBoolChannel, default=true}]}]}
   {
    type=MovieSceneAudioTrack
    display_name=Audio
    sections [{
      type=MovieSceneAudioSection
      end_frame=9151
      start_frame=216
      channels [
       Volume {type=MovieSceneScriptingFloatChannel, default=1}
       Pitch {type=MovieSceneScriptingFloatChannel, default=1}
      ]
     }]
   }
  ]
 }
 Cattus Laser Blast {
  id=#11
  possessed_class=AmbientSound
  tracks [
   {
    type=MovieSceneSpawnTrack
    sections [{
      type=MovieSceneSpawnSection
      channels [None {type=MovieSceneScriptingBoolChannel, default=false, keys [[3550,false],[3570,true],[4073,false]]}]
     }]
   }
   {
    type=MovieScene3DTransformTrack
    sections [{
      type=MovieScene3DTransformSection
      channels [
       Location.X {type=MovieSceneScriptingDoubleChannel, default=53.6562}
       Location.Y {type=MovieSceneScriptingDoubleChannel, default=156.3331}
       Location.Z {type=MovieSceneScriptingDoubleChannel, default=-1.4831}
       Rotation.X {type=MovieSceneScriptingDoubleChannel, default=-142.146}
       Rotation.Y {type=MovieSceneScriptingDoubleChannel, default=80.2342}
       Rotation.Z {type=MovieSceneScriptingDoubleChannel, default=-165.2911}
       Scale.X {type=MovieSceneScriptingDoubleChannel, default=0.01}
       Scale.Y {type=MovieSceneScriptingDoubleChannel, default=0.01}
       Scale.Z {type=MovieSceneScriptingDoubleChannel, default=0.01}
      ]
     }]
   }
   {type=MovieScene3DAttachTrack, sections [{type=MovieScene3DAttachSection, end_frame=4068, start_frame=3571}]}
   {
    type=MovieSceneAudioTrack
    display_name=Audio
    sections [{
      type=MovieSceneAudioSection
      end_frame=4068
      start_frame=3597
      channels [
       Volume {type=MovieSceneScriptingFloatChannel, default=1}
       Pitch {type=MovieSceneScriptingFloatChannel, default=1}
      ]
     }]
   }
  ]
 }
 Doggus Arm Splash {
  id=#12
  possessed_class=AmbientSound
  tracks [
   {
    type=MovieSceneSpawnTrack
    sections [{
      type=MovieSceneSpawnSection
      channels [None {type=MovieSceneScriptingBoolChannel, default=false, keys [[-6,false],[5654,true],[6139,false]]}]
     }]
   }
   {
    type=MovieSceneAudioTrack
    display_name=Audio
    sections [{
      type=MovieSceneAudioSection
      end_frame=6089
      start_frame=5729
      channels [
       Volume {type=MovieSceneScriptingFloatChannel, default=1}
       Pitch {type=MovieSceneScriptingFloatChannel, default=1}
      ]
     }]
   }
  ]
 }
 Cattus First Beam Impact {
  id=#13
  possessed_class=AmbientSound
  tracks [
   {
    type=MovieSceneSpawnTrack
    sections [{
      type=MovieSceneSpawnSection
      channels [None {type=MovieSceneScriptingBoolChannel, default=false, keys [[973,false],[988,true],[1355,false]]}]
     }]
   }
   {
    type=MovieScene3DTransformTrack
    sections [{
      type=MovieScene3DTransformSection
      channels [
       Location.X {type=MovieSceneScriptingDoubleChannel, default=0, keys [[1005,0],[1091,0],[1094,99609.8594]]}
       Location.Y {type=MovieSceneScriptingDoubleChannel, default=0, keys [[1005,0],[1091,0],[1094,8182.2773]]}
       Location.Z {type=MovieSceneScriptingDoubleChannel, default=0, keys [[1005,0],[1091,0],[1094,4379.3975]]}
       Rotation.X {type=MovieSceneScriptingDoubleChannel, default=1.0844}
       Rotation.Y {type=MovieSceneScriptingDoubleChannel, default=3.4764}
       Rotation.Z {type=MovieSceneScriptingDoubleChannel, default=9.4302}
       Scale.X {type=MovieSceneScriptingDoubleChannel, default=0.0001}
       Scale.Y {type=MovieSceneScriptingDoubleChannel, default=0.0001}
       Scale.Z {type=MovieSceneScriptingDoubleChannel, default=0.0001}
      ]
     }]
   }
   {
    type=MovieSceneAudioTrack
    display_name=Audio
    sections [{
      type=MovieSceneAudioSection
      end_frame=1320
      start_frame=1012
      channels [
       Volume {type=MovieSceneScriptingFloatChannel, default=1}
       Pitch {type=MovieSceneScriptingFloatChannel, default=1}
      ]
     }]
   }
   {type=MovieScene3DAttachTrack, sections [{type=MovieScene3DAttachSection, end_frame=1094, start_frame=1012}]}
  ]
 }
 Doggus Goes to Moon {
  id=#14
  possessed_class=AmbientSound
  tracks [
   {
    type=MovieSceneSpawnTrack
    sections [{
      type=MovieSceneSpawnSection
      channels [None {type=MovieSceneScriptingBoolChannel, default=false, keys [[9034,false],[9058,true],[9258,false]]}]
     }]
   }
   {
    type=MovieScene3DTransformTrack
    sections [{
      type=MovieScene3DTransformSection
      channels [
       Location.X {type=MovieSceneScriptingDoubleChannel, default=15570.4639}
       Location.Y {type=MovieSceneScriptingDoubleChannel, default=-36866.7344}
       Location.Z {type=MovieSceneScriptingDoubleChannel, default=721150.5625}
       Rotation.X {type=MovieSceneScriptingDoubleChannel, default=0}
       Rotation.Y {type=MovieSceneScriptingDoubleChannel, default=0}
       Rotation.Z {type=MovieSceneScriptingDoubleChannel, default=0}
       Scale.X {type=MovieSceneScriptingDoubleChannel, default=1}
       Scale.Y {type=MovieSceneScriptingDoubleChannel, default=1}
       Scale.Z {type=MovieSceneScriptingDoubleChannel, default=1}
      ]
     }]
   }
   {
    type=MovieSceneAudioTrack
    display_name=Audio
    sections [{
      type=MovieSceneAudioSection
      end_frame=9246
      start_frame=9063
      channels [
       Volume {type=MovieSceneScriptingFloatChannel, default=1}
       Pitch {type=MovieSceneScriptingFloatChannel, default=1}
      ]
     }]
   }
  ]
 }
 Cattus Vocal Stem {
  id=#15
  possessed_class=AmbientSound
  tracks [
   {type=MovieSceneSpawnTrack, sections [{type=MovieSceneSpawnSection, channels [None {type=MovieSceneScriptingBoolChannel, default=true}]}]}
   {
    type=MovieScene3DTransformTrack
    sections [{
      type=MovieScene3DTransformSection
      channels [
       Location.X {type=MovieSceneScriptingDoubleChannel, default=0}
       Location.Y {type=MovieSceneScriptingDoubleChannel, default=0}
       Location.Z {type=MovieSceneScriptingDoubleChannel, default=0}
       Rotation.X {type=MovieSceneScriptingDoubleChannel, default=1.6045}
       Rotation.Y {type=MovieSceneScriptingDoubleChannel, default=21.862}
       Rotation.Z {type=MovieSceneScriptingDoubleChannel, default=79.5326}
       Scale.X {type=MovieSceneScriptingDoubleChannel, default=0.01}
       Scale.Y {type=MovieSceneScriptingDoubleChannel, default=0.01}
       Scale.Z {type=MovieSceneScriptingDoubleChannel, default=0.01}
      ]
     }]
   }
   {type=MovieScene3DAttachTrack, sections [{type=MovieScene3DAttachSection, end_frame=8302, start_frame=221}]}
   {
    type=MovieSceneAudioTrack
    display_name=Audio
    sections [{
      type=MovieSceneAudioSection
      end_frame=9451
      start_frame=221
      channels [
       Volume {type=MovieSceneScriptingFloatChannel, default=1}
       Pitch {type=MovieSceneScriptingFloatChannel, default=1}
      ]
     }]
   }
  ]
 }
 Cattus Body Stem {
  id=#16
  possessed_class=AmbientSound
  tracks [
   {type=MovieScene3DAttachTrack, sections [{type=MovieScene3DAttachSection, end_frame=9156, start_frame=221}]}
   {
    type=MovieScene3DTransformTrack
    sections [{
      type=MovieScene3DTransformSection
      channels [
       Location.X {type=MovieSceneScriptingDoubleChannel, default=0}
       Location.Y {type=MovieSceneScriptingDoubleChannel, default=0}
       Location.Z {type=MovieSceneScriptingDoubleChannel, default=0}
       Rotation.X {type=MovieSceneScriptingDoubleChannel, default=79.5538}
       Rotation.Y {type=MovieSceneScriptingDoubleChannel, default=1.1752}
       Rotation.Z {type=MovieSceneScriptingDoubleChannel, default=-119.0496}
       Scale.X {type=MovieSceneScriptingDoubleChannel, default=0.01}
       Scale.Y {type=MovieSceneScriptingDoubleChannel, default=0.01}
       Scale.Z {type=MovieSceneScriptingDoubleChannel, default=0.01}
      ]
     }]
   }
   {type=MovieSceneSpawnTrack, sections [{type=MovieSceneSpawnSection, channels [None {type=MovieSceneScriptingBoolChannel, default=true}]}]}
   {
    type=MovieSceneAudioTrack
    display_name=Audio
    sections [{
      type=MovieSceneAudioSection
      end_frame=9156
      start_frame=221
      channels [
       Volume {type=MovieSceneScriptingFloatChannel, default=1}
       Pitch {type=MovieSceneScriptingFloatChannel, default=1}
      ]
     }]
   }
  ]
 }
 Doggus Tackle Shockwave {
  id=#17
  possessed_class=AmbientSound
  tracks [
   {
    type=MovieSceneSpawnTrack
    sections [{
      type=MovieSceneSpawnSection
      channels [None {type=MovieSceneScriptingBoolChannel, default=false, keys [[4044,false],[4111,true],[4533,false]]}]
     }]
   }
   {
    type=MovieScene3DTransformTrack
    sections [{
      type=MovieScene3DTransformSection
      channels [
       Location.X {type=MovieSceneScriptingDoubleChannel, default=63111.6055}
       Location.Y {type=MovieSceneScriptingDoubleChannel, default=-25432.1738}
       Location.Z {type=MovieSceneScriptingDoubleChannel, default=12673.7744}
       Rotation.X {type=MovieSceneScriptingDoubleChannel, default=0}
       Rotation.Y {type=MovieSceneScriptingDoubleChannel, default=0}
       Rotation.Z {type=MovieSceneScriptingDoubleChannel, default=0}
       Scale.X {type=MovieSceneScriptingDoubleChannel, default=1}
       Scale.Y {type=MovieSceneScriptingDoubleChannel, default=1}
       Scale.Z {type=MovieSceneScriptingDoubleChannel, default=1}
      ]
     }]
   }
   {
    type=MovieSceneAudioTrack
    display_name=Audio
    sections [{
      type=MovieSceneAudioSection
      end_frame=4472
      start_frame=4112
      channels [
       Volume {type=MovieSceneScriptingFloatChannel, default=1}
       Pitch {type=MovieSceneScriptingFloatChannel, default=1}
      ]
     }]
   }
  ]
 }
 Doggus Missile Barrage {
  id=#18
  possessed_class=AmbientSound
  tracks [
   {type=MovieSceneSpawnTrack, sections [{type=MovieSceneSpawnSection, channels [None {type=MovieSceneScriptingBoolChannel, default=true}]}]}
   {
    type=MovieScene3DTransformTrack
    sections [{
      type=MovieScene3DTransformSection
      channels [
       Location.X {type=MovieSceneScriptingDoubleChannel, default=-63.8524}
       Location.Y {type=MovieSceneScriptingDoubleChannel, default=292.5989}
       Location.Z {type=MovieSceneScriptingDoubleChannel, default=-21.891}
       Rotation.X {type=MovieSceneScriptingDoubleChannel, default=0}
       Rotation.Y {type=MovieSceneScriptingDoubleChannel, default=0}
       Rotation.Z {type=MovieSceneScriptingDoubleChannel, default=0}
       Scale.X {type=MovieSceneScriptingDoubleChannel, default=0.01}
       Scale.Y {type=MovieSceneScriptingDoubleChannel, default=0.01}
       Scale.Z {type=MovieSceneScriptingDoubleChannel, default=0.01}
      ]
     }]
   }
   {type=MovieScene3DAttachTrack, sections [{type=MovieScene3DAttachSection, end_frame=9252, start_frame=2037}]}
   {
    type=MovieSceneAudioTrack
    display_name=Audio
    sections [{
      type=MovieSceneAudioSection
      end_frame=3432
      start_frame=2549
      channels [
       Volume {type=MovieSceneScriptingFloatChannel, default=1}
       Pitch {type=MovieSceneScriptingFloatChannel, default=1}
      ]
     }]
   }
  ]
 }
]
display_rate {denominator=1, numerator=30}
tick_resolution {denominator=1, numerator=24000}
tracks [{
  type=MovieSceneAudioTrack
  display_name=Music
  sections [{
    type=MovieSceneAudioSection
    end_frame=9338
    start_frame=208
    channels [
     Volume {type=MovieSceneScriptingFloatChannel, default=1}
     Pitch {type=MovieSceneScriptingFloatChannel, default=1}
     AttachActor {type=MovieSceneScriptingActorReferenceChannel, default="<Struct 'MovieSceneObjectBindingID' (0x49e50840ed0) {}>"}
    ]
   }]
 }]
