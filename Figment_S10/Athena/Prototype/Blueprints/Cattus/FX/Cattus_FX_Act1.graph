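# Level Sequence that times the particle effects for Act 1 of the Cattus sequence, covering Cattus and Doggus.
# Emitter bindings cover water hand/foot effects, castle destruction, footsteps, drool and beam charge, casting and impact.
# Spawn tracks gate emitters, while particle tracks activate or trigger them; several footstep emitters are repositioned by transform keys and retriggered.
# Beam and drool emitters use attach tracks, with parameter tracks animating beam target, size, erosion, opacity, light and spawn rate.
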LevelSequence /Figment_S10/Athena/Prototype/Blueprints/Cattus/FX/Cattus_FX_Act1
bindings [
 P Doggus Step Heavy Water FX {
  id=#1
  possessed_class=Emitter
  tracks [
   {
    type=MovieSceneSpawnTrack
    sections [{
      type=MovieSceneSpawnSection
      start_frame=2442
      channels [None {type=MovieSceneScriptingBoolChannel, default=false, keys [[2442,true],[4091,false]]}]
     }]
   }
   {
    type=MovieScene3DTransformTrack
    sections [{
      type=MovieScene3DTransformSection
      end_frame=4091
      start_frame=2442
      channels [
       Location.X {
        type=MovieSceneScriptingDoubleChannel
        default=73356
        keys [[2442,73356],[2525,74592],[2677,74592],[2700,73324],[3088,69464],[3131,77508],[3305,71480],[3456,69092],[3721,70536],[3887,67596]]
       }
       Location.Y {
        type=MovieSceneScriptingDoubleChannel
        default=-38024
        keys [[2442,-38024],[2525,-42448],[2677,-42448],[2700,-44196],[3088,-36544],[3131,-45688],[3305,-38320],[3456,-32520],[3721,-37552],[3887,-35200]]
       }
       Location.Z {type=MovieSceneScriptingDoubleChannel, default=0, keys [[2442,-772],[2677,0],[2700,-576],[3131,-796]]}
       Rotation.X {type=MovieSceneScriptingDoubleChannel, default=0, const [0,2]}
       Rotation.Y {type=MovieSceneScriptingDoubleChannel, default=0, const [0,2]}
       Rotation.Z {type=MovieSceneScriptingDoubleChannel, default=0, const [0,2]}
       Scale.X {type=MovieSceneScriptingDoubleChannel, default=1, const [1,2]}
       Scale.Y {type=MovieSceneScriptingDoubleChannel, default=1, const [1,2]}
       Scale.Z {type=MovieSceneScriptingDoubleChannel, default=1, const [1,2]}
      ]
     }]
   }
   {
    type=MovieSceneParticleTrack
    display_name=Particle System
    sections [{type=MovieSceneParticleSection, channels [None {type=MovieSceneScriptingParticleChannel, const ["<ParticleKey.TRIGGER: 2>", 9]}]}]
   }
  ]
 }
 P Water Raid Hand L {
  id=#2
  possessed_class=Emitter
  tracks [
   {
    type=MovieSceneSpawnTrack
    sections [{
      type=MovieSceneSpawnSection
      end_frame=627
      start_frame=152
      channels [None {type=MovieSceneScriptingBoolChannel, default=false, keys [[152,false],[190,true],[627,false]]}]
     }]
   }
   {
    type=MovieScene3DTransformTrack
    sections [{
      type=MovieScene3DTransformSection
      end_frame=668
      channels [
       Location.X {
        type=MovieSceneScriptingDoubleChannel
        default=133696.8125
        keys [[190,140732.9844],[228,133696.8125],[354,133696.8125],[389,129885.0078],[450,132850.8594],[497,131063.7969]]
       }
       Location.Y {
        type=MovieSceneScriptingDoubleChannel
        default=-21258.8281
        keys [[190,-19493.6289],[228,-21258.8281],[354,-21258.8281],[389,-26170.4023],[450,-17545.834],[497,-21994.0117]]
       }
       Location.Z {type=MovieSceneScriptingDoubleChannel, default=-4238.8687, const [-4238.8687,6]}
       Rotation.X {type=MovieSceneScriptingDoubleChannel, default=0}
       Rotation.Y {type=MovieSceneScriptingDoubleChannel, default=0}
       Rotation.Z {type=MovieSceneScriptingDoubleChannel, default=0}
       Scale.X {type=MovieSceneScriptingDoubleChannel, default=1}
       Scale.Y {type=MovieSceneScriptingDoubleChannel, default=1}
       Scale.Z {type=MovieSceneScriptingDoubleChannel, default=1}
      ]
     }]
   }
   {
    type=MovieSceneParticleTrack
    display_name=Particle System
    sections [{
      type=MovieSceneParticleSection
      channels [None {type=MovieSceneScriptingParticleChannel, keys [[190, "<ParticleKey.ACTIVATE: 0>"], [525, "<ParticleKey.DEACTIVATE: 1>"]]}]
     }]
   }
  ]
 }
 P Castle Destruction {
  id=#3
  possessed_class=Emitter
  tracks [
   {
    type=MovieSceneSpawnTrack
    sections [{
      type=MovieSceneSpawnSection
      end_frame=1808
      start_frame=1247
      channels [None {type=MovieSceneScriptingBoolChannel, default=false, keys [[1247,false],[1338.005,true],[1808,false]]}]
     }]
   }
   {
    type=MovieScene3DTransformTrack
    sections [{
      type=MovieScene3DTransformSection
      end_frame=1808
      start_frame=1247
      channels [
       Location.X {type=MovieSceneScriptingDoubleChannel, default=0, const [111301.5469,1]}
       Location.Y {type=MovieSceneScriptingDoubleChannel, default=0, const [-26217.5566,1]}
       Location.Z {type=MovieSceneScriptingDoubleChannel, default=0, const [-110.6338,1]}
       Rotation.X {type=MovieSceneScriptingDoubleChannel, default=0, const [23.3233,1]}
       Rotation.Y {type=MovieSceneScriptingDoubleChannel, default=0, const [4.9825,1]}
       Rotation.Z {type=MovieSceneScriptingDoubleChannel, default=0, const [139.3329,1]}
       Scale.X {type=MovieSceneScriptingDoubleChannel, default=1, const [1,1]}
       Scale.Y {type=MovieSceneScriptingDoubleChannel, default=1, const [1,1]}
       Scale.Z {type=MovieSceneScriptingDoubleChannel, default=1, const [1,1]}
      ]
     }]
   }
   {
    type=MovieSceneParticleTrack
    display_name=Particle System
    sections [{type=MovieSceneParticleSection, channels [None {type=MovieSceneScriptingParticleChannel, const ["<ParticleKey.TRIGGER: 2>", 1]}]}]
   }
  ]
 }
 P Doggus Step Heavy FX {
  id=#4
  possessed_class=Emitter
  tracks [
   {
    type=MovieSceneSpawnTrack
    sections [{
      type=MovieSceneSpawnSection
      end_frame=3047
      start_frame=2756
      channels [None {type=MovieSceneScriptingBoolChannel, default=false, keys [[2757,false],[2758,true],[3040,false],[4091,false]]}]
     }]
   }
   {
    type=MovieScene3DTransformTrack
    sections [{
      type=MovieScene3DTransformSection
      end_frame=3040
      start_frame=2758
      channels [
       Location.X {type=MovieSceneScriptingDoubleChannel, default=116769.5469, const [66802.4375,1]}
       Location.Y {type=MovieSceneScriptingDoubleChannel, default=-39554.2109, const [-48119.5664,1]}
       Location.Z {type=MovieSceneScriptingDoubleChannel, default=2149.8784, const [-1371.2397,1]}
       Rotation.X {type=MovieSceneScriptingDoubleChannel, default=0, const [0,1]}
       Rotation.Y {type=MovieSceneScriptingDoubleChannel, default=0, const [0,1]}
       Rotation.Z {type=MovieSceneScriptingDoubleChannel, default=37.6014, const [-175.9416,1]}
       Scale.X {type=MovieSceneScriptingDoubleChannel, default=1.4}
       Scale.Y {type=MovieSceneScriptingDoubleChannel, default=1.4}
       Scale.Z {type=MovieSceneScriptingDoubleChannel, default=1.4}
      ]
     }]
   }
   {
    type=MovieSceneParticleTrack
    display_name=Particle System
    sections [{type=MovieSceneParticleSection, channels [None {type=MovieSceneScriptingParticleChannel, const ["<ParticleKey.TRIGGER: 2>", 2]}]}]
   }
  ]
 }
 P Doggus Step Light FX {
  id=#5
  possessed_class=Emitter
  tracks [
   {
    type=MovieSceneSpawnTrack
    sections [{
      type=MovieSceneSpawnSection
      start_frame=690
      channels [None {type=MovieSceneScriptingBoolChannel, default=false, keys [[690,true],[4091,false]]}]
     }]
   }
   {
    type=MovieScene3DTransformTrack
    sections [{
      type=MovieScene3DTransformSection
      end_frame=4091
      start_frame=690
      channels [
       Location.X {type=MovieSceneScriptingDoubleChannel, default=123761.6875, const [121264.0156,2]}
       Location.Y {type=MovieSceneScriptingDoubleChannel, default=-35244.4453, const [-30861.4941,2]}
       Location.Z {type=MovieSceneScriptingDoubleChannel, default=1773.166, const [111.95,2]}
       Rotation.X {type=MovieSceneScriptingDoubleChannel, default=1.5301}
       Rotation.Y {type=MovieSceneScriptingDoubleChannel, default=-7.2796}
       Rotation.Z {type=MovieSceneScriptingDoubleChannel, default=69.0458}
       Scale.X {type=MovieSceneScriptingDoubleChannel, default=0.75}
       Scale.Y {type=MovieSceneScriptingDoubleChannel, default=1}
       Scale.Z {type=MovieSceneScriptingDoubleChannel, default=1}
      ]
     }]
   }
   {
    type=MovieSceneParticleTrack
    display_name=Particle System
    sections [{type=MovieSceneParticleSection, channels [None {type=MovieSceneScriptingParticleChannel, const ["<ParticleKey.TRIGGER: 2>", 1]}]}]
   }
  ]
 }
 P Doggus Step Heavy FX {
  id=#6
  possessed_class=Emitter
  tracks [
   {
    type=MovieSceneSpawnTrack
    sections [{
      type=MovieSceneSpawnSection
      end_frame=2973
      start_frame=2788
      channels [None {type=MovieSceneScriptingBoolChannel, default=false, keys [[2788,false],[2791,true],[2973,false],[4091,false]]}]
     }]
   }
   {
    type=MovieScene3DTransformTrack
    sections [{
      type=MovieScene3DTransformSection
      end_frame=2973
      start_frame=2788
      channels [
       Location.X {type=MovieSceneScriptingDoubleChannel, default=116769.5469, const [66549.4375,1]}
       Location.Y {type=MovieSceneScriptingDoubleChannel, default=-39554.2109, const [-30324.9883,1]}
       Location.Z {type=MovieSceneScriptingDoubleChannel, default=2149.8784, const [-633.2802,1]}
       Rotation.X {type=MovieSceneScriptingDoubleChannel, default=0, const [0,1]}
       Rotation.Y {type=MovieSceneScriptingDoubleChannel, default=0, const [0,1]}
       Rotation.Z {type=MovieSceneScriptingDoubleChannel, default=37.6014, const [180,1]}
       Scale.X {type=MovieSceneScriptingDoubleChannel, default=1.4, const [1.4,1]}
       Scale.Y {type=MovieSceneScriptingDoubleChannel, default=1.4, const [1.4,1]}
       Scale.Z {type=MovieSceneScriptingDoubleChannel, default=1.4, const [1.4,1]}
      ]
     }]
   }
   {
    type=MovieSceneParticleTrack
    display_name=Particle System
    sections [{type=MovieSceneParticleSection, channels [None {type=MovieSceneScriptingParticleChannel, const ["<ParticleKey.TRIGGER: 2>", 2]}]}]
   }
  ]
 }
 P Cattus Beam Casting {
  id=#7
  possessed_class=Emitter
  tracks [
   {
    type=MovieSceneSpawnTrack
    sections [{
      type=MovieSceneSpawnSection
      end_frame=3720
      start_frame=3488
      channels [None {type=MovieSceneScriptingBoolChannel, default=false, keys [[3488.3425,false],[3490.3813,true],[3720,false]]}]
     }]
   }
   {
    type=MovieScene3DTransformTrack
    sections [{
      type=MovieScene3DTransformSection
      end_frame=3720
      start_frame=3488
      channels [
       Location.X {type=MovieSceneScriptingDoubleChannel, default=-0.0767, const [-0.0767,1]}
       Location.Y {type=MovieSceneScriptingDoubleChannel, default=0.0147, const [0.0147,1]}
       Location.Z {type=MovieSceneScriptingDoubleChannel, default=-0.0024, const [-0.0024,1]}
       Rotation.X {type=MovieSceneScriptingDoubleChannel, default=0, const [0,1]}
       Rotation.Y {type=MovieSceneScriptingDoubleChannel, default=0, const [0,1]}
       Rotation.Z {type=MovieSceneScriptingDoubleChannel, default=0, const [0,1]}
       Scale.X {type=MovieSceneScriptingDoubleChannel, default=1, const [1,1]}
       Scale.Y {type=MovieSceneScriptingDoubleChannel, default=1, const [1,1]}
       Scale.Z {type=MovieSceneScriptingDoubleChannel, default=1, const [1,1]}
      ]
     }]
   }
   {
    type=MovieSceneParticleTrack
    display_name=Particle System
    sections [{
      type=MovieSceneParticleSection
      channels [None {type=MovieSceneScriptingParticleChannel, keys [[3489.9575, "<ParticleKey.ACTIVATE: 0>"], [3638.7025, "<ParticleKey.DEACTIVATE: 1>"]]}]
     }]
   }
   {
    type=MovieSceneParticleParameterTrack
    display_name=Particle Parameter
    sections [{
      type=MovieSceneParameterSection
      end_frame=3639
      start_frame=3489
      channels [
       LightBrightness {type=MovieSceneScriptingFloatChannel, default=0, keys [[3494.815,0],[3497.0775,1],[3556.545,1],[3579.415,0]]}
       SpawnRate {type=MovieSceneScriptingFloatChannel, default=1, keys [[3512.7225,1],[3535.415,1],[3572.415,0]]}
      ]
     }]
   }
   {type=MovieScene3DAttachTrack, sections [{type=MovieScene3DAttachSection, end_frame=3639, start_frame=3490}]}
  ]
 }
 P Doggus Step Heavy FX {
  id=#8
  possessed_class=Emitter
  tracks [
   {
    type=MovieSceneSpawnTrack
    sections [{
      type=MovieSceneSpawnSection
      end_frame=2973
      start_frame=2690
      channels [None {type=MovieSceneScriptingBoolChannel, default=false, keys [[2690,false],[2691,true],[2973,false],[4091,false]]}]
     }]
   }
   {
    type=MovieScene3DTransformTrack
    sections [{
      type=MovieScene3DTransformSection
      end_frame=2973
      start_frame=2691
      channels [
       Location.X {type=MovieSceneScriptingDoubleChannel, default=116769.5469, const [68019.3672,1]}
       Location.Y {type=MovieSceneScriptingDoubleChannel, default=-39554.2109, const [-28593.332,1]}
       Location.Z {type=MovieSceneScriptingDoubleChannel, default=2149.8784, const [-633.2802,1]}
       Rotation.X {type=MovieSceneScriptingDoubleChannel, default=0, const [0,1]}
       Rotation.Y {type=MovieSceneScriptingDoubleChannel, default=0, const [0,1]}
       Rotation.Z {type=MovieSceneScriptingDoubleChannel, default=37.6014, const [180,1]}
       Scale.X {type=MovieSceneScriptingDoubleChannel, default=1.4}
       Scale.Y {type=MovieSceneScriptingDoubleChannel, default=1.4}
       Scale.Z {type=MovieSceneScriptingDoubleChannel, default=1.4}
      ]
     }]
   }
   {
    type=MovieSceneParticleTrack
    display_name=Particle System
    sections [{type=MovieSceneParticleSection, channels [None {type=MovieSceneScriptingParticleChannel, const ["<ParticleKey.TRIGGER: 2>", 2]}]}]
   }
  ]
 }
 P Cattus Beam Charge Up Yellow On Back FX {
  id=#9
  possessed_class=Emitter
  tracks [
   {
    type=MovieSceneSpawnTrack
    sections [{
      type=MovieSceneSpawnSection
      channels [None {type=MovieSceneScriptingBoolChannel, default=true, keys [[773.2075,false],[789.4887,true],[987.565,false]]}]
     }]
   }
   {
    type=MovieScene3DTransformTrack
    sections [
     {
      type=MovieScene3DTransformSection
      channels [
       Location.X {type=MovieSceneScriptingDoubleChannel, default=0, const [0,1]}
       Location.Y {type=MovieSceneScriptingDoubleChannel, default=0, const [0,1]}
       Location.Z {type=MovieSceneScriptingDoubleChannel, default=0, const [0,1]}
       Rotation.X {type=MovieSceneScriptingDoubleChannel, default=0, const [0,1]}
       Rotation.Y {type=MovieSceneScriptingDoubleChannel, default=0, const [0,1]}
       Rotation.Z {type=MovieSceneScriptingDoubleChannel, default=0, const [0,1]}
       Scale.X {type=MovieSceneScriptingDoubleChannel, default=1, const [1,1]}
       Scale.Y {type=MovieSceneScriptingDoubleChannel, default=1, const [1,1]}
       Scale.Z {type=MovieSceneScriptingDoubleChannel, default=1, const [1,1]}
      ]
     }
     {
      type=MovieScene3DTransformSection
      end_frame=8368
      start_frame=842
      channels [
       Location.X {type=MovieSceneScriptingDoubleChannel, default=1112.156}
       Location.Y {type=MovieSceneScriptingDoubleChannel, default=264.4742}
       Location.Z {type=MovieSceneScriptingDoubleChannel, default=191.201}
       Rotation.X {type=MovieSceneScriptingDoubleChannel, default=-1.1133}
       Rotation.Y {type=MovieSceneScriptingDoubleChannel, default=-15.1532}
       Rotation.Z {type=MovieSceneScriptingDoubleChannel, default=-150.4489}
      ]
     }
     {
      type=MovieScene3DTransformSection
      end_frame=8368
      start_frame=842
      channels [
       Location.X {type=MovieSceneScriptingDoubleChannel, default=-1112.156}
       Location.Y {type=MovieSceneScriptingDoubleChannel, default=-264.4742}
       Location.Z {type=MovieSceneScriptingDoubleChannel, default=-191.201}
       Rotation.X {type=MovieSceneScriptingDoubleChannel, default=1.1133}
       Rotation.Y {type=MovieSceneScriptingDoubleChannel, default=15.1532}
       Rotation.Z {type=MovieSceneScriptingDoubleChannel, default=150.4489}
      ]
     }
    ]
   }
   {type=MovieScene3DAttachTrack, sections [{type=MovieScene3DAttachSection, end_frame=1104, start_frame=784}]}
   {
    type=MovieSceneParticleTrack
    display_name=Particle System
    sections [{
      type=MovieSceneParticleSection
      channels [None {type=MovieSceneScriptingParticleChannel, keys [[788.5838, "<ParticleKey.ACTIVATE: 0>"], [919.7313, "<ParticleKey.DEACTIVATE: 1>"]]}]
     }]
   }
   {
    type=MovieSceneParticleParameterTrack
    display_name=Particle Parameter
    sections [{
      type=MovieSceneParameterSection
      end_frame=1113
      start_frame=780
      channels [
       Opacity {type=MovieSceneScriptingFloatChannel, default=0, keys [[792.5675,1],[919.19,0]]}
       SpawnRate {type=MovieSceneScriptingFloatChannel, default=0, keys [[792.5675,1],[919.19,0]]}
      ]
     }]
   }
  ]
 }
 P Cattus Beam Casting Yellow Mouth FX {
  id=#10
  possessed_class=Emitter
  tracks [
   {
    type=MovieSceneSpawnTrack
    sections [{
      type=MovieSceneSpawnSection
      channels [None {type=MovieSceneScriptingBoolChannel, default=false, keys [[826.5712,false],[842.8512,true],[1002.9388,false],[1421.7013,false]]}]
     }]
   }
   {
    type=MovieScene3DTransformTrack
    sections [{
      type=MovieScene3DTransformSection
      channels [
       Location.X {type=MovieSceneScriptingDoubleChannel, default=0}
       Location.Y {type=MovieSceneScriptingDoubleChannel, default=0}
       Location.Z {type=MovieSceneScriptingDoubleChannel, default=0}
       Rotation.X {type=MovieSceneScriptingDoubleChannel, default=0, const [0,2]}
       Rotation.Y {type=MovieSceneScriptingDoubleChannel, default=0, keys [[844.66,22.7922],[920.5438,-5.4678]]}
       Rotation.Z {type=MovieSceneScriptingDoubleChannel, default=0, keys [[844.66,0],[920.5438,0]]}
       Scale.X {type=MovieSceneScriptingDoubleChannel, default=1}
       Scale.Y {type=MovieSceneScriptingDoubleChannel, default=1}
       Scale.Z {type=MovieSceneScriptingDoubleChannel, default=1}
      ]
     }]
   }
   {type=MovieScene3DAttachTrack, sections [{type=MovieScene3DAttachSection, end_frame=1500, start_frame=842}]}
   {
    type=MovieSceneParticleTrack
    display_name=Particle System
    sections [{
      type=MovieSceneParticleSection
      channels [None {
        type=MovieSceneScriptingParticleChannel
        keys [
         [842.8512, "<ParticleKey.ACTIVATE: 0>"]
         [972.1888, "<ParticleKey.DEACTIVATE: 1>"]
         [1357.485, "<ParticleKey.ACTIVATE: 0>"]
         [1382.81, "<ParticleKey.DEACTIVATE: 1>"]
        ]
       }]
     }]
   }
   {
    type=MovieSceneParticleParameterTrack
    display_name=Particle Parameter
    sections [{
      type=MovieSceneParameterSection
      end_frame=1503
      start_frame=842
      channels [
       LightBrightness {type=MovieSceneScriptingFloatChannel, default=1, keys [[843,0],[845,1],[900.7362,1],[929.6788,0],[1357.485,1],[1391.855,0]]}
       SpawnRate {type=MovieSceneScriptingFloatChannel, default=1, keys [[899.5362,1],[928.4775,0]]}
      ]
     }]
   }
  ]
 }
 "P Doggus Step Heavy Water FX (2)" {
  id=#11
  possessed_class=Emitter
  tracks [
   {type=MovieSceneSpawnTrack, sections [{type=MovieSceneSpawnSection, channels [None {type=MovieSceneScriptingBoolChannel, default=true}]}]}
   {
    type=MovieScene3DTransformTrack
    sections [{
      type=MovieScene3DTransformSection
      channels [
       Location.X {type=MovieSceneScriptingDoubleChannel, default=76436, keys [[2487,71640],[2549,76732]]}
       Location.Y {type=MovieSceneScriptingDoubleChannel, default=-44784, keys [[2487,-38472],[2549,-45080]]}
       Location.Z {type=MovieSceneScriptingDoubleChannel, default=0, keys [[2487,-796],[2549,-780]]}
       Rotation.X {type=MovieSceneScriptingDoubleChannel, default=0}
       Rotation.Y {type=MovieSceneScriptingDoubleChannel, default=0}
       Rotation.Z {type=MovieSceneScriptingDoubleChannel, default=0}
       Scale.X {type=MovieSceneScriptingDoubleChannel, default=1}
       Scale.Y {type=MovieSceneScriptingDoubleChannel, default=1}
       Scale.Z {type=MovieSceneScriptingDoubleChannel, default=1}
      ]
     }]
   }
   {
    type=MovieSceneParticleTrack
    display_name=Particle System
    sections [{type=MovieSceneParticleSection, channels [None {type=MovieSceneScriptingParticleChannel, const ["<ParticleKey.TRIGGER: 2>", 2]}]}]
   }
  ]
 }
 P Doggus Step Light FX {
  id=#12
  possessed_class=Emitter
  tracks [
   {
    type=MovieSceneSpawnTrack
    sections [{
      type=MovieSceneSpawnSection
      start_frame=620
      channels [None {type=MovieSceneScriptingBoolChannel, default=false, keys [[620,true],[4091,false]]}]
     }]
   }
   {
    type=MovieScene3DTransformTrack
    sections [{
      type=MovieScene3DTransformSection
      end_frame=4091
      start_frame=620
      channels [
       Location.X {type=MovieSceneScriptingDoubleChannel, default=123761.6875, keys [[620,123761.6875],[2525,73920.9531]]}
       Location.Y {type=MovieSceneScriptingDoubleChannel, default=-35244.4453, keys [[620,-35244.4453],[2525,-40920.0859]]}
       Location.Z {type=MovieSceneScriptingDoubleChannel, default=1773.166, keys [[620,1773.166],[2525,296.6902]]}
       Rotation.X {type=MovieSceneScriptingDoubleChannel, default=0, keys [[620,0],[2525,1.3953]]}
       Rotation.Y {type=MovieSceneScriptingDoubleChannel, default=-28.5122, keys [[620,-28.5122],[2525,-6.38]]}
       Rotation.Z {type=MovieSceneScriptingDoubleChannel, default=61.4073, keys [[620,61.4073],[2525,573.5331]]}
       Scale.X {type=MovieSceneScriptingDoubleChannel, default=0.75, const [0.75,2]}
       Scale.Y {type=MovieSceneScriptingDoubleChannel, default=1, const [1,2]}
       Scale.Z {type=MovieSceneScriptingDoubleChannel, default=1, const [1,2]}
      ]
     }]
   }
   {
    type=MovieSceneParticleTrack
    display_name=Particle System
    sections [{type=MovieSceneParticleSection, channels [None {type=MovieSceneScriptingParticleChannel, const ["<ParticleKey.TRIGGER: 2>", 1]}]}]
   }
  ]
 }
 "P Doggus Step Heavy FX (2)" {
  id=#13
  possessed_class=Emitter
  tracks [
   {
    type=MovieSceneSpawnTrack
    sections [{
      type=MovieSceneSpawnSection
      start_frame=1086
      channels [None {type=MovieSceneScriptingBoolChannel, default=false, keys [[1086.88,true],[4091,false]]}]
     }]
   }
   {
    type=MovieScene3DTransformTrack
    sections [{
      type=MovieScene3DTransformSection
      end_frame=4091
      start_frame=1086
      channels [
       Location.X {
        type=MovieSceneScriptingDoubleChannel
        default=115944
        keys [[1086.88,115944],[1471,115944],[2552,76086.8828],[3088.07,69698.3281],[3489.92,59859.4531],[3907.5087,56733.9844]]
       }
       Location.Y {
        type=MovieSceneScriptingDoubleChannel
        default=-27909.4805
        keys [[1086.88,-27909.4805],[1471,-27909.4805],[2552,-44315.3281],[3088.07,-36086.082],[3489.92,-27281.791],[3907.5087,-27572.4043]]
       }
       Location.Z {
        type=MovieSceneScriptingDoubleChannel
        default=-721.2304
        thinned_from=6
        keys [[1086.88,-721.2304],[3489.92,-721.2304],[3907.5087,146.3697]]
       }
       Rotation.X {type=MovieSceneScriptingDoubleChannel, default=0, const [0,6]}
       Rotation.Y {type=MovieSceneScriptingDoubleChannel, default=0, const [0,6]}
       Rotation.Z {
        type=MovieSceneScriptingDoubleChannel
        default=0
        keys [[1086.88,0],[1471,0],[2552,68.5648],[3088.07,49.6429],[3489.92,91.5841],[3907.5087,91.5841]]
       }
       Scale.X {type=MovieSceneScriptingDoubleChannel, default=1, thinned_from=6, keys [[1086.88,1],[3489.92,1],[3907.5087,1.3]]}
       Scale.Y {type=MovieSceneScriptingDoubleChannel, default=1, thinned_from=6, keys [[1086.88,1],[3489.92,1],[3907.5087,1.3]]}
       Scale.Z {type=MovieSceneScriptingDoubleChannel, default=2, thinned_from=6, keys [[1086.88,2],[1471,1],[3489.92,1],[3907.5087,1.3]]}
      ]
     }]
   }
   {
    type=MovieSceneParticleTrack
    display_name=Particle System
    sections [{type=MovieSceneParticleSection, channels [None {type=MovieSceneScriptingParticleChannel, const ["<ParticleKey.TRIGGER: 2>", 4]}]}]
   }
  ]
 }
 P Water Raid Foot L {
  id=#14
  possessed_class=Emitter
  tracks [
   {
    type=MovieSceneSpawnTrack
    sections [{
      type=MovieSceneSpawnSection
      end_frame=589
      start_frame=210
      channels [None {type=MovieSceneScriptingBoolChannel, default=false, keys [[210,false],[226,true],[589,false]]}]
     }]
   }
   {
    type=MovieScene3DTransformTrack
    sections [{
      type=MovieScene3DTransformSection
      end_frame=589
      start_frame=210
      channels [
       Location.X {type=MovieSceneScriptingDoubleChannel, default=142877.9844, keys [[228,142877.9844],[273,138322.4531],[504,138322.4531]]}
       Location.Y {type=MovieSceneScriptingDoubleChannel, default=-21999.0527, const [-21999.0527,3]}
       Location.Z {type=MovieSceneScriptingDoubleChannel, default=-4238.8687, thinned_from=3, keys [[228,-5013.6201],[504,-5012.8315]]}
       Rotation.X {type=MovieSceneScriptingDoubleChannel, default=0}
       Rotation.Y {type=MovieSceneScriptingDoubleChannel, default=0}
       Rotation.Z {type=MovieSceneScriptingDoubleChannel, default=0}
       Scale.X {type=MovieSceneScriptingDoubleChannel, default=1}
       Scale.Y {type=MovieSceneScriptingDoubleChannel, default=1}
       Scale.Z {type=MovieSceneScriptingDoubleChannel, default=1}
      ]
     }]
   }
   {
    type=MovieSceneParticleTrack
    display_name=Particle System
    sections [{
      type=MovieSceneParticleSection
      channels [None {type=MovieSceneScriptingParticleChannel, keys [[226, "<ParticleKey.ACTIVATE: 0>"], [525, "<ParticleKey.DEACTIVATE: 1>"]]}]
     }]
   }
  ]
 }
 P Cattus Drool Spit {
  id=#15
  possessed_class=Emitter
  tracks [
   {
    type=MovieSceneSpawnTrack
    sections [{
      type=MovieSceneSpawnSection
      end_frame=2101
      start_frame=531
      channels [None {
        type=MovieSceneScriptingBoolChannel
        default=false
        keys [[531,false],[578,true],[876,false],[1098,true],[1283,false],[1445,true],[2101,false]]
       }]
     }]
   }
   {
    type=MovieScene3DTransformTrack
    sections [{
      type=MovieScene3DTransformSection
      end_frame=2101
      start_frame=531
      channels [
       Location.X {type=MovieSceneScriptingDoubleChannel, default=-13.8635, const [-7.5298,1]}
       Location.Y {type=MovieSceneScriptingDoubleChannel, default=18.4708, const [31.8889,1]}
       Location.Z {type=MovieSceneScriptingDoubleChannel, default=14.642, const [-0.2742,1]}
       Rotation.X {type=MovieSceneScriptingDoubleChannel, default=-146.5356, const [-5759.2417,1]}
       Rotation.Y {type=MovieSceneScriptingDoubleChannel, default=71.1346, const [0.6826,1]}
       Rotation.Z {type=MovieSceneScriptingDoubleChannel, default=-145.0673, const [-5654.501,1]}
       Scale.X {type=MovieSceneScriptingDoubleChannel, default=1, const [1,1]}
       Scale.Y {type=MovieSceneScriptingDoubleChannel, default=1, const [1,1]}
       Scale.Z {type=MovieSceneScriptingDoubleChannel, default=1, const [1,1]}
      ]
     }]
   }
   {type=MovieScene3DAttachTrack, sections [{type=MovieScene3DAttachSection, end_frame=2164, start_frame=578}]}
   {
    type=MovieSceneParticleTrack
    display_name=Particle System
    sections [{
      type=MovieSceneParticleSection
      channels [None {
        type=MovieSceneScriptingParticleChannel
        keys [
         [578, "<ParticleKey.ACTIVATE: 0>"]
         [646, "<ParticleKey.DEACTIVATE: 1>"]
         [718, "<ParticleKey.ACTIVATE: 0>"]
         [814, "<ParticleKey.DEACTIVATE: 1>"]
         [1098, "<ParticleKey.ACTIVATE: 0>"]
         [1231, "<ParticleKey.DEACTIVATE: 1>"]
         [1445, "<ParticleKey.ACTIVATE: 0>"]
         [2029, "<ParticleKey.DEACTIVATE: 1>"]
        ]
       }]
     }]
   }
  ]
 }
 P Water Raid Foot R {
  id=#16
  possessed_class=Emitter
  tracks [
   {
    type=MovieSceneSpawnTrack
    sections [{
      type=MovieSceneSpawnSection
      end_frame=668
      start_frame=210
      channels [None {type=MovieSceneScriptingBoolChannel, default=false, keys [[210,false],[226,true],[668,false]]}]
     }]
   }
   {
    type=MovieScene3DTransformTrack
    sections [{
      type=MovieScene3DTransformSection
      end_frame=668
      channels [
       Location.X {
        type=MovieSceneScriptingDoubleChannel
        default=141613.3125
        keys [[260,140969.3438],[273,141249.3438],[325,138794.3594],[429,138794.3594],[489,132834.5312]]
       }
       Location.Y {
        type=MovieSceneScriptingDoubleChannel
        default=-27963.5391
        keys [[260,-27132.9492],[273,-26894.1309],[325,-29631.2012],[429,-29631.2012],[489,-28577.252]]
       }
       Location.Z {type=MovieSceneScriptingDoubleChannel, default=-4238.8687, const [-4238.8687,5]}
       Rotation.X {type=MovieSceneScriptingDoubleChannel, default=0}
       Rotation.Y {type=MovieSceneScriptingDoubleChannel, default=0}
       Rotation.Z {type=MovieSceneScriptingDoubleChannel, default=0}
       Scale.X {type=MovieSceneScriptingDoubleChannel, default=1}
       Scale.Y {type=MovieSceneScriptingDoubleChannel, default=1}
       Scale.Z {type=MovieSceneScriptingDoubleChannel, default=1}
      ]
     }]
   }
   {
    type=MovieSceneParticleTrack
    display_name=Particle System
    sections [{
      type=MovieSceneParticleSection
      channels [None {type=MovieSceneScriptingParticleChannel, keys [[226, "<ParticleKey.ACTIVATE: 0>"], [560, "<ParticleKey.DEACTIVATE: 1>"]]}]
     }]
   }
  ]
 }
 P Cattus Beam Yellow Extended {
  id=#17
  possessed_class=Emitter
  tracks [
   {
    type=MovieSceneSpawnTrack
    sections [{
      type=MovieSceneSpawnSection
      start_frame=845
      channels [None {type=MovieSceneScriptingBoolChannel, default=true, keys [[845,false],[846,true],[957,false]]}]
     }]
   }
   {
    type=MovieScene3DTransformTrack
    sections [{
      type=MovieScene3DTransformSection
      start_frame=845
      channels [
       Location.X {type=MovieSceneScriptingDoubleChannel, default=0.3888}
       Location.Y {type=MovieSceneScriptingDoubleChannel, default=-0.0002}
       Location.Z {type=MovieSceneScriptingDoubleChannel, default=0.0495}
       Rotation.X {
        type=MovieSceneScriptingDoubleChannel
        default=-0.0001
        thinned_from=8
        keys [[846,0.6899],[875.6737,-0.0031],[920.5438,-0.001],[942.9787,-0.0008],[958.87,-0.0021],[964.4787,-0.0015],[1356.0188,-0.0023]]
       }
       Rotation.Y {
        type=MovieSceneScriptingDoubleChannel
        default=22.7921
        keys [[846,12.934],[875.6737,25.058],[920.5438,-5.4678],[942.9787,4.3677],[952.3262,7.1645],[958.87,-13.4695],[964.4787,-9.3159],[1356.0188,19.2584]]
       }
       Rotation.Z {
        type=MovieSceneScriptingDoubleChannel
        default=0.0003
        keys [[846,-0.127],[875.6737,-0.0048],[920.5438,0.001],[942.9787,0.0001],[952.3262,0.0004],[958.87,0.0014],[964.4787,0.0004],[1356.0188,-0.0014]]
       }
       Scale.X {type=MovieSceneScriptingDoubleChannel, default=1}
       Scale.Y {type=MovieSceneScriptingDoubleChannel, default=1}
       Scale.Z {type=MovieSceneScriptingDoubleChannel, default=1}
      ]
     }]
   }
   {type=MovieScene3DAttachTrack, sections [{type=MovieScene3DAttachSection, end_frame=1421, start_frame=846}]}
   {
    type=MovieSceneParticleTrack
    display_name=Particle System
    sections [{
      type=MovieSceneParticleSection
      channels [None {type=MovieSceneScriptingParticleChannel, keys [[846, "<ParticleKey.ACTIVATE: 0>"], [957, "<ParticleKey.DEACTIVATE: 1>"]]}]
     }]
   }
   {
    type=MovieSceneParticleParameterTrack
    display_name=Particle Parameter
    sections [{
      type=MovieSceneParameterSection
      end_frame=1479
      start_frame=846
      channels [
       BeamErode2 {type=MovieSceneScriptingFloatChannel, default=1, keys [[847,0],[940,0],[946.5725,1]]}
       SpawnRate {type=MovieSceneScriptingFloatChannel, default=0, keys [[847,1],[940,1],[946.5725,0]]}
       BeamOpacity {type=MovieSceneScriptingFloatChannel, default=1, keys [[847,1],[940,1],[946.5725,0]]}
       BeamSize.X {type=MovieSceneScriptingFloatChannel, default=1, keys [[847,2],[911.2937,1.9549],[938.4275,1]]}
       BeamSize.Y {type=MovieSceneScriptingFloatChannel, default=1, const [1,1]}
       BeamSize.Z {type=MovieSceneScriptingFloatChannel, default=1, const [1,1]}
       BeamTarget.X {
        type=MovieSceneScriptingFloatChannel
        default=400
        keys [[847,20],[861.305,800],[876,615.7266],[883,625],[893,625],[895,565.2388],[905,516.0457],[935,500],[958,400]]
       }
       BeamTarget.Y {type=MovieSceneScriptingFloatChannel}
       BeamTarget.Z {type=MovieSceneScriptingFloatChannel}
       BeamBrightness.X {type=MovieSceneScriptingFloatChannel, default=1, const [1,1]}
       BeamBrightness.Y {type=MovieSceneScriptingFloatChannel, default=1, const [1,1]}
       BeamBrightness.Z {type=MovieSceneScriptingFloatChannel, default=0.5, const [0.5,1]}
      ]
     }]
   }
  ]
 }
 P Doggus Step Light FX {
  id=#18
  possessed_class=Emitter
  tracks [
   {
    type=MovieSceneSpawnTrack
    sections [{
      type=MovieSceneSpawnSection
      start_frame=555
      channels [None {type=MovieSceneScriptingBoolChannel, default=false, keys [[555,true],[4091,false]]}]
     }]
   }
   {
    type=MovieScene3DTransformTrack
    sections [{
      type=MovieScene3DTransformSection
      end_frame=4091
      start_frame=555
      channels [
       Location.X {
        type=MovieSceneScriptingDoubleChannel
        default=116033.6641
        thinned_from=16
        keys [[544,116033.6641],[658,113636.6719],[1616,113636.6719],[1732,101807.2266],[1981,77211.1875],[2076,76371.3125],[2439,73552.3047],[2542,74161.6719],[2865,74592.0859],[3037.7075,74645.4219],[3109,67217.4219],[3229,70777.4219],[3456.345,68710.7266],[3485,63934.7266],[3686,60670.7266]]
       }
       Location.Y {
        type=MovieSceneScriptingDoubleChannel
        default=-22146.1367
        keys [[544,-22146.1367],[658,-40422.1133],[1616,-37131.0312],[1732,-38086.1484],[1865,-35809.6211],[1981,-34829.1328],[2076,-33421.3594],[2439,-38002.6133],[2542,-35291.2109],[2865,-35888.2422],[3037.7075,-23885.6602],[3109,-38253.6602],[3229,-47665.6602],[3456.345,-33360.7852],[3485,-36140.7852],[3686,-33092.7852]]
       }
       Location.Z {
        type=MovieSceneScriptingDoubleChannel
        default=-2565.9028
        keys [[544,-2565.9028],[658,1376.2988],[1616,1376.2988],[1732,663.6601],[1865,611.5743],[1981,-126.5782],[2076,-142.5841],[2439,-634.8032],[2542,221.9471],[2865,228.0548],[3037.7075,528.1488],[3109,-639.8512],[3456.345,-718.7278],[3485,-1226.7278],[3686,-718.7278]]
       }
       Rotation.X {
        type=MovieSceneScriptingDoubleChannel
        default=0
        keys [[1616,0],[1732,27.0056],[1865,0.9031],[1981,0.9031],[2076,-0.6558],[2439,0.1271],[2542,1.0389],[2865,0.325],[3037.7075,-16.262],[3456.345,1.5842]]
       }
       Rotation.Y {
        type=MovieSceneScriptingDoubleChannel
        default=0
        keys [[1616,0],[1732,2.5756],[1865,-0.6243],[1981,-0.6243],[2076,-0.8804],[2439,-1.0903],[2542,0.3546],[2865,1.0485],[3037.7075,-19.4951],[3456.345,-2.5456]]
       }
       Rotation.Z {
        type=MovieSceneScriptingDoubleChannel
        default=0
        keys [[1616,0],[1732,1.3119],[1865,6.1076],[1981,6.1076],[2076,98.1386],[2439,54.802],[2542,-47.3899],[2865,-101.3243],[3037.7075,-83.7247],[3456.345,-71.5701]]
       }
       Scale.X {type=MovieSceneScriptingDoubleChannel, default=1, const [1,10]}
       Scale.Y {type=MovieSceneScriptingDoubleChannel, default=1, const [1,10]}
       Scale.Z {type=MovieSceneScriptingDoubleChannel, default=1, const [1,10]}
      ]
     }]
   }
   {
    type=MovieSceneParticleTrack
    display_name=Particle System
    sections [{type=MovieSceneParticleSection, channels [None {type=MovieSceneScriptingParticleChannel, const ["<ParticleKey.TRIGGER: 2>", 14]}]}]
   }
  ]
 }
 P Cattus Beam Yellow Impact FX {
  id=#19
  possessed_class=Emitter
  tracks [
   {
    type=MovieSceneSpawnTrack
    sections [{
      type=MovieSceneSpawnSection
      start_frame=842
      channels [None {type=MovieSceneScriptingBoolChannel, default=false, keys [[842,false],[845,true],[1356.0188,true]]}]
     }]
   }
   {
    type=MovieScene3DTransformTrack
    sections [{
      type=MovieScene3DTransformSection
      start_frame=845
      channels [
       Location.X {type=MovieSceneScriptingDoubleChannel, default=0}
       Location.Y {type=MovieSceneScriptingDoubleChannel, default=0}
       Location.Z {type=MovieSceneScriptingDoubleChannel, default=0}
       Rotation.X {type=MovieSceneScriptingDoubleChannel, default=0}
       Rotation.Y {type=MovieSceneScriptingDoubleChannel, default=0}
       Rotation.Z {type=MovieSceneScriptingDoubleChannel, default=0}
       Scale.X {type=MovieSceneScriptingDoubleChannel, default=1}
       Scale.Y {type=MovieSceneScriptingDoubleChannel, default=1}
       Scale.Z {type=MovieSceneScriptingDoubleChannel, default=1}
      ]
     }]
   }
   {type=MovieScene3DAttachTrack, sections [{type=MovieScene3DAttachSection, end_frame=1356, start_frame=846}]}
   {
    type=MovieSceneParticleTrack
    display_name=Particle System
    sections [{
      type=MovieSceneParticleSection
      channels [None {type=MovieSceneScriptingParticleChannel, keys [[845, "<ParticleKey.ACTIVATE: 0>"], [945, "<ParticleKey.DEACTIVATE: 1>"]]}]
     }]
   }
   {
    type=MovieSceneParticleParameterTrack
    display_name=Particle Parameter
    sections [{
      type=MovieSceneParameterSection
      end_frame=952
      start_frame=936
      channels [Alpha {type=MovieSceneScriptingFloatChannel, default=1, keys [[936,1],[948,0]]}]
     }]
   }
  ]
 }
 P Doggus Step Light FX {
  id=#20
  possessed_class=Emitter
  tracks [
   {
    type=MovieSceneSpawnTrack
    sections [{
      type=MovieSceneSpawnSection
      start_frame=555
      channels [None {type=MovieSceneScriptingBoolChannel, default=false, keys [[555,true],[4091,false]]}]
     }]
   }
   {
    type=MovieScene3DTransformTrack
    sections [{
      type=MovieScene3DTransformSection
      end_frame=4091
      start_frame=555
      channels [
       Location.X {
        type=MovieSceneScriptingDoubleChannel
        default=116033.6641
        keys [[555,125982.7656],[679,110163.0078],[1231,113535.3594],[1679,109653.9062],[1796,95772.2109],[1925,81792.9688],[2040,69598.8906],[2464,73641.5391],[2704,72838.5469],[2905,69370.8359],[3411.2287,64118.0195]]
       }
       Location.Y {
        type=MovieSceneScriptingDoubleChannel
        default=-22146.1367
        keys [[555,-25323.0703],[679,-32169.2246],[1231,-32169.2266],[1679,-30775.6602],[1796,-30968.5137],[1925,-29481.0098],[2040,-29304.9824],[2464,-28408.2793],[2704,-44047.1016],[2905,-43460.4375],[3411.2287,-41703.6719]]
       }
       Location.Z {
        type=MovieSceneScriptingDoubleChannel
        default=-2565.9028
        keys [[555,-2565.9028],[679,114.0146],[1231,114.0146],[1679,-769.3072],[1796,-355.9572],[1925,243.6403],[2040,-422.3213],[2464,-255.3307],[2704,-373.0457],[2905,-1854.9166],[3411.2287,-1599.3778]]
       }
       Rotation.X {
        type=MovieSceneScriptingDoubleChannel
        default=0
        keys [[1231,0],[1679,0.9858],[1796,0.9858],[1925,2.1859],[2040,3.8156],[2464,-1.6948],[2704,-3.935],[2905,-3.6111],[3411.2287,-1.8153]]
       }
       Rotation.Y {
        type=MovieSceneScriptingDoubleChannel
        default=0
        keys [[1231,0],[1679,7.0787],[1796,7.0787],[1925,3.8816],[2040,2.2992],[2464,4.1193],[2704,2.0877],[2905,-2.6084],[3411.2287,-4.0675]]
       }
       Rotation.Z {
        type=MovieSceneScriptingDoubleChannel
        default=0
        keys [[1231,0],[1679,-26.7773],[1796,-26.7774],[1925,-16.8977],[2040,12.6584],[2464,-68.7005],[2704,-108.4065],[2905,-172.1087],[3411.2287,-1282.2369]]
       }
       Scale.X {type=MovieSceneScriptingDoubleChannel, default=1, const [1,7]}
       Scale.Y {type=MovieSceneScriptingDoubleChannel, default=1, const [1,7]}
       Scale.Z {type=MovieSceneScriptingDoubleChannel, default=1, const [1,7]}
      ]
     }]
   }
   {
    type=MovieSceneParticleTrack
    display_name=Particle System
    sections [{type=MovieSceneParticleSection, channels [None {type=MovieSceneScriptingParticleChannel, const ["<ParticleKey.TRIGGER: 2>", 10]}]}]
   }
  ]
 }
 "P Doggus Step Heavy FX (3)" {
  id=#21
  possessed_class=Emitter
  tracks [
   {
    type=MovieSceneSpawnTrack
    sections [{
      type=MovieSceneSpawnSection
      end_frame=1731
      start_frame=1301
      channels [None {type=MovieSceneScriptingBoolChannel, default=false, keys [[1301,false],[1338.005,true],[1731,false]]}]
     }]
   }
   {
    type=MovieScene3DTransformTrack
    sections [{
      type=MovieScene3DTransformSection
      end_frame=1731
      start_frame=1301
      channels [
       Location.X {type=MovieSceneScriptingDoubleChannel, default=111147.5781, const [111147.5781,1]}
       Location.Y {type=MovieSceneScriptingDoubleChannel, default=-26250.3652, const [-26250.3652,1]}
       Location.Z {type=MovieSceneScriptingDoubleChannel, default=692.6227, const [692.6227,1]}
       Rotation.X {type=MovieSceneScriptingDoubleChannel, default=0, const [0,1]}
       Rotation.Y {type=MovieSceneScriptingDoubleChannel, default=0, const [0,1]}
       Rotation.Z {type=MovieSceneScriptingDoubleChannel, default=0, const [0,1]}
       Scale.X {type=MovieSceneScriptingDoubleChannel, default=1.4, const [1.4,1]}
       Scale.Y {type=MovieSceneScriptingDoubleChannel, default=1.4, const [1.4,1]}
       Scale.Z {type=MovieSceneScriptingDoubleChannel, default=1.4, const [1.4,1]}
      ]
     }]
   }
   {
    type=MovieSceneParticleTrack
    display_name=Particle System
    sections [{type=MovieSceneParticleSection, channels [None {type=MovieSceneScriptingParticleChannel, const ["<ParticleKey.TRIGGER: 2>", 1]}]}]
   }
  ]
 }
 P Cattus Giant Water Drips {
  id=#22
  possessed_class=Emitter
  tracks [
   {
    type=MovieSceneSpawnTrack
    sections [{
      type=MovieSceneSpawnSection
      end_frame=758
      start_frame=175
      channels [None {type=MovieSceneScriptingBoolChannel, default=false, keys [[175,false],[175,true],[758,false]]}]
     }]
   }
   {
    type=MovieScene3DTransformTrack
    sections [{
      type=MovieScene3DTransformSection
      end_frame=758
      start_frame=175
      channels [
       Location.X {type=MovieSceneScriptingDoubleChannel, default=0}
       Location.Y {type=MovieSceneScriptingDoubleChannel, default=0}
       Location.Z {type=MovieSceneScriptingDoubleChannel, default=0}
       Rotation.X {type=MovieSceneScriptingDoubleChannel, default=0}
       Rotation.Y {type=MovieSceneScriptingDoubleChannel, default=0}
       Rotation.Z {type=MovieSceneScriptingDoubleChannel, default=0}
       Scale.X {type=MovieSceneScriptingDoubleChannel, default=1}
       Scale.Y {type=MovieSceneScriptingDoubleChannel, default=1}
       Scale.Z {type=MovieSceneScriptingDoubleChannel, default=1}
      ]
     }]
   }
   {type=MovieScene3DAttachTrack, sections [{type=MovieScene3DAttachSection, end_frame=8189, start_frame=-47}]}
   {
    type=MovieSceneParticleTrack
    display_name=Particle System
    sections [{
      type=MovieSceneParticleSection
      channels [None {type=MovieSceneScriptingParticleChannel, keys [[160, "<ParticleKey.ACTIVATE: 0>"], [639, "<ParticleKey.DEACTIVATE: 1>"]]}]
     }]
   }
   {
    type=MovieSceneParticleParameterTrack
    display_name=Particle Parameter
    sections [{
      type=MovieSceneParameterSection
      end_frame=700
      start_frame=-36
      channels [
       SpawnRate {type=MovieSceneScriptingFloatChannel, default=0, keys [[175,1],[380,1],[486,0]]}
       SpawnRateOnFeet {type=MovieSceneScriptingFloatChannel, default=0, keys [[175,1],[547,1],[639,0]]}
      ]
     }]
   }
  ]
 }
 P Cattus Beam Aftermath {
  id=#23
  possessed_class=Emitter
  tracks [
   {
    type=MovieSceneSpawnTrack
    sections [{
      type=MovieSceneSpawnSection
      end_frame=4093
      start_frame=3528
      channels [None {type=MovieSceneScriptingBoolChannel, default=false, keys [[3528.3125,false],[3535.415,true],[4093.6437,false]]}]
     }]
   }
   {
    type=MovieScene3DTransformTrack
    sections [{
      type=MovieScene3DTransformSection
      end_frame=4093
      start_frame=3528
      channels [
       Location.X {type=MovieSceneScriptingDoubleChannel, default=0}
       Location.Y {type=MovieSceneScriptingDoubleChannel, default=0}
       Location.Z {type=MovieSceneScriptingDoubleChannel, default=0}
       Rotation.X {type=MovieSceneScriptingDoubleChannel, default=0}
       Rotation.Y {type=MovieSceneScriptingDoubleChannel, default=0}
       Rotation.Z {type=MovieSceneScriptingDoubleChannel, default=0}
       Scale.X {type=MovieSceneScriptingDoubleChannel, default=1}
       Scale.Y {type=MovieSceneScriptingDoubleChannel, default=1}
       Scale.Z {type=MovieSceneScriptingDoubleChannel, default=1}
      ]
     }]
   }
   {type=MovieScene3DAttachTrack, sections [{type=MovieScene3DAttachSection, end_frame=4106, start_frame=3535}]}
   {
    type=MovieSceneParticleTrack
    display_name=Particle System
    sections [{
      type=MovieSceneParticleSection
      channels [None {type=MovieSceneScriptingParticleChannel, keys [[3535.415, "<ParticleKey.ACTIVATE: 0>"], [4087.2988, "<ParticleKey.DEACTIVATE: 1>"]]}]
     }]
   }
   {
    type=MovieSceneParticleParameterTrack
    display_name=Particle Parameter
    sections [{
      type=MovieSceneParameterSection
      end_frame=4097
      start_frame=3517
      channels [
       SpawnRate {type=MovieSceneScriptingFloatChannel, default=1, keys [[3573.5738,1],[3854.6463,0]]}
       LightBrightness {type=MovieSceneScriptingFloatChannel, default=0.1, keys [[3572.0625,1],[3875.5013,1],[3945.4038,0.1]]}
       SpawnRateFire {type=MovieSceneScriptingFloatChannel, default=0, keys [[3873.7763,1],[3926.9112,0]]}
       ColorR {type=MovieSceneScriptingFloatChannel, default=5, const [5,1]}
       ColorG {type=MovieSceneScriptingFloatChannel, default=2.5, keys [[3540,2.5],[3540.415,1]]}
       ColorB {type=MovieSceneScriptingFloatChannel, default=0.1, const [0.1,1]}
       ColorA {type=MovieSceneScriptingFloatChannel, const [0,1]}
      ]
     }]
   }
  ]
 }
 P Cattus Ice Destruction {
  id=#24
  possessed_class=Emitter
  tracks [
   {
    type=MovieSceneSpawnTrack
    sections [{
      type=MovieSceneSpawnSection
      end_frame=2872
      start_frame=2470
      channels [None {type=MovieSceneScriptingBoolChannel, default=false, keys [[2470,true],[2872,false]]}]
     }]
   }
   {
    type=MovieScene3DTransformTrack
    sections [{
      type=MovieScene3DTransformSection
      end_frame=2872
      start_frame=2469
      channels [
       Location.X {type=MovieSceneScriptingDoubleChannel, default=68144}
       Location.Y {type=MovieSceneScriptingDoubleChannel, default=-29500}
       Location.Z {type=MovieSceneScriptingDoubleChannel, default=14204}
       Rotation.X {type=MovieSceneScriptingDoubleChannel, default=0}
       Rotation.Y {type=MovieSceneScriptingDoubleChannel, default=0}
       Rotation.Z {type=MovieSceneScriptingDoubleChannel, default=0}
       Scale.X {type=MovieSceneScriptingDoubleChannel, default=2.2}
       Scale.Y {type=MovieSceneScriptingDoubleChannel, default=2.2}
       Scale.Z {type=MovieSceneScriptingDoubleChannel, default=2.2}
      ]
     }]
   }
   {
    type=MovieSceneParticleTrack
    display_name=Particle System
    sections [{
      type=MovieSceneParticleSection
      channels [None {type=MovieSceneScriptingParticleChannel, keys [[2470, "<ParticleKey.ACTIVATE: 0>"], [2518, "<ParticleKey.DEACTIVATE: 1>"]]}]
     }]
   }
  ]
 }
 Impact_Point_Actor {
  id=#25
  possessed_class=FortStaticMeshActor
  tracks [
   {
    type=MovieSceneSpawnTrack
    sections [{
      type=MovieSceneSpawnSection
      end_frame=1356
      start_frame=827
      channels [None {type=MovieSceneScriptingBoolChannel, default=true, keys [[827,true],[1356.0188,false]]}]
     }]
   }
   {type=MovieScene3DAttachTrack, sections [{type=MovieScene3DAttachSection, end_frame=1356, start_frame=827}]}
   {
    type=MovieScene3DTransformTrack
    sections [{
      type=MovieScene3DTransformSection
      end_frame=1356
      start_frame=827
      channels [
       Location.X {
        type=MovieSceneScriptingDoubleChannel
        default=0.3888
        keys [[845,211.0989],[849,211.0123],[862,237.5277],[867,369.5477],[870,473.8118],[872,480.2803],[876,500.8066],[880,546.4376],[882,573.0652],[886,559.5035],[890,596.6762],[892,631.8568],[894,526.3513],[896,532.8368],[899,483.8308],[902,463.8477],[906,442.4891],[909,433.9324],[913,442.4843],[916,428.2762],[920,418.2015],[923,425.0045],[928,433.483],[932,421.9829],[936,430.679],[939,436.1808],[943,404.7885],[948,438.4038],[952,446.1676]]
       }
       Location.Y {
        type=MovieSceneScriptingDoubleChannel
        default=-0.4618
        thinned_from=29
        keys [[845,-0.4672],[849,-0.4385],[862,-1.176],[867,-0.6718],[870,-0.6313],[872,-0.5818],[876,-0.6268],[880,-0.6255],[882,-0.758],[886,-0.6147],[890,-0.5873],[892,-0.5838],[894,-0.5678],[896,-0.5679],[899,-0.5565],[902,-0.5431],[906,-0.4472],[909,-0.4684],[913,-0.4979],[916,-0.7335],[920,-0.4824],[923,-0.4806],[928,-0.482],[932,-0.4816],[936,-0.4788],[943,-0.4857],[952,-0.488]]
       }
       Location.Z {
        type=MovieSceneScriptingDoubleChannel
        default=47.8673
        keys [[845,48.4741],[849,53.7834],[862,89.6983],[867,153.0012],[870,219.7506],[872,223.988],[876,231.2418],[880,243.3324],[882,249.9285],[886,216.8311],[890,194.1203],[892,185.8156],[894,134.4045],[896,114.4256],[899,75.8528],[902,52.3692],[906,18.3914],[909,-1.6043],[913,-21.2422],[916,-31.0154],[920,-35.3205],[923,-33.1787],[928,-25.8162],[932,-11.4578],[936,7.5687],[939,21.6289],[943,34.302],[948,38.8996],[952,38.1744]]
       }
       Rotation.X {
        type=MovieSceneScriptingDoubleChannel
        default=-0.0001
        thinned_from=8
        keys [[845,0.6899],[875.6737,-0.0031],[920.5438,-0.001],[943,-0.0008],[958.87,-0.0021],[964.4787,-0.0015],[1356.0188,-0.0023]]
       }
       Rotation.Y {
        type=MovieSceneScriptingDoubleChannel
        default=22.7921
        keys [[845,12.934],[875.6737,25.058],[920.5438,-5.4678],[943,4.3677],[952.3262,7.1645],[958.87,-13.4695],[964.4787,-9.3159],[1356.0188,19.2584]]
       }
       Rotation.Z {
        type=MovieSceneScriptingDoubleChannel
        default=0.0003
        keys [[845,-0.127],[875.6737,-0.0048],[920.5438,0.001],[943,0.0001],[952.3262,0.0004],[958.87,0.0014],[964.4787,0.0004],[1356.0188,-0.0014]]
       }
       Scale.X {type=MovieSceneScriptingDoubleChannel, default=1}
       Scale.Y {type=MovieSceneScriptingDoubleChannel, default=1}
       Scale.Z {type=MovieSceneScriptingDoubleChannel, default=1}
      ]
     }]
   }
  ]
 }
 P Cattus Roar Soundwave {
  id=#26
  possessed_class=Emitter
  tracks [
   {
    type=MovieSceneSpawnTrack
    sections [{
      type=MovieSceneSpawnSection
      end_frame=4018
      start_frame=817
      channels [None {type=MovieSceneScriptingBoolChannel, default=false, keys [[823.9562,false],[2905.395,true],[4018.695,false]]}]
     }]
   }
   {
    type=MovieScene3DTransformTrack
    sections [{
      type=MovieScene3DTransformSection
      end_frame=4018
      start_frame=833
      channels [
       Location.X {type=MovieSceneScriptingDoubleChannel, default=-5.7667, keys [[2928.4937,-5.7667],[3811.97,-1.7623]]}
       Location.Y {type=MovieSceneScriptingDoubleChannel, default=-8.9062, keys [[2928.4937,-8.9062],[3811.97,-20.972]]}
       Location.Z {type=MovieSceneScriptingDoubleChannel, default=0.0003, keys [[2928.4937,0.0003],[3811.97,0.0024]]}
       Rotation.X {type=MovieSceneScriptingDoubleChannel, default=-0.0017}
       Rotation.Y {type=MovieSceneScriptingDoubleChannel, default=-0.0001}
       Rotation.Z {type=MovieSceneScriptingDoubleChannel, default=108.3581}
       Scale.X {type=MovieSceneScriptingDoubleChannel, default=1.3, keys [[2906,1.3],[3811.97,1.3]]}
       Scale.Y {type=MovieSceneScriptingDoubleChannel, default=1.3, const [1.3,2]}
       Scale.Z {type=MovieSceneScriptingDoubleChannel, default=1.3, const [1.3,2]}
      ]
     }]
   }
   {type=MovieScene3DAttachTrack, sections [{type=MovieScene3DAttachSection, end_frame=4025, start_frame=833}]}
   {
    type=MovieSceneParticleTrack
    display_name=Particle System
    sections [{
      type=MovieSceneParticleSection
      channels [None {
        type=MovieSceneScriptingParticleChannel
        keys [
         [2917, "<ParticleKey.ACTIVATE: 0>"]
         [2975, "<ParticleKey.DEACTIVATE: 1>"]
         [3811.97, "<ParticleKey.ACTIVATE: 0>"]
         [3854.605, "<ParticleKey.DEACTIVATE: 1>"]
        ]
       }]
     }]
   }
  ]
 }
 P Cattus Beam Impact Once {
  id=#27
  possessed_class=Emitter
  tracks [
   {
    type=MovieSceneSpawnTrack
    sections [{
      type=MovieSceneSpawnSection
      end_frame=3605
      start_frame=3490
      channels [None {type=MovieSceneScriptingBoolChannel, default=false, keys [[3490.1587,false],[3494.415,true],[3605.415,false]]}]
     }]
   }
   {
    type=MovieScene3DTransformTrack
    sections [{
      type=MovieScene3DTransformSection
      end_frame=3605
      start_frame=3490
      channels [
       Location.X {type=MovieSceneScriptingDoubleChannel, default=60258.4453}
       Location.Y {type=MovieSceneScriptingDoubleChannel, default=23351.4121}
       Location.Z {type=MovieSceneScriptingDoubleChannel, default=13733.2773}
       Rotation.X {type=MovieSceneScriptingDoubleChannel, default=0}
       Rotation.Y {type=MovieSceneScriptingDoubleChannel, default=0}
       Rotation.Z {type=MovieSceneScriptingDoubleChannel, default=0}
       Scale.X {type=MovieSceneScriptingDoubleChannel, default=1}
       Scale.Y {type=MovieSceneScriptingDoubleChannel, default=1}
       Scale.Z {type=MovieSceneScriptingDoubleChannel, default=1}
      ]
     }]
   }
   {
    type=MovieSceneParticleTrack
    display_name=Particle System
    sections [{type=MovieSceneParticleSection, channels [None {type=MovieSceneScriptingParticleChannel, const ["<ParticleKey.TRIGGER: 2>", 1]}]}]
   }
  ]
 }
 P Water Raid Entire Body {
  id=#28
  possessed_class=Emitter
  tracks [
   {
    type=MovieSceneSpawnTrack
    sections [{
      type=MovieSceneSpawnSection
      end_frame=474
      start_frame=29
      channels [None {type=MovieSceneScriptingBoolChannel, default=false, keys [[29.8825,false],[69.1838,true],[474,false]]}]
     }]
   }
   {
    type=MovieScene3DTransformTrack
    sections [{
      type=MovieScene3DTransformSection
      end_frame=474
      start_frame=29
      channels [
       Location.X {type=MovieSceneScriptingDoubleChannel, default=146379.9844, keys [[88.3825,150404.5625],[210,139424.7188],[355,138907.2188]]}
       Location.Y {type=MovieSceneScriptingDoubleChannel, default=-27116.2109, keys [[88.3825,-25380.1816],[210,-27973.9219],[355,-27116.3672]]}
       Location.Z {type=MovieSceneScriptingDoubleChannel, default=-4238.8687, keys [[88.3825,-4960.3232],[210,-5202.9932],[355,-4654.5576]]}
       Rotation.X {type=MovieSceneScriptingDoubleChannel, default=0}
       Rotation.Y {type=MovieSceneScriptingDoubleChannel, default=0}
       Rotation.Z {type=MovieSceneScriptingDoubleChannel, default=0}
       Scale.X {type=MovieSceneScriptingDoubleChannel, default=4, const [2,3]}
       Scale.Y {type=MovieSceneScriptingDoubleChannel, default=4, const [2,3]}
       Scale.Z {type=MovieSceneScriptingDoubleChannel, default=4, const [2,3]}
      ]
     }]
   }
   {
    type=MovieSceneParticleTrack
    display_name=Particle System
    sections [{
      type=MovieSceneParticleSection
      channels [None {type=MovieSceneScriptingParticleChannel, keys [[70.4038, "<ParticleKey.ACTIVATE: 0>"], [273, "<ParticleKey.DEACTIVATE: 1>"]]}]
     }]
   }
  ]
 }
 P Cattus Beam Charge Up {
  id=#29
  possessed_class=Emitter
  tracks [
   {
    type=MovieSceneSpawnTrack
    sections [{
      type=MovieSceneSpawnSection
      end_frame=3606
      start_frame=3425
      channels [None {type=MovieSceneScriptingBoolChannel, default=false, keys [[3425,false],[3436,true],[3606,false]]}]
     }]
   }
   {
    type=MovieScene3DTransformTrack
    sections [{
      type=MovieScene3DTransformSection
      end_frame=3620
      start_frame=3425
      channels [
       Location.X {type=MovieSceneScriptingDoubleChannel, default=0}
       Location.Y {type=MovieSceneScriptingDoubleChannel, default=0}
       Location.Z {type=MovieSceneScriptingDoubleChannel, default=0}
       Rotation.X {type=MovieSceneScriptingDoubleChannel, default=0}
       Rotation.Y {type=MovieSceneScriptingDoubleChannel, default=0}
       Rotation.Z {type=MovieSceneScriptingDoubleChannel, default=0}
       Scale.X {type=MovieSceneScriptingDoubleChannel, default=1}
       Scale.Y {type=MovieSceneScriptingDoubleChannel, default=1}
       Scale.Z {type=MovieSceneScriptingDoubleChannel, default=1}
      ]
     }]
   }
   {
    type=MovieSceneParticleTrack
    display_name=Particle System
    sections [{
      type=MovieSceneParticleSection
      channels [None {type=MovieSceneScriptingParticleChannel, keys [[3436, "<ParticleKey.ACTIVATE: 0>"], [3552, "<ParticleKey.DEACTIVATE: 1>"]]}]
     }]
   }
   {type=MovieScene3DAttachTrack, sections [{type=MovieScene3DAttachSection, end_frame=3620, start_frame=3436}]}
   {
    type=MovieSceneParticleParameterTrack
    display_name=Particle Parameter
    sections [{
      type=MovieSceneParameterSection
      end_frame=3554
      start_frame=3430
      channels [
       Opacity {type=MovieSceneScriptingFloatChannel, default=1, keys [[3445.575,0],[3455.075,1],[3501,1],[3552,0]]}
       SpawnRate {type=MovieSceneScriptingFloatChannel, default=1, keys [[3501,1],[3552,0]]}
      ]
     }]
   }
  ]
 }
 P Water Raid Hand R {
  id=#30
  possessed_class=Emitter
  tracks [
   {
    type=MovieSceneSpawnTrack
    sections [{
      type=MovieSceneSpawnSection
      end_frame=694
      start_frame=152
      channels [None {type=MovieSceneScriptingBoolChannel, default=false, keys [[152,false],[169,true],[694,false]]}]
     }]
   }
   {
    type=MovieScene3DTransformTrack
    sections [{
      type=MovieScene3DTransformSection
      end_frame=694
      start_frame=152
      channels [
       Location.X {
        type=MovieSceneScriptingDoubleChannel
        default=147642.0156
        keys [[169,147642.0156],[269,140401.9375],[337,129679.8672],[365,129104.5938],[392,137817.6719],[415,137924.0625],[457,128625.9297],[497,129099.7422],[540,136100.3125]]
       }
       Location.Y {
        type=MovieSceneScriptingDoubleChannel
        default=-32082.9512
        thinned_from=9
        keys [[169,-32082.9512],[269,-33224.9062],[337,-26608.9043],[365,-35880.6641],[392,-39598.8477],[457,-28371.7285],[497,-32725.4199],[540,-32354.4492]]
       }
       Location.Z {type=MovieSceneScriptingDoubleChannel, default=-4238.8687, thinned_from=9, keys [[169,-4238.8687],[540,-4238.8628]]}
       Rotation.X {type=MovieSceneScriptingDoubleChannel, default=0}
       Rotation.Y {type=MovieSceneScriptingDoubleChannel, default=0}
       Rotation.Z {type=MovieSceneScriptingDoubleChannel, default=0}
       Scale.X {type=MovieSceneScriptingDoubleChannel, default=1}
       Scale.Y {type=MovieSceneScriptingDoubleChannel, default=1}
       Scale.Z {type=MovieSceneScriptingDoubleChannel, default=1}
      ]
     }]
   }
   {
    type=MovieSceneParticleTrack
    display_name=Particle System
    sections [{
      type=MovieSceneParticleSection
      channels [None {type=MovieSceneScriptingParticleChannel, keys [[169, "<ParticleKey.ACTIVATE: 0>"], [560, "<ParticleKey.DEACTIVATE: 1>"]]}]
     }]
   }
  ]
 }
 P Cattus Debris Falling {
  id=#31
  possessed_class=Emitter
  tracks [
   {
    type=MovieSceneSpawnTrack
    sections [{
      type=MovieSceneSpawnSection
      end_frame=2878
      start_frame=-4
      channels [None {type=MovieSceneScriptingBoolChannel, default=false, keys [[-4,true],[2878,false]]}]
     }]
   }
   {
    type=MovieScene3DTransformTrack
    sections [{
      type=MovieScene3DTransformSection
      end_frame=2878
      channels [
       Location.X {type=MovieSceneScriptingDoubleChannel, default=40.6971, const [40.6971,1]}
       Location.Y {type=MovieSceneScriptingDoubleChannel, default=-9.3166, const [-9.3166,1]}
       Location.Z {type=MovieSceneScriptingDoubleChannel, default=9.0861, const [9.0861,1]}
       Rotation.X {type=MovieSceneScriptingDoubleChannel, default=9.8795, const [9.8795,1]}
       Rotation.Y {type=MovieSceneScriptingDoubleChannel, default=1.5828, const [1.5828,1]}
       Rotation.Z {type=MovieSceneScriptingDoubleChannel, default=28.4558, const [28.4558,1]}
       Scale.X {type=MovieSceneScriptingDoubleChannel, default=0.2, const [0.2,1]}
       Scale.Y {type=MovieSceneScriptingDoubleChannel, default=0.2, const [0.2,1]}
       Scale.Z {type=MovieSceneScriptingDoubleChannel, default=0.2, const [0.2,1]}
      ]
     }]
   }
   {type=MovieScene3DAttachTrack, sections [{type=MovieScene3DAttachSection, end_frame=7843, start_frame=-4}]}
   {
    type=MovieSceneParticleParameterTrack
    display_name=Particle Parameter
    sections [{
      type=MovieSceneParameterSection
      end_frame=4104
      start_frame=-4
      channels [
       Spawn_L {
        type=MovieSceneScriptingFloatChannel
        default=0
        thinned_from=19
        keys [[-4,0],[1090,0],[1094,1],[1104,1],[1109,0],[1462,0],[1467,0.7],[1477,0.7],[1484,0.062],[2149,0.062],[2156,0.48],[2162,0.4],[2177,0],[2239,0],[2243,0.48],[2247,0.4],[2255,0]]
       }
       Spawn_R {
        type=MovieSceneScriptingFloatChannel
        default=0.4
        thinned_from=18
        keys [[-4,0],[1090,0],[1094,1],[1104,1],[1109,0],[1462,0],[1467,0.7],[1477,0.7],[1484,0.062],[2149,0.062],[2156,0.48],[2162,0.4],[2177,0],[2239,0],[2243,0.48],[2247,0.4],[2255,0]]
       }
       Spawn_dust {
        type=MovieSceneScriptingFloatChannel
        default=3
        keys [[-4,0],[19,0.2],[40,0],[211,0.05],[230,0.3],[243,0],[288,0.05],[295,0.35],[300,0],[417,0.1],[422,0.2],[424,0.18],[432,0],[550,0.05],[552,0.2],[554,0.15],[558,0],[694,0],[886,0.25],[1090,0],[1094,3],[1104,7],[1109,0.2],[1462,0.15],[1467,10],[1477,4],[1484,0],[1535,0]]
       }
      ]
     }]
   }
   {
    type=MovieSceneParticleTrack
    display_name=Particle System
    sections [{
      type=MovieSceneParticleSection
      channels [None {type=MovieSceneScriptingParticleChannel, keys [[23, "<ParticleKey.ACTIVATE: 0>"], [2455, "<ParticleKey.DEACTIVATE: 1>"]]}]
     }]
   }
  ]
 }
 P Cattus Beam {
  id=#32
  possessed_class=Emitter
  tracks [
   {
    type=MovieSceneSpawnTrack
    sections [{
      type=MovieSceneSpawnSection
      end_frame=3611
      start_frame=3489
      channels [None {type=MovieSceneScriptingBoolChannel, default=false, keys [[3489.3975,false],[3490.1587,true],[3611.415,false]]}]
     }]
   }
   {
    type=MovieScene3DTransformTrack
    sections [{
      type=MovieScene3DTransformSection
      start_frame=3490
      channels [
       Location.X {
        type=MovieSceneScriptingDoubleChannel
        default=-3.1524
        thinned_from=63
        keys [[3491.12,-3.1524],[3494.9188,-3.1524],[3495.415,-3.1822],[3496.415,-3.2296],[3497.415,-3.1798],[3497.4987,-3.1524],[3498,-3.406],[3498.415,-3.1941],[3499,-3.2258],[3499.415,-3.2085],[3499.5038,-3.1555],[3499.7825,-3.1553],[3500,-3.1776],[3501.2537,-3.1524],[3501.415,-3.1783],[3501.7088,-3.1524],[3503.415,-3.1416],[3504.415,-3.1485],[3505,-3.1441],[3507,-3.1546],[3511,-3.1464],[3512,-3.1401],[3514.415,-3.1469],[3521.415,-3.1369],[3523,-3.0875],[3525.415,-3.1529],[3527,-3.1165],[3528,-3.1315],[3535,-3.193],[3541.415,-3.1808],[3548.415,-3.1822],[3549,-3.1961],[3559.415,-3.1671],[3564.415,-3.1689],[3594.84,-3.1524]]
       }
       Location.Y {
        type=MovieSceneScriptingDoubleChannel
        default=-2.1416
        thinned_from=63
        keys [[3491.12,-2.1416],[3494.9188,-2.1416],[3495.415,-2.1066],[3496.415,-2.0543],[3497.415,-2.1116],[3497.4987,-2.1416],[3498,-1.9119],[3498.415,-2.085],[3499,-2.0503],[3499.415,-2.0703],[3499.5688,-2.1416],[3499.7825,-2.1416],[3500,-2.1257],[3500.415,-2.1087],[3501,-2.1239],[3501.2537,-2.1416],[3501.415,-2.1988],[3501.7088,-2.1416],[3502,-2.1402],[3503.415,-2.1616],[3505,-2.1529],[3506.415,-2.1405],[3506.6737,-2.1419],[3507,-2.1283],[3507.415,-2.145],[3509,-2.1415],[3510.415,-2.1611],[3511,-2.1532],[3512,-2.1111],[3512.415,-2.1591],[3514,-2.1425],[3514.415,-2.1516],[3516,-2.1375],[3517.415,-2.1675],[3518.415,-2.1702],[3519,-2.1568],[3519.415,-2.1712],[3521.415,-2.1718],[3523,-2.1258],[3525.415,-2.1587],[3528,-2.1516],[3532,-2.1305],[3534.415,-2.1122],[3535,-2.0898],[3539,-2.1184],[3541.415,-2.119],[3545,-2.1132],[3557,-2.1236],[3559.415,-2.1396],[3563,-2.1279],[3594.84,-2.1416]]
       }
       Location.Z {
        type=MovieSceneScriptingDoubleChannel
        default=-0.0019
        thinned_from=63
        keys [[3491.12,-0.0019],[3494.9188,-0.0019],[3495.415,-0.0806],[3496.415,-0.156],[3497.415,-0.0192],[3497.4987,-0.0019],[3498,-0.1079],[3498.415,0.0154],[3499,0.054],[3499.415,0.0554],[3499.5038,-0.0028],[3499.7825,-0.0028],[3500,0.0085],[3500.415,0.0321],[3501,0.0068],[3501.2537,-0.0019],[3501.415,-0.0032],[3501.7088,-0.0019],[3502,-0.0041],[3503.415,-0.0348],[3504.415,-0.0025],[3505,0.0089],[3506.415,-0.0019],[3506.6737,-0.0019],[3507,-0.0023],[3507.415,-0.0012],[3509,0.0312],[3510.415,0.0078],[3511,0.008],[3512,0.0078],[3512.415,0.0118],[3514,0.0134],[3514.415,0.0141],[3516,0.0141],[3517.415,0.0159],[3518.415,0.016],[3519,0.015],[3519.415,0.0162],[3521.415,0.0165],[3523,0.0097],[3525.415,0.0204],[3527,0.0152],[3528,0.0154],[3532,0.0143],[3534.415,0.0131],[3535,0.0118],[3539,0.0134],[3540,0.0134],[3541.415,0.0115],[3545,0.0108],[3548.415,0.0105],[3549,0.0119],[3556.415,0.0076],[3557,0.0058],[3559.415,0.005],[3563,0.002],[3564.415,0.0053],[3583,-0.0045],[3594.84,-0.0019]]
       }
       Rotation.X {type=MovieSceneScriptingDoubleChannel, default=0}
       Rotation.Y {type=MovieSceneScriptingDoubleChannel, default=0}
       Rotation.Z {type=MovieSceneScriptingDoubleChannel, default=0}
       Scale.X {type=MovieSceneScriptingDoubleChannel, default=1, thinned_from=26, keys [[3494,1],[3594.84,1]]}
       Scale.Y {type=MovieSceneScriptingDoubleChannel, default=1, thinned_from=26, keys [[3494,1],[3594.84,1]]}
       Scale.Z {type=MovieSceneScriptingDoubleChannel, default=1, thinned_from=26, keys [[3494,1],[3594.84,1]]}
      ]
     }]
   }
   {
    type=MovieSceneParticleTrack
    display_name=Particle System
    sections [{
      type=MovieSceneParticleSection
      channels [None {type=MovieSceneScriptingParticleChannel, keys [[3490.1587, "<ParticleKey.ACTIVATE: 0>"], [3611.415, "<ParticleKey.DEACTIVATE: 1>"]]}]
     }]
   }
   {type=MovieScene3DAttachTrack, sections [{type=MovieScene3DAttachSection, end_frame=3611, start_frame=3489}]}
   {
    type=MovieSceneParticleParameterTrack
    display_name=Particle Parameter
    sections [{
      type=MovieSceneParameterSection
      end_frame=3611
      start_frame=3490
      channels [
       BeamErode2 {type=MovieSceneScriptingFloatChannel, default=1, keys [[3490.2912,0.2454],[3492.1187,0],[3563.2312,0.1986],[3588.4525,1]]}
       BeamOpacity {type=MovieSceneScriptingFloatChannel, default=0, keys [[3492.415,1],[3563.355,1],[3570.8,0]]}
       BeamTarget.X {
        type=MovieSceneScriptingFloatChannel
        default=403.2545
        keys [[3491.47,10],[3498.415,444.2094],[3502.415,450.5415],[3513.415,495.7411],[3524.415,519.144],[3533.415,558.4481],[3558.415,683.4349],[3568.415,694.1438]]
       }
       BeamTarget.Y {type=MovieSceneScriptingFloatChannel, thinned_from=4, keys [[3491.4362,0],[3516.415,0],[3524.415,-12.469]]}
       BeamTarget.Z {type=MovieSceneScriptingFloatChannel, thinned_from=4, keys [[3491.4362,0],[3516.415,0],[3524.415,-12.469]]}
       BeamSize.X {
        type=MovieSceneScriptingFloatChannel
        default=1.5
        keys [[3490.8963,1.3173],[3492.6025,1.6951],[3514,1.7151],[3549.565,1.6379],[3570.8475,1.9175]]
       }
       BeamSize.Y {type=MovieSceneScriptingFloatChannel, default=1}
       BeamSize.Z {type=MovieSceneScriptingFloatChannel, default=1}
       BeamBrightness.X {type=MovieSceneScriptingFloatChannel, default=0, keys [[3500.39,2],[3527.0675,1.7627],[3574.8712,1.2],[3589.5038,0]]}
       BeamBrightness.Y {type=MovieSceneScriptingFloatChannel, default=0, keys [[3500.39,1.5],[3527.0675,1.416],[3574.8712,1],[3589.5038,0]]}
       BeamBrightness.Z {type=MovieSceneScriptingFloatChannel, default=0.8, keys [[3500.39,0.7],[3527.0675,0.8],[3574.87,1],[3589.5038,0]]}
      ]
     }]
   }
  ]
 }
 P Doggus Step Heavy FX {
  id=#33
  possessed_class=Emitter
  tracks [
   {
    type=MovieSceneSpawnTrack
    sections [{
      type=MovieSceneSpawnSection
      start_frame=594
      channels [None {type=MovieSceneScriptingBoolChannel, default=false, keys [[594,true],[4091,false]]}]
     }]
   }
   {
    type=MovieScene3DTransformTrack
    sections [{
      type=MovieScene3DTransformSection
      end_frame=4091
      start_frame=594
      channels [
       Location.X {
        type=MovieSceneScriptingDoubleChannel
        default=116769.5469
        keys [[594,116769.5469],[1093,113811.3594],[1471,113811.3594],[2514.8025,77215.6094],[3018,67622.6172],[3498.3138,73617.7266],[3899.115,69456.0391]]
       }
       Location.Y {
        type=MovieSceneScriptingDoubleChannel
        default=-39554.2109
        thinned_from=7
        keys [[594,-39554.2109],[1471,-39492.8945],[2514.8025,-32651.041],[3018,-29556.1543],[3498.3138,-30344.707],[3899.115,-26178.8281]]
       }
       Location.Z {
        type=MovieSceneScriptingDoubleChannel
        default=2149.8784
        keys [[594,2149.8784],[1093,1274.394],[1471,1274.394],[2514.8025,-151.082],[3018,-814.6067],[3498.3138,-293.0889],[3899.115,-113.8792]]
       }
       Rotation.X {type=MovieSceneScriptingDoubleChannel, default=0, thinned_from=5, keys [[1471,0],[3018,0],[3498.3138,-0.171],[3899.115,-0.171]]}
       Rotation.Y {type=MovieSceneScriptingDoubleChannel, default=0, thinned_from=5, keys [[1471,0],[3018,0],[3498.3138,-2.5298],[3899.115,-2.5298]]}
       Rotation.Z {
        type=MovieSceneScriptingDoubleChannel
        default=37.6014
        keys [[1471,37.6014],[2514.8025,-101.7163],[3018,-33.4574],[3498.3138,-87.7482],[3899.115,-87.7482]]
       }
       Scale.X {type=MovieSceneScriptingDoubleChannel, default=1, thinned_from=6, keys [[594,1],[3498.3138,1],[3899.115,1.3]]}
       Scale.Y {type=MovieSceneScriptingDoubleChannel, default=1, thinned_from=6, keys [[594,1],[3498.3138,1],[3899.115,1.3]]}
       Scale.Z {type=MovieSceneScriptingDoubleChannel, default=2, thinned_from=7, keys [[594,1],[1093,2],[1471,1],[3498.3138,1],[3899.115,1.3]]}
      ]
     }]
   }
   {
    type=MovieSceneParticleTrack
    display_name=Particle System
    sections [{type=MovieSceneParticleSection, channels [None {type=MovieSceneScriptingParticleChannel, const ["<ParticleKey.TRIGGER: 2>", 7]}]}]
   }
  ]
 }
 P Cattus Beam Impact Loop {
  id=#34
  possessed_class=Emitter
  tracks [
   {
    type=MovieSceneSpawnTrack
    sections [{
      type=MovieSceneSpawnSection
      end_frame=3671
      start_frame=3496
      channels [None {type=MovieSceneScriptingBoolChannel, default=false, keys [[3496.9162,false],[3497.4987,true],[3671.415,false]]}]
     }]
   }
   {
    type=MovieScene3DTransformTrack
    sections [{
      type=MovieScene3DTransformSection
      end_frame=3671
      start_frame=3496
      channels [
       Location.X {type=MovieSceneScriptingDoubleChannel, default=0}
       Location.Y {type=MovieSceneScriptingDoubleChannel, default=0}
       Location.Z {type=MovieSceneScriptingDoubleChannel, default=0}
       Rotation.X {type=MovieSceneScriptingDoubleChannel, default=1.6772}
       Rotation.Y {type=MovieSceneScriptingDoubleChannel, default=23.8901}
       Rotation.Z {type=MovieSceneScriptingDoubleChannel, default=108.0883}
       Scale.X {type=MovieSceneScriptingDoubleChannel, default=1}
       Scale.Y {type=MovieSceneScriptingDoubleChannel, default=1}
       Scale.Z {type=MovieSceneScriptingDoubleChannel, default=1}
      ]
     }]
   }
   {
    type=MovieSceneParticleTrack
    display_name=Particle System
    sections [{
      type=MovieSceneParticleSection
      channels [None {type=MovieSceneScriptingParticleChannel, keys [[3497.4987, "<ParticleKey.ACTIVATE: 0>"], [3671.415, "<ParticleKey.DEACTIVATE: 1>"]]}]
     }]
   }
   {
    type=MovieSceneParticleParameterTrack
    display_name=Particle Parameter
    sections [{
      type=MovieSceneParameterSection
      end_frame=3671
      start_frame=3496
      channels [
       LightBrightness {type=MovieSceneScriptingFloatChannel, default=0, keys [[3497.4987,0],[3502.5712,1],[3551.415,1],[3601.415,0]]}
       SpawnRate {type=MovieSceneScriptingFloatChannel, default=2, keys [[3497.89,2],[3528.3125,2],[3604.415,0]]}
      ]
     }]
   }
   {type=MovieScene3DAttachTrack, sections [{type=MovieScene3DAttachSection, end_frame=3671, start_frame=3498}]}
  ]
 }
 P Cattus Punch Metal {
  id=#35
  possessed_class=Emitter
  tracks [
   {
    type=MovieSceneSpawnTrack
    sections [{
      type=MovieSceneSpawnSection
      end_frame=2370
      start_frame=2142
      channels [None {type=MovieSceneScriptingBoolChannel, default=false, keys [[2142,false],[2155,true],[2370,false]]}]
     }]
   }
   {
    type=MovieScene3DTransformTrack
    sections [{
      type=MovieScene3DTransformSection
      end_frame=2370
      start_frame=2142
      channels [
       Location.X {type=MovieSceneScriptingDoubleChannel, default=66099.7891, const [66099.7891,2]}
       Location.Y {type=MovieSceneScriptingDoubleChannel, default=-43117.0781, keys [[2155,-43117.0781],[2243,-42381.0742]]}
       Location.Z {type=MovieSceneScriptingDoubleChannel, default=-2180.895, const [-2180.895,2]}
       Rotation.X {type=MovieSceneScriptingDoubleChannel, default=0, const [0,2]}
       Rotation.Y {type=MovieSceneScriptingDoubleChannel, default=0, const [0,2]}
       Rotation.Z {type=MovieSceneScriptingDoubleChannel, default=0, const [0,2]}
       Scale.X {type=MovieSceneScriptingDoubleChannel, default=100, const [100,2]}
       Scale.Y {type=MovieSceneScriptingDoubleChannel, default=100, const [100,2]}
       Scale.Z {type=MovieSceneScriptingDoubleChannel, default=100, const [100,2]}
      ]
     }]
   }
   {
    type=MovieSceneParticleTrack
    display_name=Particle System
    sections [{type=MovieSceneParticleSection, channels [None {type=MovieSceneScriptingParticleChannel, const ["<ParticleKey.TRIGGER: 2>", 2]}]}]
   }
  ]
 }
]
display_rate {denominator=1, numerator=30}
tick_resolution {denominator=1, numerator=24000}
tracks []
