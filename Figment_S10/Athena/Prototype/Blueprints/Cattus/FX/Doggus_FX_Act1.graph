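# Level Sequence that times the Act 1 particle effects for Doggus in the Figment Cattus prototype.
# Spawn tracks bring in individual Emitters, whose particle tracks activate or trigger them, then usually deactivate them before despawn.
# Rocket impacts, step effects and dust mostly use fixed world transforms, while trails, smoke simmer and wing engines use attach tracks.
# Particle parameter tracks tint directional dust and fade air-trail intensity.
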LevelSequence /Figment_S10/Athena/Prototype/Blueprints/Cattus/FX/Doggus_FX_Act1
bindings [
 P_Cattus_Rocket_Impact_17 {
  id=#1
  possessed_class=Emitter
  tracks [
   {
    type=MovieSceneSpawnTrack
    sections [{
      type=MovieSceneSpawnSection
      end_frame=7242
      start_frame=6678
      channels [None {type=MovieSceneScriptingBoolChannel, default=false, keys [[6678,false],[6680,true],[7242,false]]}]
     }]
   }
   {
    type=MovieSceneParticleTrack
    display_name=Particle System
    sections [{
      type=MovieSceneParticleSection
      channels [None {type=MovieSceneScriptingParticleChannel, keys [[6680, "<ParticleKey.TRIGGER: 2>"], [7242, "<ParticleKey.DEACTIVATE: 1>"]]}]
     }]
   }
   {
    type=MovieScene3DTransformTrack
    sections [{
      type=MovieScene3DTransformSection
      end_frame=7242
      start_frame=6680
      channels [
       Location.X {type=MovieSceneScriptingDoubleChannel, default=67403.5469}
       Location.Y {type=MovieSceneScriptingDoubleChannel, default=-39188.3477}
       Location.Z {type=MovieSceneScriptingDoubleChannel, default=9018.4531}
       Rotation.X {type=MovieSceneScriptingDoubleChannel, default=1.0744}
       Rotation.Y {type=MovieSceneScriptingDoubleChannel, default=-13.3235}
       Rotation.Z {type=MovieSceneScriptingDoubleChannel, default=97.7609}
       Scale.X {type=MovieSceneScriptingDoubleChannel, default=1.4}
       Scale.Y {type=MovieSceneScriptingDoubleChannel, default=1.4}
       Scale.Z {type=MovieSceneScriptingDoubleChannel, default=1.4}
      ]
     }]
   }
  ]
 }
 P_Cattus_Rocket_Impact_11 {
  id=#2
  possessed_class=Emitter
  tracks [
   {
    type=MovieSceneSpawnTrack
    sections [{
      type=MovieSceneSpawnSection
      end_frame=6980
      start_frame=6494
      channels [None {type=MovieSceneScriptingBoolChannel, default=false, keys [[6494,false],[6496,true],[6978,true],[6980,false]]}]
     }]
   }
   {
    type=MovieSceneParticleTrack
    display_name=Particle System
    sections [{
      type=MovieSceneParticleSection
      channels [None {type=MovieSceneScriptingParticleChannel, keys [[6496, "<ParticleKey.TRIGGER: 2>"], [6978, "<ParticleKey.DEACTIVATE: 1>"]]}]
     }]
   }
  ]
 }
 P_Doggus_Trail {
  id=#3
  possessed_class=Emitter
  tracks [
   {
    type=MovieSceneSpawnTrack
    sections [{
      type=MovieSceneSpawnSection
      end_frame=6638
      start_frame=6152
      channels [None {type=MovieSceneScriptingBoolChannel, default=false, keys [[6152,false],[6158,true],[6638,false]]}]
     }]
   }
   {
    type=MovieScene3DTransformTrack
    sections [{
      type=MovieScene3DTransformSection
      end_frame=6638
      start_frame=6158
      channels [
       Location.X {type=MovieSceneScriptingDoubleChannel, default=0}
       Location.Y {type=MovieSceneScriptingDoubleChannel, default=0}
       Location.Z {type=MovieSceneScriptingDoubleChannel, default=0}
       Rotation.X {type=MovieSceneScriptingDoubleChannel, default=0}
       Rotation.Y {type=MovieSceneScriptingDoubleChannel, default=180}
       Rotation.Z {type=MovieSceneScriptingDoubleChannel, default=0}
       Scale.X {type=MovieSceneScriptingDoubleChannel, default=0.01}
       Scale.Y {type=MovieSceneScriptingDoubleChannel, default=0.01}
       Scale.Z {type=MovieSceneScriptingDoubleChannel, default=0.01}
      ]
     }]
   }
   {
    type=MovieSceneParticleTrack
    display_name=Particle System
    sections [{
      type=MovieSceneParticleSection
      channels [None {type=MovieSceneScriptingParticleChannel, keys [[6158, "<ParticleKey.ACTIVATE: 0>"], [6348, "<ParticleKey.DEACTIVATE: 1>"]]}]
     }]
   }
   {type=MovieScene3DAttachTrack, sections [{type=MovieScene3DAttachSection, end_frame=6638, start_frame=6158}]}
  ]
 }
 P_Doggus_Trail_03_R {
  id=#4
  possessed_class=Emitter
  tracks [
   {
    type=MovieSceneSpawnTrack
    sections [{
      type=MovieSceneSpawnSection
      end_frame=5816
      start_frame=5250
      channels [None {type=MovieSceneScriptingBoolChannel, default=false, keys [[5250,false],[5256,true],[5816,false]]}]
     }]
   }
   {
    type=MovieScene3DTransformTrack
    sections [{
      type=MovieScene3DTransformSection
      end_frame=5820
      start_frame=5250
      channels [
       Location.X {type=MovieSceneScriptingDoubleChannel, default=0}
       Location.Y {type=MovieSceneScriptingDoubleChannel, default=0}
       Location.Z {type=MovieSceneScriptingDoubleChannel, default=0}
       Rotation.X {type=MovieSceneScriptingDoubleChannel, default=0}
       Rotation.Y {type=MovieSceneScriptingDoubleChannel, default=180}
       Rotation.Z {type=MovieSceneScriptingDoubleChannel, default=0}
       Scale.X {type=MovieSceneScriptingDoubleChannel, default=0.01}
       Scale.Y {type=MovieSceneScriptingDoubleChannel, default=0.01}
       Scale.Z {type=MovieSceneScriptingDoubleChannel, default=0.01}
      ]
     }]
   }
   {
    type=MovieSceneParticleTrack
    display_name=Particle System
    sections [{
      type=MovieSceneParticleSection
      channels [None {type=MovieSceneScriptingParticleChannel, keys [[5256, "<ParticleKey.ACTIVATE: 0>"], [5407, "<ParticleKey.DEACTIVATE: 1>"]]}]
     }]
   }
   {type=MovieScene3DAttachTrack, sections [{type=MovieScene3DAttachSection, end_frame=5820, start_frame=5250}]}
  ]
 }
 P_Cattus_Rocket_Impact_4 {
  id=#5
  possessed_class=Emitter
  tracks [
   {
    type=MovieSceneSpawnTrack
    sections [{
      type=MovieSceneSpawnSection
      end_frame=6742
      start_frame=6342
      channels [None {type=MovieSceneScriptingBoolChannel, default=false, keys [[6344,false],[6346,true],[6734,false]]}]
     }]
   }
   {
    type=MovieSceneParticleTrack
    display_name=Particle System
    sections [{
      type=MovieSceneParticleSection
      channels [None {type=MovieSceneScriptingParticleChannel, keys [[6346, "<ParticleKey.TRIGGER: 2>"], [6732, "<ParticleKey.DEACTIVATE: 1>"]]}]
     }]
   }
  ]
 }
 P Directional Dust {
  id=#6
  possessed_class=Emitter
  tracks [
   {
    type=MovieSceneSpawnTrack
    sections [{
      type=MovieSceneSpawnSection
      end_frame=8393
      start_frame=8185
      channels [None {type=MovieSceneScriptingBoolChannel, default=false, keys [[8185,false],[8187,true],[8393,false]]}]
     }]
   }
   {
    type=MovieScene3DTransformTrack
    sections [{
      type=MovieScene3DTransformSection
      end_frame=8401
      start_frame=8185
      channels [
       Location.X {type=MovieSceneScriptingDoubleChannel, default=63319.2812}
       Location.Y {type=MovieSceneScriptingDoubleChannel, default=-35839.3516}
       Location.Z {type=MovieSceneScriptingDoubleChannel, default=-915.3789}
       Rotation.X {type=MovieSceneScriptingDoubleChannel, default=0}
       Rotation.Y {type=MovieSceneScriptingDoubleChannel, default=0}
       Rotation.Z {type=MovieSceneScriptingDoubleChannel, default=119.0241}
       Scale.X {type=MovieSceneScriptingDoubleChannel, default=1.1}
       Scale.Y {type=MovieSceneScriptingDoubleChannel, default=1.1}
       Scale.Z {type=MovieSceneScriptingDoubleChannel, default=1.1}
      ]
     }]
   }
   {
    type=MovieSceneParticleTrack
    display_name=Particle System
    sections [{
      type=MovieSceneParticleSection
      channels [None {type=MovieSceneScriptingParticleChannel, keys [[8187, "<ParticleKey.ACTIVATE: 0>"], [8258, "<ParticleKey.DEACTIVATE: 1>"]]}]
     }]
   }
   {
    type=MovieSceneParticleParameterTrack
    display_name=Particle Parameter
    sections [{
      type=MovieSceneParameterSection
      end_frame=8404
      start_frame=8186
      channels [
       ColorMultiplier.X {type=MovieSceneScriptingFloatChannel, default=0.1, const [0.1,1]}
       ColorMultiplier.Y {type=MovieSceneScriptingFloatChannel, default=0.085, const [0.085,1]}
       ColorMultiplier.Z {type=MovieSceneScriptingFloatChannel, default=0.0821, const [0.0821,1]}
      ]
     }]
   }
  ]
 }
 P_Cattus_Rocket_Impact_9 {
  id=#7
  possessed_class=Emitter
  tracks [
   {
    type=MovieSceneSpawnTrack
    sections [{
      type=MovieSceneSpawnSection
      end_frame=6898
      start_frame=6450
      channels [None {type=MovieSceneScriptingBoolChannel, default=false, keys [[6450,false],[6452,true],[6898,false]]}]
     }]
   }
   {
    type=MovieSceneParticleTrack
    display_name=Particle System
    sections [{
      type=MovieSceneParticleSection
      channels [None {type=MovieSceneScriptingParticleChannel, keys [[6452, "<ParticleKey.TRIGGER: 2>"], [6896, "<ParticleKey.DEACTIVATE: 1>"]]}]
     }]
   }
  ]
 }
 P_Doggus_Trail_02 {
  id=#8
  possessed_class=Emitter
  tracks [
   {
    type=MovieSceneSpawnTrack
    sections [{
      type=MovieSceneSpawnSection
      end_frame=5656
      start_frame=5102
      channels [None {type=MovieSceneScriptingBoolChannel, default=false, keys [[5102,false],[5112,true],[5656,false]]}]
     }]
   }
   {
    type=MovieScene3DTransformTrack
    sections [{
      type=MovieScene3DTransformSection
      end_frame=5656
      start_frame=5112
      channels [
       Location.X {type=MovieSceneScriptingDoubleChannel, default=0}
       Location.Y {type=MovieSceneScriptingDoubleChannel, default=0}
       Location.Z {type=MovieSceneScriptingDoubleChannel, default=0}
       Rotation.X {type=MovieSceneScriptingDoubleChannel, default=0}
       Rotation.Y {type=MovieSceneScriptingDoubleChannel, default=180}
       Rotation.Z {type=MovieSceneScriptingDoubleChannel, default=0}
       Scale.X {type=MovieSceneScriptingDoubleChannel, default=0.01}
       Scale.Y {type=MovieSceneScriptingDoubleChannel, default=0.01}
       Scale.Z {type=MovieSceneScriptingDoubleChannel, default=0.01}
      ]
     }]
   }
   {
    type=MovieSceneParticleTrack
    display_name=Particle System
    sections [{
      type=MovieSceneParticleSection
      channels [None {type=MovieSceneScriptingParticleChannel, keys [[5112, "<ParticleKey.ACTIVATE: 0>"], [5285, "<ParticleKey.DEACTIVATE: 1>"]]}]
     }]
   }
   {type=MovieScene3DAttachTrack, sections [{type=MovieScene3DAttachSection, end_frame=5656, start_frame=5112}]}
  ]
 }
 P_Doggus_Step_FX {
  id=#9
  possessed_class=Emitter
  tracks [
   {
    type=MovieSceneSpawnTrack
    sections [{
      type=MovieSceneSpawnSection
      end_frame=8066
      start_frame=7764
      channels [None {type=MovieSceneScriptingBoolChannel, default=false, keys [[7764,false],[7770,true],[8066,false]]}]
     }]
   }
   {
    type=MovieScene3DTransformTrack
    sections [{
      type=MovieScene3DTransformSection
      end_frame=8066
      start_frame=7764
      channels [
       Location.X {type=MovieSceneScriptingDoubleChannel, default=51122.4688, const [51122.4688,1]}
       Location.Y {type=MovieSceneScriptingDoubleChannel, default=43208.9102, const [43208.9102,1]}
       Location.Z {type=MovieSceneScriptingDoubleChannel, default=2136.8782, const [2136.8782,1]}
       Rotation.X {type=MovieSceneScriptingDoubleChannel, default=0, const [0,1]}
       Rotation.Y {type=MovieSceneScriptingDoubleChannel, default=-15.2364, const [-15.2364,1]}
       Rotation.Z {type=MovieSceneScriptingDoubleChannel, default=905, const [1265,1]}
       Scale.X {type=MovieSceneScriptingDoubleChannel, default=1}
       Scale.Y {type=MovieSceneScriptingDoubleChannel, default=1}
       Scale.Z {type=MovieSceneScriptingDoubleChannel, default=1}
      ]
     }]
   }
   {
    type=MovieSceneParticleTrack
    display_name=Particle System
    sections [{
      type=MovieSceneParticleSection
      channels [None {type=MovieSceneScriptingParticleChannel, keys [[7770, "<ParticleKey.ACTIVATE: 0>"], [7836, "<ParticleKey.DEACTIVATE: 1>"]]}]
     }]
   }
  ]
 }
 P_Cattus_Rocket_Impact_Large_05 {
  id=#10
  possessed_class=Emitter
  tracks [
   {
    type=MovieSceneSpawnTrack
    sections [{
      type=MovieSceneSpawnSection
      end_frame=5826
      start_frame=5336
      channels [None {type=MovieSceneScriptingBoolChannel, default=false, keys [[5336,false],[5338,true],[5826,false]]}]
     }]
   }
   {
    type=MovieScene3DTransformTrack
    sections [{
      type=MovieScene3DTransformSection
      end_frame=5826
      start_frame=5338
      channels [
       Location.X {type=MovieSceneScriptingDoubleChannel, default=61913.0156}
       Location.Y {type=MovieSceneScriptingDoubleChannel, default=-35991.1758}
       Location.Z {type=MovieSceneScriptingDoubleChannel, default=10490.2607}
       Rotation.X {type=MovieSceneScriptingDoubleChannel, default=0}
       Rotation.Y {type=MovieSceneScriptingDoubleChannel, default=0.5916}
       Rotation.Z {type=MovieSceneScriptingDoubleChannel, default=115}
       Scale.X {type=MovieSceneScriptingDoubleChannel, default=1.4}
       Scale.Y {type=MovieSceneScriptingDoubleChannel, default=1.4}
       Scale.Z {type=MovieSceneScriptingDoubleChannel, default=1.4}
      ]
     }]
   }
   {
    type=MovieSceneParticleTrack
    display_name=Particle System
    sections [{
      type=MovieSceneParticleSection
      channels [None {type=MovieSceneScriptingParticleChannel, keys [[5338, "<ParticleKey.ACTIVATE: 0>"], [5572, "<ParticleKey.DEACTIVATE: 1>"]]}]
     }]
   }
  ]
 }
 P_Cattus_Rocket_Impact_6 {
  id=#11
  possessed_class=Emitter
  tracks [
   {
    type=MovieSceneSpawnTrack
    sections [{
      type=MovieSceneSpawnSection
      end_frame=6792
      start_frame=6398
      channels [None {type=MovieSceneScriptingBoolChannel, default=false, keys [[6398,false],[6400,true],[6792,false]]}]
     }]
   }
   {
    type=MovieSceneParticleTrack
    display_name=Particle System
    sections [{
      type=MovieSceneParticleSection
      channels [None {type=MovieSceneScriptingParticleChannel, keys [[6400, "<ParticleKey.TRIGGER: 2>"], [6790, "<ParticleKey.DEACTIVATE: 1>"]]}]
     }]
   }
  ]
 }
 P_Doggus_Trail_03_R {
  id=#12
  possessed_class=Emitter
  tracks [
   {
    type=MovieSceneSpawnTrack
    sections [{
      type=MovieSceneSpawnSection
      end_frame=5928
      start_frame=5324
      channels [None {type=MovieSceneScriptingBoolChannel, default=false, keys [[5324,false],[5326,true],[5928,false]]}]
     }]
   }
   {
    type=MovieScene3DTransformTrack
    sections [{
      type=MovieScene3DTransformSection
      end_frame=5928
      start_frame=5324
      channels [
       Location.X {type=MovieSceneScriptingDoubleChannel, default=0}
       Location.Y {type=MovieSceneScriptingDoubleChannel, default=0}
       Location.Z {type=MovieSceneScriptingDoubleChannel, default=0}
       Rotation.X {type=MovieSceneScriptingDoubleChannel, default=0}
       Rotation.Y {type=MovieSceneScriptingDoubleChannel, default=180}
       Rotation.Z {type=MovieSceneScriptingDoubleChannel, default=0}
       Scale.X {type=MovieSceneScriptingDoubleChannel, default=0.01}
       Scale.Y {type=MovieSceneScriptingDoubleChannel, default=0.01}
       Scale.Z {type=MovieSceneScriptingDoubleChannel, default=0.01}
      ]
     }]
   }
   {
    type=MovieSceneParticleTrack
    display_name=Particle System
    sections [{
      type=MovieSceneParticleSection
      channels [None {type=MovieSceneScriptingParticleChannel, keys [[5326, "<ParticleKey.ACTIVATE: 0>"], [5439, "<ParticleKey.DEACTIVATE: 1>"]]}]
     }]
   }
   {type=MovieScene3DAttachTrack, sections [{type=MovieScene3DAttachSection, end_frame=5932, start_frame=5324}]}
  ]
 }
 P_Doggus_Step_FX {
  id=#13
  possessed_class=Emitter
  tracks [
   {
    type=MovieSceneSpawnTrack
    sections [{
      type=MovieSceneSpawnSection
      end_frame=8126
      start_frame=7840
      channels [None {type=MovieSceneScriptingBoolChannel, default=false, keys [[7840,false],[7846,true],[8126,false]]}]
     }]
   }
   {
    type=MovieScene3DTransformTrack
    sections [{
      type=MovieScene3DTransformSection
      end_frame=8126
      start_frame=7840
      channels [
       Location.X {type=MovieSceneScriptingDoubleChannel, default=47137.0469}
       Location.Y {type=MovieSceneScriptingDoubleChannel, default=46954.2656}
       Location.Z {type=MovieSceneScriptingDoubleChannel, default=2809.0735}
       Rotation.X {type=MovieSceneScriptingDoubleChannel, default=-1.7663}
       Rotation.Y {type=MovieSceneScriptingDoubleChannel, default=-15.9782}
       Rotation.Z {type=MovieSceneScriptingDoubleChannel, default=-75}
       Scale.X {type=MovieSceneScriptingDoubleChannel, default=1}
       Scale.Y {type=MovieSceneScriptingDoubleChannel, default=1}
       Scale.Z {type=MovieSceneScriptingDoubleChannel, default=1}
      ]
     }]
   }
   {
    type=MovieSceneParticleTrack
    display_name=Particle System
    sections [{
      type=MovieSceneParticleSection
      channels [None {type=MovieSceneScriptingParticleChannel, keys [[7846, "<ParticleKey.ACTIVATE: 0>"], [7912, "<ParticleKey.DEACTIVATE: 1>"]]}]
     }]
   }
  ]
 }
 P_Doggus_Trail {
  id=#14
  possessed_class=Emitter
  tracks [
   {
    type=MovieSceneSpawnTrack
    sections [{
      type=MovieSceneSpawnSection
      end_frame=6760
      start_frame=6274
      channels [None {type=MovieSceneScriptingBoolChannel, default=false, keys [[6274,false],[6280,true],[6760,false]]}]
     }]
   }
   {
    type=MovieScene3DTransformTrack
    sections [{
      type=MovieScene3DTransformSection
      end_frame=6760
      start_frame=6280
      channels [
       Location.X {type=MovieSceneScriptingDoubleChannel, default=0}
       Location.Y {type=MovieSceneScriptingDoubleChannel, default=0}
       Location.Z {type=MovieSceneScriptingDoubleChannel, default=0}
       Rotation.X {type=MovieSceneScriptingDoubleChannel, default=0}
       Rotation.Y {type=MovieSceneScriptingDoubleChannel, default=180}
       Rotation.Z {type=MovieSceneScriptingDoubleChannel, default=0}
       Scale.X {type=MovieSceneScriptingDoubleChannel, default=0.01}
       Scale.Y {type=MovieSceneScriptingDoubleChannel, default=0.01}
       Scale.Z {type=MovieSceneScriptingDoubleChannel, default=0.01}
      ]
     }]
   }
   {
    type=MovieSceneParticleTrack
    display_name=Particle System
    sections [{
      type=MovieSceneParticleSection
      channels [None {type=MovieSceneScriptingParticleChannel, keys [[6280, "<ParticleKey.ACTIVATE: 0>"], [6454, "<ParticleKey.DEACTIVATE: 1>"]]}]
     }]
   }
   {type=MovieScene3DAttachTrack, sections [{type=MovieScene3DAttachSection, end_frame=6758, start_frame=6280}]}
  ]
 }
 P_Doggus_Step_FX {
  id=#15
  possessed_class=Emitter
  tracks [
   {
    type=MovieSceneSpawnTrack
    sections [{
      type=MovieSceneSpawnSection
      end_frame=4718
      start_frame=4458
      channels [None {type=MovieSceneScriptingBoolChannel, default=false, keys [[4458,false],[4464,true],[4718,false]]}]
     }]
   }
   {
    type=MovieScene3DTransformTrack
    sections [{
      type=MovieScene3DTransformSection
      end_frame=4718
      start_frame=4464
      channels [
       Location.X {type=MovieSceneScriptingDoubleChannel, default=59498.8008}
       Location.Y {type=MovieSceneScriptingDoubleChannel, default=33878.9258}
       Location.Z {type=MovieSceneScriptingDoubleChannel, default=2495.7539}
       Rotation.X {type=MovieSceneScriptingDoubleChannel, default=0}
       Rotation.Y {type=MovieSceneScriptingDoubleChannel, default=0}
       Rotation.Z {type=MovieSceneScriptingDoubleChannel, default=2700}
       Scale.X {type=MovieSceneScriptingDoubleChannel, default=1}
       Scale.Y {type=MovieSceneScriptingDoubleChannel, default=1}
       Scale.Z {type=MovieSceneScriptingDoubleChannel, default=1}
      ]
     }]
   }
   {
    type=MovieSceneParticleTrack
    display_name=Particle System
    sections [{
      type=MovieSceneParticleSection
      channels [None {type=MovieSceneScriptingParticleChannel, keys [[4464, "<ParticleKey.ACTIVATE: 0>"], [4530, "<ParticleKey.DEACTIVATE: 1>"]]}]
     }]
   }
  ]
 }
 P_Cattus_Rocket_Impact_01 {
  id=#16
  possessed_class=Emitter
  tracks [
   {
    type=MovieSceneSpawnTrack
    sections [{
      type=MovieSceneSpawnSection
      end_frame=6679
      start_frame=6318
      channels [None {type=MovieSceneScriptingBoolChannel, default=false, keys [[6318,false],[6320,true],[6679.13,false]]}]
     }]
   }
   {
    type=MovieScene3DTransformTrack
    sections [{
      type=MovieScene3DTransformSection
      end_frame=6679
      start_frame=6318
      channels [
       Location.X {type=MovieSceneScriptingDoubleChannel, default=67934.7188, const [67934.7188,1]}
       Location.Y {type=MovieSceneScriptingDoubleChannel, default=-23702.3164, const [-23702.3164,1]}
       Location.Z {type=MovieSceneScriptingDoubleChannel, default=8706.749, const [8706.749,1]}
       Rotation.X {type=MovieSceneScriptingDoubleChannel, default=-16.566, const [-16.566,1]}
       Rotation.Y {type=MovieSceneScriptingDoubleChannel, default=-24.2017, const [-24.2017,1]}
       Rotation.Z {type=MovieSceneScriptingDoubleChannel, default=113.0048, const [113.0049,1]}
       Scale.X {type=MovieSceneScriptingDoubleChannel, default=1}
       Scale.Y {type=MovieSceneScriptingDoubleChannel, default=1}
       Scale.Z {type=MovieSceneScriptingDoubleChannel, default=1}
      ]
     }]
   }
   {
    type=MovieSceneParticleTrack
    display_name=Particle System
    sections [{
      type=MovieSceneParticleSection
      channels [None {type=MovieSceneScriptingParticleChannel, keys [[6320, "<ParticleKey.TRIGGER: 2>"], [6679.13, "<ParticleKey.DEACTIVATE: 1>"]]}]
     }]
   }
  ]
 }
 P_Doggus_Step_FX_Light {
  id=#17
  possessed_class=Emitter
  tracks [
   {
    type=MovieScene3DTransformTrack
    sections [{
      type=MovieScene3DTransformSection
      end_frame=7278
      start_frame=6782
      channels [
       Location.X {type=MovieSceneScriptingDoubleChannel, default=55130.7852}
       Location.Y {type=MovieSceneScriptingDoubleChannel, default=35747.6484}
       Location.Z {type=MovieSceneScriptingDoubleChannel, default=2690.0457}
       Rotation.X {type=MovieSceneScriptingDoubleChannel, default=-3.1118}
       Rotation.Y {type=MovieSceneScriptingDoubleChannel, default=-3.7103}
       Rotation.Z {type=MovieSceneScriptingDoubleChannel, default=8847280}
       Scale.X {type=MovieSceneScriptingDoubleChannel, default=1}
       Scale.Y {type=MovieSceneScriptingDoubleChannel, default=1}
       Scale.Z {type=MovieSceneScriptingDoubleChannel, default=1}
      ]
     }]
   }
   {
    type=MovieSceneParticleTrack
    display_name=Particle System
    sections [{
      type=MovieSceneParticleSection
      channels [None {type=MovieSceneScriptingParticleChannel, keys [[6784, "<ParticleKey.ACTIVATE: 0>"], [6858, "<ParticleKey.DEACTIVATE: 1>"]]}]
     }]
   }
   {
    type=MovieSceneSpawnTrack
    sections [{
      type=MovieSceneSpawnSection
      end_frame=7278
      start_frame=6782
      channels [None {type=MovieSceneScriptingBoolChannel, default=false, keys [[6782,false],[6784,true],[7278,false]]}]
     }]
   }
  ]
 }
 P_Doggus_Trail_03_R {
  id=#18
  possessed_class=Emitter
  tracks [
   {
    type=MovieSceneSpawnTrack
    sections [{
      type=MovieSceneSpawnSection
      end_frame=5880
      start_frame=5276
      channels [None {type=MovieSceneScriptingBoolChannel, default=false, keys [[5276,false],[5278,true],[5880,false]]}]
     }]
   }
   {
    type=MovieScene3DTransformTrack
    sections [{
      type=MovieScene3DTransformSection
      end_frame=5932
      start_frame=5276
      channels [
       Location.X {type=MovieSceneScriptingDoubleChannel, default=0}
       Location.Y {type=MovieSceneScriptingDoubleChannel, default=0}
       Location.Z {type=MovieSceneScriptingDoubleChannel, default=0}
       Rotation.X {type=MovieSceneScriptingDoubleChannel, default=0}
       Rotation.Y {type=MovieSceneScriptingDoubleChannel, default=180}
       Rotation.Z {type=MovieSceneScriptingDoubleChannel, default=0}
       Scale.X {type=MovieSceneScriptingDoubleChannel, default=0.01}
       Scale.Y {type=MovieSceneScriptingDoubleChannel, default=0.01}
       Scale.Z {type=MovieSceneScriptingDoubleChannel, default=0.01}
      ]
     }]
   }
   {
    type=MovieSceneParticleTrack
    display_name=Particle System
    sections [{
      type=MovieSceneParticleSection
      channels [None {type=MovieSceneScriptingParticleChannel, keys [[5278, "<ParticleKey.ACTIVATE: 0>"], [5413, "<ParticleKey.DEACTIVATE: 1>"]]}]
     }]
   }
   {type=MovieScene3DAttachTrack, sections [{type=MovieScene3DAttachSection, end_frame=5934, start_frame=5276}]}
  ]
 }
 P_Doggus_Smoke_Simmer {
  id=#19
  possessed_class=Emitter
  tracks [
   {
    type=MovieSceneSpawnTrack
    sections [{
      type=MovieSceneSpawnSection
      end_frame=7166
      start_frame=5228
      channels [None {type=MovieSceneScriptingBoolChannel, default=false, keys [[5448,false],[5454,true],[7166,false]]}]
     }]
   }
   {
    type=MovieScene3DTransformTrack
    sections [{
      type=MovieScene3DTransformSection
      end_frame=7166
      start_frame=5228
      channels [
       Location.X {type=MovieSceneScriptingDoubleChannel, default=-4.4809}
       Location.Y {type=MovieSceneScriptingDoubleChannel, default=-1.7689}
       Location.Z {type=MovieSceneScriptingDoubleChannel, default=-20.4015}
       Rotation.X {type=MovieSceneScriptingDoubleChannel, default=-25.6928}
       Rotation.Y {type=MovieSceneScriptingDoubleChannel, default=-59.5471}
       Rotation.Z {type=MovieSceneScriptingDoubleChannel, default=16.3658}
       Scale.X {type=MovieSceneScriptingDoubleChannel, default=0.01}
       Scale.Y {type=MovieSceneScriptingDoubleChannel, default=0.01}
       Scale.Z {type=MovieSceneScriptingDoubleChannel, default=0.01}
      ]
     }]
   }
   {
    type=MovieSceneParticleTrack
    display_name=Particle System
    sections [{
      type=MovieSceneParticleSection
      channels [None {type=MovieSceneScriptingParticleChannel, keys [[5456, "<ParticleKey.ACTIVATE: 0>"], [5854, "<ParticleKey.DEACTIVATE: 1>"]]}]
     }]
   }
   {type=MovieScene3DAttachTrack, sections [{type=MovieScene3DAttachSection, end_frame=7166, start_frame=5454}]}
  ]
 }
 P_Doggus_Smoke_Simmer {
  id=#20
  possessed_class=Emitter
  tracks [
   {
    type=MovieSceneSpawnTrack
    sections [{
      type=MovieSceneSpawnSection
      end_frame=8194
      start_frame=6432
      channels [None {type=MovieSceneScriptingBoolChannel, default=false, keys [[6432,false],[6438,true],[8194.005,false]]}]
     }]
   }
   {
    type=MovieSceneParticleTrack
    display_name=Particle System
    sections [{
      type=MovieSceneParticleSection
      channels [None {type=MovieSceneScriptingParticleChannel, keys [[6440, "<ParticleKey.ACTIVATE: 0>"], [7154, "<ParticleKey.DEACTIVATE: 1>"]]}]
     }]
   }
   {type=MovieScene3DAttachTrack, sections [{type=MovieScene3DAttachSection, end_frame=8378, start_frame=5928}]}
  ]
 }
 P Doggus Air Trails {
  id=#21
  possessed_class=Emitter
  tracks [
   {
    type=MovieSceneSpawnTrack
    sections [{
      type=MovieSceneSpawnSection
      end_frame=8723
      start_frame=8026
      channels [None {type=MovieSceneScriptingBoolChannel, default=false, keys [[8026,true],[8723,false]]}]
     }]
   }
   {
    type=MovieScene3DTransformTrack
    sections [{
      type=MovieScene3DTransformSection
      channels [
       Location.X {type=MovieSceneScriptingDoubleChannel, default=0}
       Location.Y {type=MovieSceneScriptingDoubleChannel, default=0}
       Location.Z {type=MovieSceneScriptingDoubleChannel, default=0}
       Rotation.X {type=MovieSceneScriptingDoubleChannel, default=37.694}
       Rotation.Y {type=MovieSceneScriptingDoubleChannel, default=-16.9885}
       Rotation.Z {type=MovieSceneScriptingDoubleChannel, default=98.8522}
       Scale.X {type=MovieSceneScriptingDoubleChannel, default=0.01}
       Scale.Y {type=MovieSceneScriptingDoubleChannel, default=0.01}
       Scale.Z {type=MovieSceneScriptingDoubleChannel, default=0.01}
      ]
     }]
   }
   {type=MovieScene3DAttachTrack, sections [{type=MovieScene3DAttachSection, end_frame=8726, start_frame=8026}]}
   {
    type=MovieSceneParticleTrack
    display_name=Particle System
    sections [{
      type=MovieSceneParticleSection
      channels [None {type=MovieSceneScriptingParticleChannel, keys [[8026, "<ParticleKey.ACTIVATE: 0>"], [8264, "<ParticleKey.DEACTIVATE: 1>"]]}]
     }]
   }
   {
    type=MovieSceneParticleParameterTrack
    display_name=Particle Parameter
    sections [{
      type=MovieSceneParameterSection
      end_frame=8322
      start_frame=8026
      channels [Intensity {type=MovieSceneScriptingFloatChannel, default=0.15, keys [[8026,0],[8076,0.15],[8322,0]]}]
     }]
   }
  ]
 }
 P_Doggus_Step_FX {
  id=#22
  possessed_class=Emitter
  tracks [
   {
    type=MovieSceneSpawnTrack
    sections [{
      type=MovieSceneSpawnSection
      end_frame=7902
      start_frame=7600
      channels [None {type=MovieSceneScriptingBoolChannel, default=false, keys [[7600,false],[7606,true],[7902,false]]}]
     }]
   }
   {
    type=MovieScene3DTransformTrack
    sections [{
      type=MovieScene3DTransformSection
      end_frame=7902
      start_frame=7600
      channels [
       Location.X {type=MovieSceneScriptingDoubleChannel, default=53024.8281}
       Location.Y {type=MovieSceneScriptingDoubleChannel, default=45263.6602}
       Location.Z {type=MovieSceneScriptingDoubleChannel, default=2696.4651}
       Rotation.X {type=MovieSceneScriptingDoubleChannel, default=-11.9875}
       Rotation.Y {type=MovieSceneScriptingDoubleChannel, default=1.1244}
       Rotation.Z {type=MovieSceneScriptingDoubleChannel, default=437932928}
       Scale.X {type=MovieSceneScriptingDoubleChannel, default=1}
       Scale.Y {type=MovieSceneScriptingDoubleChannel, default=1}
       Scale.Z {type=MovieSceneScriptingDoubleChannel, default=1}
      ]
     }]
   }
   {
    type=MovieSceneParticleTrack
    display_name=Particle System
    sections [{
      type=MovieSceneParticleSection
      channels [None {type=MovieSceneScriptingParticleChannel, keys [[7606, "<ParticleKey.ACTIVATE: 0>"], [7696, "<ParticleKey.DEACTIVATE: 1>"]]}]
     }]
   }
  ]
 }
 P_Doggus_Step_FX {
  id=#23
  possessed_class=Emitter
  tracks [
   {
    type=MovieSceneSpawnTrack
    sections [{
      type=MovieSceneSpawnSection
      end_frame=8438
      start_frame=8122
      channels [None {type=MovieSceneScriptingBoolChannel, default=false, keys [[8122,false],[8128,true],[8438,false]]}]
     }]
   }
   {
    type=MovieScene3DTransformTrack
    sections [{
      type=MovieScene3DTransformSection
      end_frame=8438
      start_frame=8122
      channels [
       Location.X {type=MovieSceneScriptingDoubleChannel, default=57634.3242}
       Location.Y {type=MovieSceneScriptingDoubleChannel, default=4777.2417}
       Location.Z {type=MovieSceneScriptingDoubleChannel, default=1245.6219}
       Rotation.X {type=MovieSceneScriptingDoubleChannel, default=0.9348}
       Rotation.Y {type=MovieSceneScriptingDoubleChannel, default=-5.6774}
       Rotation.Z {type=MovieSceneScriptingDoubleChannel, default=-70}
       Scale.X {type=MovieSceneScriptingDoubleChannel, default=1}
       Scale.Y {type=MovieSceneScriptingDoubleChannel, default=1}
       Scale.Z {type=MovieSceneScriptingDoubleChannel, default=1}
      ]
     }]
   }
   {
    type=MovieSceneParticleTrack
    display_name=Particle System
    sections [{type=MovieSceneParticleSection, channels [None {type=MovieSceneScriptingParticleChannel, const ["<ParticleKey.ACTIVATE: 0>", 1]}]}]
   }
  ]
 }
 P_Doggus_Trail_02 {
  id=#24
  possessed_class=Emitter
  tracks [
   {
    type=MovieSceneSpawnTrack
    sections [{
      type=MovieSceneSpawnSection
      end_frame=5656
      start_frame=5084
      channels [None {type=MovieSceneScriptingBoolChannel, default=false, keys [[5084,false],[5094,true],[5656,false]]}]
     }]
   }
   {
    type=MovieScene3DTransformTrack
    sections [{
      type=MovieScene3DTransformSection
      end_frame=5656
      start_frame=5094
      channels [
       Location.X {type=MovieSceneScriptingDoubleChannel, default=0}
       Location.Y {type=MovieSceneScriptingDoubleChannel, default=0}
       Location.Z {type=MovieSceneScriptingDoubleChannel, default=0}
       Rotation.X {type=MovieSceneScriptingDoubleChannel, default=0}
       Rotation.Y {type=MovieSceneScriptingDoubleChannel, default=180}
       Rotation.Z {type=MovieSceneScriptingDoubleChannel, default=0}
       Scale.X {type=MovieSceneScriptingDoubleChannel, default=0.01}
       Scale.Y {type=MovieSceneScriptingDoubleChannel, default=0.01}
       Scale.Z {type=MovieSceneScriptingDoubleChannel, default=0.01}
      ]
     }]
   }
   {
    type=MovieSceneParticleTrack
    display_name=Particle System
    sections [{
      type=MovieSceneParticleSection
      channels [None {type=MovieSceneScriptingParticleChannel, keys [[5094, "<ParticleKey.ACTIVATE: 0>"], [5291, "<ParticleKey.DEACTIVATE: 1>"]]}]
     }]
   }
   {type=MovieScene3DAttachTrack, sections [{type=MovieScene3DAttachSection, end_frame=5656, start_frame=5094}]}
  ]
 }
 P_Doggus_Smoke_Simmer {
  id=#25
  possessed_class=Emitter
  tracks [
   {
    type=MovieSceneSpawnTrack
    sections [{
      type=MovieSceneSpawnSection
      end_frame=7166
      start_frame=5228
      channels [None {type=MovieSceneScriptingBoolChannel, default=false, keys [[5448,false],[5454,true],[7166,false]]}]
     }]
   }
   {
    type=MovieScene3DTransformTrack
    sections [{
      type=MovieScene3DTransformSection
      end_frame=7166
      start_frame=5228
      channels [
       Location.X {type=MovieSceneScriptingDoubleChannel, default=-0.5676}
       Location.Y {type=MovieSceneScriptingDoubleChannel, default=-0.0001}
       Location.Z {type=MovieSceneScriptingDoubleChannel, default=10.6212}
       Rotation.X {type=MovieSceneScriptingDoubleChannel, default=-0.001}
       Rotation.Y {type=MovieSceneScriptingDoubleChannel, default=31.1664}
       Rotation.Z {type=MovieSceneScriptingDoubleChannel, default=-0.0004}
       Scale.X {type=MovieSceneScriptingDoubleChannel, default=0.01}
       Scale.Y {type=MovieSceneScriptingDoubleChannel, default=0.01}
       Scale.Z {type=MovieSceneScriptingDoubleChannel, default=0.01}
      ]
     }]
   }
   {
    type=MovieSceneParticleTrack
    display_name=Particle System
    sections [{
      type=MovieSceneParticleSection
      channels [None {type=MovieSceneScriptingParticleChannel, keys [[5456, "<ParticleKey.ACTIVATE: 0>"], [5898, "<ParticleKey.DEACTIVATE: 1>"]]}]
     }]
   }
   {type=MovieScene3DAttachTrack, sections [{type=MovieScene3DAttachSection, end_frame=7166, start_frame=5454}]}
  ]
 }
 P_Doggus_Trail {
  id=#26
  possessed_class=Emitter
  tracks [
   {
    type=MovieSceneSpawnTrack
    sections [{
      type=MovieSceneSpawnSection
      end_frame=6796
      start_frame=6308
      channels [None {type=MovieSceneScriptingBoolChannel, default=false, keys [[6308,false],[6314,true],[6796,false]]}]
     }]
   }
   {
    type=MovieScene3DTransformTrack
    sections [{
      type=MovieScene3DTransformSection
      end_frame=6796
      start_frame=6314
      channels [
       Location.X {type=MovieSceneScriptingDoubleChannel, default=0}
       Location.Y {type=MovieSceneScriptingDoubleChannel, default=0}
       Location.Z {type=MovieSceneScriptingDoubleChannel, default=0}
       Rotation.X {type=MovieSceneScriptingDoubleChannel, default=0}
       Rotation.Y {type=MovieSceneScriptingDoubleChannel, default=180}
       Rotation.Z {type=MovieSceneScriptingDoubleChannel, default=0}
       Scale.X {type=MovieSceneScriptingDoubleChannel, default=0.01}
       Scale.Y {type=MovieSceneScriptingDoubleChannel, default=0.01}
       Scale.Z {type=MovieSceneScriptingDoubleChannel, default=0.01}
      ]
     }]
   }
   {
    type=MovieSceneParticleTrack
    display_name=Particle System
    sections [{
      type=MovieSceneParticleSection
      channels [None {type=MovieSceneScriptingParticleChannel, keys [[6314, "<ParticleKey.ACTIVATE: 0>"], [6496, "<ParticleKey.DEACTIVATE: 1>"]]}]
     }]
   }
   {type=MovieScene3DAttachTrack, sections [{type=MovieScene3DAttachSection, end_frame=6794, start_frame=6314}]}
  ]
 }
 P_Doggus_Step_FX {
  id=#27
  possessed_class=Emitter
  tracks [
   {
    type=MovieSceneSpawnTrack
    sections [{
      type=MovieSceneSpawnSection
      end_frame=7624
      start_frame=7322
      channels [None {type=MovieSceneScriptingBoolChannel, default=false, keys [[7322,false],[7328,true],[7624,false]]}]
     }]
   }
   {
    type=MovieScene3DTransformTrack
    sections [{
      type=MovieScene3DTransformSection
      end_frame=7624
      start_frame=7322
      channels [
       Location.X {type=MovieSceneScriptingDoubleChannel, default=57344.5547}
       Location.Y {type=MovieSceneScriptingDoubleChannel, default=29599.1738}
       Location.Z {type=MovieSceneScriptingDoubleChannel, default=2760.7283}
       Rotation.X {type=MovieSceneScriptingDoubleChannel, default=0.7115}
       Rotation.Y {type=MovieSceneScriptingDoubleChannel, default=0.8848}
       Rotation.Z {type=MovieSceneScriptingDoubleChannel, default=54741768}
       Scale.X {type=MovieSceneScriptingDoubleChannel, default=1}
       Scale.Y {type=MovieSceneScriptingDoubleChannel, default=1}
       Scale.Z {type=MovieSceneScriptingDoubleChannel, default=1}
      ]
     }]
   }
   {
    type=MovieSceneParticleTrack
    display_name=Particle System
    sections [{
      type=MovieSceneParticleSection
      channels [None {type=MovieSceneScriptingParticleChannel, keys [[7328, "<ParticleKey.ACTIVATE: 0>"], [7398, "<ParticleKey.DEACTIVATE: 1>"]]}]
     }]
   }
  ]
 }
 P_Doggus_Trail_02_R {
  id=#28
  possessed_class=Emitter
  tracks [
   {
    type=MovieSceneSpawnTrack
    sections [{
      type=MovieSceneSpawnSection
      end_frame=5684
      start_frame=5160
      channels [None {type=MovieSceneScriptingBoolChannel, default=false, keys [[5160,false],[5162,true],[5684,false]]}]
     }]
   }
   {
    type=MovieScene3DTransformTrack
    sections [{
      type=MovieScene3DTransformSection
      end_frame=5684
      start_frame=5160
      channels [
       Location.X {type=MovieSceneScriptingDoubleChannel, default=0}
       Location.Y {type=MovieSceneScriptingDoubleChannel, default=0}
       Location.Z {type=MovieSceneScriptingDoubleChannel, default=0}
       Rotation.X {type=MovieSceneScriptingDoubleChannel, default=0}
       Rotation.Y {type=MovieSceneScriptingDoubleChannel, default=180}
       Rotation.Z {type=MovieSceneScriptingDoubleChannel, default=0}
       Scale.X {type=MovieSceneScriptingDoubleChannel, default=0.01}
       Scale.Y {type=MovieSceneScriptingDoubleChannel, default=0.01}
       Scale.Z {type=MovieSceneScriptingDoubleChannel, default=0.01}
      ]
     }]
   }
   {
    type=MovieSceneParticleTrack
    display_name=Particle System
    sections [{
      type=MovieSceneParticleSection
      channels [None {type=MovieSceneScriptingParticleChannel, keys [[5162, "<ParticleKey.ACTIVATE: 0>"], [5360, "<ParticleKey.DEACTIVATE: 1>"]]}]
     }]
   }
   {type=MovieScene3DAttachTrack, sections [{type=MovieScene3DAttachSection, end_frame=5684, start_frame=5160}]}
  ]
 }
 P Doggus Wing Engines R {
  id=#29
  possessed_class=Emitter
  tracks [
   {
    type=MovieScene3DTransformTrack
    sections [{
      type=MovieScene3DTransformSection
      end_frame=8462
      start_frame=8146
      channels [
       Location.X {type=MovieSceneScriptingDoubleChannel, default=0}
       Location.Y {type=MovieSceneScriptingDoubleChannel, default=0}
       Location.Z {type=MovieSceneScriptingDoubleChannel, default=0}
       Rotation.X {type=MovieSceneScriptingDoubleChannel, default=0}
       Rotation.Y {type=MovieSceneScriptingDoubleChannel, default=0}
       Rotation.Z {type=MovieSceneScriptingDoubleChannel, default=0}
       Scale.X {type=MovieSceneScriptingDoubleChannel, default=1}
       Scale.Y {type=MovieSceneScriptingDoubleChannel, default=1}
       Scale.Z {type=MovieSceneScriptingDoubleChannel, default=1}
      ]
     }]
   }
   {type=MovieScene3DAttachTrack, sections [{type=MovieScene3DAttachSection, end_frame=8462, start_frame=8146}]}
   {
    type=MovieSceneParticleTrack
    display_name=Particle System
    sections [{
      type=MovieSceneParticleSection
      channels [None {type=MovieSceneScriptingParticleChannel, keys [[8146, "<ParticleKey.ACTIVATE: 0>"], [8416, "<ParticleKey.DEACTIVATE: 1>"]]}]
     }]
   }
   {
    type=MovieSceneSpawnTrack
    sections [{
      type=MovieSceneSpawnSection
      end_frame=8442
      start_frame=8144
      channels [None {type=MovieSceneScriptingBoolChannel, default=false, keys [[8144,false],[8146,true],[8442,false]]}]
     }]
   }
   {
    type=MovieSceneParticleParameterTrack
    display_name=Particle Parameter
    sections [{
      type=MovieSceneParameterSection
      end_frame=8416
      start_frame=8146
      channels [
       WPO_Scale {type=MovieSceneScriptingFloatChannel, default=0, keys [[8146,-1],[8222,0]]}
       JetScale.X {type=MovieSceneScriptingFloatChannel, default=1, const [1,2]}
       JetScale.Y {type=MovieSceneScriptingFloatChannel, default=1, keys [[8146,1.5],[8222,1]]}
       JetScale.Z {type=MovieSceneScriptingFloatChannel, default=1, const [1,2]}
      ]
     }]
   }
  ]
 }
 P_Cattus_Rocket_Impact_01 {
  id=#30
  possessed_class=Emitter
  tracks [
   {
    type=MovieSceneSpawnTrack
    sections [{
      type=MovieSceneSpawnSection
      end_frame=5710
      start_frame=5226
      channels [None {type=MovieSceneScriptingBoolChannel, default=false, keys [[5226,false],[5228,true],[5710,false]]}]
     }]
   }
   {
    type=MovieScene3DTransformTrack
    sections [{
      type=MovieScene3DTransformSection
      end_frame=5710
      start_frame=5226
      channels [
       Location.X {type=MovieSceneScriptingDoubleChannel, default=71259.7266, keys [[5228,71259.7266],[5266,70815.4609]]}
       Location.Y {type=MovieSceneScriptingDoubleChannel, default=-26383.0645, const [-26383.0645,2]}
       Location.Z {type=MovieSceneScriptingDoubleChannel, default=12900.9199, const [12900.9199,2]}
       Rotation.X {type=MovieSceneScriptingDoubleChannel, default=-63.7221}
       Rotation.Y {type=MovieSceneScriptingDoubleChannel, default=32.3268}
       Rotation.Z {type=MovieSceneScriptingDoubleChannel, default=-41.8429}
       Scale.X {type=MovieSceneScriptingDoubleChannel, default=1.2}
       Scale.Y {type=MovieSceneScriptingDoubleChannel, default=1.2}
       Scale.Z {type=MovieSceneScriptingDoubleChannel, default=1.2}
      ]
     }]
   }
   {
    type=MovieSceneParticleTrack
    display_name=Particle System
    sections [{
      type=MovieSceneParticleSection
      channels [None {type=MovieSceneScriptingParticleChannel, keys [[5228, "<ParticleKey.ACTIVATE: 0>"], [5308, "<ParticleKey.DEACTIVATE: 1>"]]}]
     }]
   }
  ]
 }
 P_Doggus_Step_FX {
  id=#31
  possessed_class=Emitter
  tracks [
   {
    type=MovieSceneSpawnTrack
    sections [{
      type=MovieSceneSpawnSection
      end_frame=6202
      start_frame=5892
      channels [None {type=MovieSceneScriptingBoolChannel, default=false, keys [[5892,false],[5898,true],[6202,false]]}]
     }]
   }
   {
    type=MovieScene3DTransformTrack
    sections [{
      type=MovieScene3DTransformSection
      end_frame=6202
      start_frame=5892
      channels [
       Location.X {type=MovieSceneScriptingDoubleChannel, default=56093.8164}
       Location.Y {type=MovieSceneScriptingDoubleChannel, default=42020.7617}
       Location.Z {type=MovieSceneScriptingDoubleChannel, default=2696.4199}
       Rotation.X {type=MovieSceneScriptingDoubleChannel, default=-3.3787}
       Rotation.Y {type=MovieSceneScriptingDoubleChannel, default=-0.0635}
       Rotation.Z {type=MovieSceneScriptingDoubleChannel, default=63310.6094}
       Scale.X {type=MovieSceneScriptingDoubleChannel, default=1}
       Scale.Y {type=MovieSceneScriptingDoubleChannel, default=1}
       Scale.Z {type=MovieSceneScriptingDoubleChannel, default=1}
      ]
     }]
   }
   {
    type=MovieSceneParticleTrack
    display_name=Particle System
    sections [{
      type=MovieSceneParticleSection
      channels [None {type=MovieSceneScriptingParticleChannel, keys [[5898, "<ParticleKey.ACTIVATE: 0>"], [5964, "<ParticleKey.DEACTIVATE: 1>"]]}]
     }]
   }
  ]
 }
 P_Cattus_Rocket_Impact_10 {
  id=#32
  possessed_class=Emitter
  tracks [
   {
    type=MovieSceneSpawnTrack
    sections [{
      type=MovieSceneSpawnSection
      end_frame=6946
      start_frame=6464
      channels [None {type=MovieSceneScriptingBoolChannel, default=false, keys [[6464,false],[6468,true],[6946,false]]}]
     }]
   }
   {
    type=MovieSceneParticleTrack
    display_name=Particle System
    sections [{
      type=MovieSceneParticleSection
      channels [None {type=MovieSceneScriptingParticleChannel, keys [[6468, "<ParticleKey.TRIGGER: 2>"], [6944, "<ParticleKey.DEACTIVATE: 1>"]]}]
     }]
   }
  ]
 }
 P_Doggus_Trail {
  id=#33
  possessed_class=Emitter
  tracks [
   {
    type=MovieSceneSpawnTrack
    sections [{
      type=MovieSceneSpawnSection
      end_frame=6796
      start_frame=6310
      channels [None {type=MovieSceneScriptingBoolChannel, default=false, keys [[6310,false],[6312,true],[6796,false]]}]
     }]
   }
   {
    type=MovieScene3DTransformTrack
    sections [{
      type=MovieScene3DTransformSection
      end_frame=6796
      start_frame=6312
      channels [
       Location.X {type=MovieSceneScriptingDoubleChannel, default=0}
       Location.Y {type=MovieSceneScriptingDoubleChannel, default=0}
       Location.Z {type=MovieSceneScriptingDoubleChannel, default=0}
       Rotation.X {type=MovieSceneScriptingDoubleChannel, default=0}
       Rotation.Y {type=MovieSceneScriptingDoubleChannel, default=0}
       Rotation.Z {type=MovieSceneScriptingDoubleChannel, default=0}
       Scale.X {type=MovieSceneScriptingDoubleChannel, default=0.01}
       Scale.Y {type=MovieSceneScriptingDoubleChannel, default=0.01}
       Scale.Z {type=MovieSceneScriptingDoubleChannel, default=0.01}
      ]
     }]
   }
   {
    type=MovieSceneParticleTrack
    display_name=Particle System
    sections [{
      type=MovieSceneParticleSection
      channels [None {type=MovieSceneScriptingParticleChannel, keys [[6312, "<ParticleKey.ACTIVATE: 0>"], [6511, "<ParticleKey.DEACTIVATE: 1>"]]}]
     }]
   }
   {type=MovieScene3DAttachTrack, sections [{type=MovieScene3DAttachSection, end_frame=6794, start_frame=6312}]}
  ]
 }
 P_Cattus_Rocket_Impact_Large_09 {
  id=#34
  possessed_class=Emitter
  tracks [
   {
    type=MovieSceneSpawnTrack
    sections [{
      type=MovieSceneSpawnSection
      end_frame=5848
      start_frame=5405
      channels [None {type=MovieSceneScriptingBoolChannel, default=false, keys [[5405,false],[5407,true],[5848,false]]}]
     }]
   }
   {
    type=MovieScene3DTransformTrack
    sections [{
      type=MovieScene3DTransformSection
      end_frame=5848
      start_frame=5407
      channels [
       Location.X {type=MovieSceneScriptingDoubleChannel, default=69392.5469}
       Location.Y {type=MovieSceneScriptingDoubleChannel, default=-41491.0703}
       Location.Z {type=MovieSceneScriptingDoubleChannel, default=11828.9453}
       Rotation.X {type=MovieSceneScriptingDoubleChannel, default=13.5026}
       Rotation.Y {type=MovieSceneScriptingDoubleChannel, default=26.8527}
       Rotation.Z {type=MovieSceneScriptingDoubleChannel, default=104.4541}
       Scale.X {type=MovieSceneScriptingDoubleChannel, default=1}
       Scale.Y {type=MovieSceneScriptingDoubleChannel, default=1}
       Scale.Z {type=MovieSceneScriptingDoubleChannel, default=1}
      ]
     }]
   }
   {
    type=MovieSceneParticleTrack
    display_name=Particle System
    sections [{
      type=MovieSceneParticleSection
      channels [None {type=MovieSceneScriptingParticleChannel, keys [[5407, "<ParticleKey.ACTIVATE: 0>"], [5642, "<ParticleKey.DEACTIVATE: 1>"]]}]
     }]
   }
  ]
 }
 P_Doggus_Smoke_Simmer {
  id=#35
  possessed_class=Emitter
  tracks [
   {
    type=MovieSceneSpawnTrack
    sections [{
      type=MovieSceneSpawnSection
      end_frame=8378
      start_frame=6616
      channels [None {type=MovieSceneScriptingBoolChannel, default=false, keys [[6616,false],[6622,true],[8378.0025,false]]}]
     }]
   }
   {
    type=MovieSceneParticleTrack
    display_name=Particle System
    sections [{
      type=MovieSceneParticleSection
      channels [None {type=MovieSceneScriptingParticleChannel, keys [[6624, "<ParticleKey.ACTIVATE: 0>"], [7022, "<ParticleKey.DEACTIVATE: 1>"]]}]
     }]
   }
   {type=MovieScene3DAttachTrack, sections [{type=MovieScene3DAttachSection, end_frame=8378, start_frame=5928}]}
  ]
 }
 P_Cattus_Impact_Large {
  id=#36
  possessed_class=Emitter
  tracks [
   {
    type=MovieSceneSpawnTrack
    sections [{
      type=MovieSceneSpawnSection
      end_frame=8320
      start_frame=8186
      channels [None {type=MovieSceneScriptingBoolChannel, default=false, keys [[8186,false],[8190,true],[8320,false]]}]
     }]
   }
   {
    type=MovieScene3DTransformTrack
    sections [{
      type=MovieScene3DTransformSection
      end_frame=8320
      start_frame=8186
      channels [
       Location.X {type=MovieSceneScriptingDoubleChannel, default=62676.8086}
       Location.Y {type=MovieSceneScriptingDoubleChannel, default=-26310.2988}
       Location.Z {type=MovieSceneScriptingDoubleChannel, default=10988.9434}
       Rotation.X {type=MovieSceneScriptingDoubleChannel, default=6.8877}
       Rotation.Y {type=MovieSceneScriptingDoubleChannel, default=-7.5149}
       Rotation.Z {type=MovieSceneScriptingDoubleChannel, default=101.9083}
       Scale.X {type=MovieSceneScriptingDoubleChannel, default=1}
       Scale.Y {type=MovieSceneScriptingDoubleChannel, default=1}
       Scale.Z {type=MovieSceneScriptingDoubleChannel, default=1}
      ]
     }]
   }
   {
    type=MovieSceneParticleTrack
    display_name=Particle System
    sections [{
      type=MovieSceneParticleSection
      channels [None {type=MovieSceneScriptingParticleChannel, keys [[8190, "<ParticleKey.ACTIVATE: 0>"], [8252, "<ParticleKey.DEACTIVATE: 1>"]]}]
     }]
   }
  ]
 }
 P_Doggus_Trail {
  id=#37
  possessed_class=Emitter
  tracks [
   {
    type=MovieSceneSpawnTrack
    sections [{
      type=MovieSceneSpawnSection
      end_frame=6796
      start_frame=6360
      channels [None {type=MovieSceneScriptingBoolChannel, default=false, keys [[6360,false],[6366,true],[6796,false]]}]
     }]
   }
   {
    type=MovieScene3DTransformTrack
    sections [{
      type=MovieScene3DTransformSection
      end_frame=6796
      start_frame=6366
      channels [
       Location.X {type=MovieSceneScriptingDoubleChannel, default=0}
       Location.Y {type=MovieSceneScriptingDoubleChannel, default=0}
       Location.Z {type=MovieSceneScriptingDoubleChannel, default=0}
       Rotation.X {type=MovieSceneScriptingDoubleChannel, default=0}
       Rotation.Y {type=MovieSceneScriptingDoubleChannel, default=180}
       Rotation.Z {type=MovieSceneScriptingDoubleChannel, default=0}
       Scale.X {type=MovieSceneScriptingDoubleChannel, default=0.01}
       Scale.Y {type=MovieSceneScriptingDoubleChannel, default=0.01}
       Scale.Z {type=MovieSceneScriptingDoubleChannel, default=0.01}
      ]
     }]
   }
   {
    type=MovieSceneParticleTrack
    display_name=Particle System
    sections [{
      type=MovieSceneParticleSection
      channels [None {type=MovieSceneScriptingParticleChannel, keys [[6366, "<ParticleKey.ACTIVATE: 0>"], [6534, "<ParticleKey.DEACTIVATE: 1>"]]}]
     }]
   }
   {type=MovieScene3DAttachTrack, sections [{type=MovieScene3DAttachSection, end_frame=6794, start_frame=6366}]}
  ]
 }
 P Doggus Foot Jets_R {
  id=#38
  possessed_class=Emitter
  tracks [
   {
    type=MovieSceneSpawnTrack
    sections [{
      type=MovieSceneSpawnSection
      end_frame=8462
      start_frame=8216
      channels [None {type=MovieSceneScriptingBoolChannel, default=false, keys [[8216,true],[8462,false]]}]
     }]
   }
   {
    type=MovieScene3DTransformTrack
    sections [{
      type=MovieScene3DTransformSection
      end_frame=8462
      start_frame=8216
      channels [
       Location.X {type=MovieSceneScriptingDoubleChannel, default=0}
       Location.Y {type=MovieSceneScriptingDoubleChannel, default=0}
       Location.Z {type=MovieSceneScriptingDoubleChannel, default=0}
       Rotation.X {type=MovieSceneScriptingDoubleChannel, default=0}
       Rotation.Y {type=MovieSceneScriptingDoubleChannel, default=0}
       Rotation.Z {type=MovieSceneScriptingDoubleChannel, default=0}
       Scale.X {type=MovieSceneScriptingDoubleChannel, default=1}
       Scale.Y {type=MovieSceneScriptingDoubleChannel, default=1}
       Scale.Z {type=MovieSceneScriptingDoubleChannel, default=1}
      ]
     }]
   }
   {type=MovieScene3DAttachTrack, sections [{type=MovieScene3DAttachSection, end_frame=8462, start_frame=8216}]}
   {
    type=MovieSceneParticleTrack
    display_name=Particle System
    sections [{
      type=MovieSceneParticleSection
      channels [None {type=MovieSceneScriptingParticleChannel, keys [[8216, "<ParticleKey.ACTIVATE: 0>"], [8340, "<ParticleKey.DEACTIVATE: 1>"]]}]
     }]
   }
   {
    type=MovieSceneParticleParameterTrack
    display_name=Particle Parameter
    sections [{
      type=MovieSceneParameterSection
      end_frame=8462
      start_frame=8216
      channels [
       MeshVelocity.X {type=MovieSceneScriptingFloatChannel, default=-0.5, const [-0.5,1]}
       MeshVelocity.Y {type=MovieSceneScriptingFloatChannel, default=4, const [4,1]}
       MeshVelocity.Z {type=MovieSceneScriptingFloatChannel, default=-0.82, const [-0.82,1]}
       SizeScale.X {type=MovieSceneScriptingFloatChannel, default=2, keys [[8216,0],[8236,2],[8308,1]]}
       SizeScale.Y {type=MovieSceneScriptingFloatChannel, default=3, keys [[8216,3],[8308,1]]}
       SizeScale.Z {type=MovieSceneScriptingFloatChannel, default=3, keys [[8216,3],[8308,1]]}
       SpriteVelocity.X {type=MovieSceneScriptingFloatChannel, default=-78.3298, const [-78.3298,1]}
       SpriteVelocity.Y {type=MovieSceneScriptingFloatChannel, default=600, const [600,1]}
       SpriteVelocity.Z {type=MovieSceneScriptingFloatChannel, default=-124.9804, const [-124.9804,1]}
      ]
     }]
   }
  ]
 }
 P_Doggus_Step_FX_Light {
  id=#39
  possessed_class=Emitter
  tracks [
   {
    type=MovieScene3DTransformTrack
    sections [{
      type=MovieScene3DTransformSection
      end_frame=4848
      start_frame=4576
      channels [
       Location.X {type=MovieSceneScriptingDoubleChannel, default=58663.0586}
       Location.Y {type=MovieSceneScriptingDoubleChannel, default=34468.2812}
       Location.Z {type=MovieSceneScriptingDoubleChannel, default=2648.9136}
       Rotation.X {type=MovieSceneScriptingDoubleChannel, default=-4.0936}
       Rotation.Y {type=MovieSceneScriptingDoubleChannel, default=-3.6239}
       Rotation.Z {type=MovieSceneScriptingDoubleChannel, default=345436.625}
       Scale.X {type=MovieSceneScriptingDoubleChannel, default=1}
       Scale.Y {type=MovieSceneScriptingDoubleChannel, default=1}
       Scale.Z {type=MovieSceneScriptingDoubleChannel, default=1}
      ]
     }]
   }
   {
    type=MovieSceneParticleTrack
    display_name=Particle System
    sections [{
      type=MovieSceneParticleSection
      channels [None {type=MovieSceneScriptingParticleChannel, keys [[4576, "<ParticleKey.ACTIVATE: 0>"], [4642, "<ParticleKey.DEACTIVATE: 1>"]]}]
     }]
   }
   {
    type=MovieSceneSpawnTrack
    sections [{
      type=MovieSceneSpawnSection
      end_frame=4848
      start_frame=4574
      channels [None {type=MovieSceneScriptingBoolChannel, default=false, keys [[4574,false],[4576,true],[4848,false]]}]
     }]
   }
  ]
 }
 P Directional Dust {
  id=#40
  possessed_class=Emitter
  tracks [
   {
    type=MovieSceneSpawnTrack
    sections [{
      type=MovieSceneSpawnSection
      end_frame=8482
      start_frame=8006
      channels [None {type=MovieSceneScriptingBoolChannel, default=false, keys [[8006,false],[8008,true],[8482,false]]}]
     }]
   }
   {
    type=MovieScene3DTransformTrack
    sections [{
      type=MovieScene3DTransformSection
      end_frame=8646
      start_frame=8006
      channels [
       Location.X {type=MovieSceneScriptingDoubleChannel, default=49394.5195, keys [[8008,48309.9297],[8074,49408.0195]]}
       Location.Y {type=MovieSceneScriptingDoubleChannel, default=51882.9961, keys [[8008,51882.9961],[8074,50288.7383]]}
       Location.Z {type=MovieSceneScriptingDoubleChannel, default=2618.2417, const [2618.2417,2]}
       Rotation.X {type=MovieSceneScriptingDoubleChannel, default=0}
       Rotation.Y {type=MovieSceneScriptingDoubleChannel, default=0}
       Rotation.Z {type=MovieSceneScriptingDoubleChannel, default=124.5581}
       Scale.X {type=MovieSceneScriptingDoubleChannel, default=1, keys [[8008,0.75],[8074,1]]}
       Scale.Y {type=MovieSceneScriptingDoubleChannel, default=1, keys [[8008,0.75],[8074,1]]}
       Scale.Z {type=MovieSceneScriptingDoubleChannel, default=1, keys [[8008,0.75],[8074,1]]}
      ]
     }]
   }
   {
    type=MovieSceneParticleTrack
    display_name=Particle System
    sections [{
      type=MovieSceneParticleSection
      channels [None {type=MovieSceneScriptingParticleChannel, keys [[8008, "<ParticleKey.ACTIVATE: 0>"], [8074, "<ParticleKey.DEACTIVATE: 1>"]]}]
     }]
   }
  ]
 }
 P_Cattus_Rocket_Impact_01 {
  id=#41
  possessed_class=Emitter
  tracks [
   {
    type=MovieSceneSpawnTrack
    sections [{
      type=MovieSceneSpawnSection
      end_frame=5790
      start_frame=5278
      channels [None {type=MovieSceneScriptingBoolChannel, default=false, keys [[5278,false],[5280,true],[5790,false]]}]
     }]
   }
   {
    type=MovieScene3DTransformTrack
    sections [{
      type=MovieScene3DTransformSection
      end_frame=5790
      start_frame=5280
      channels [
       Location.X {type=MovieSceneScriptingDoubleChannel, default=64081.2969}
       Location.Y {type=MovieSceneScriptingDoubleChannel, default=-27563.3574}
       Location.Z {type=MovieSceneScriptingDoubleChannel, default=2927.4275}
       Rotation.X {type=MovieSceneScriptingDoubleChannel, default=-17.4086}
       Rotation.Y {type=MovieSceneScriptingDoubleChannel, default=3.2577}
       Rotation.Z {type=MovieSceneScriptingDoubleChannel, default=95.0326}
       Scale.X {type=MovieSceneScriptingDoubleChannel, default=1.2}
       Scale.Y {type=MovieSceneScriptingDoubleChannel, default=1.2}
       Scale.Z {type=MovieSceneScriptingDoubleChannel, default=1.2}
      ]
     }]
   }
   {
    type=MovieSceneParticleTrack
    display_name=Particle System
    sections [{
      type=MovieSceneParticleSection
      channels [None {type=MovieSceneScriptingParticleChannel, keys [[5280, "<ParticleKey.ACTIVATE: 0>"], [5360, "<ParticleKey.DEACTIVATE: 1>"]]}]
     }]
   }
  ]
 }
 P Doggus Air Trails {
  id=#42
  possessed_class=Emitter
  tracks [
   {
    type=MovieSceneSpawnTrack
    sections [{
      type=MovieSceneSpawnSection
      end_frame=8723
      start_frame=8022
      channels [None {type=MovieSceneScriptingBoolChannel, default=false, keys [[8022,false],[8026,true],[8723,false]]}]
     }]
   }
   {
    type=MovieScene3DTransformTrack
    sections [{
      type=MovieScene3DTransformSection
      channels [
       Location.X {type=MovieSceneScriptingDoubleChannel, default=0}
       Location.Y {type=MovieSceneScriptingDoubleChannel, default=0}
       Location.Z {type=MovieSceneScriptingDoubleChannel, default=0}
       Rotation.X {type=MovieSceneScriptingDoubleChannel, default=37.694}
       Rotation.Y {type=MovieSceneScriptingDoubleChannel, default=-16.9885}
       Rotation.Z {type=MovieSceneScriptingDoubleChannel, default=98.8522}
       Scale.X {type=MovieSceneScriptingDoubleChannel, default=0.01}
       Scale.Y {type=MovieSceneScriptingDoubleChannel, default=0.01}
       Scale.Z {type=MovieSceneScriptingDoubleChannel, default=0.01}
      ]
     }]
   }
   {type=MovieScene3DAttachTrack, sections [{type=MovieScene3DAttachSection, end_frame=8726, start_frame=8026}]}
   {
    type=MovieSceneParticleTrack
    display_name=Particle System
    sections [{
      type=MovieSceneParticleSection
      channels [None {type=MovieSceneScriptingParticleChannel, keys [[8026, "<ParticleKey.ACTIVATE: 0>"], [8264, "<ParticleKey.DEACTIVATE: 1>"]]}]
     }]
   }
   {
    type=MovieSceneParticleParameterTrack
    display_name=Particle Parameter
    sections [{
      type=MovieSceneParameterSection
      end_frame=8322
      start_frame=8026
      channels [Intensity {type=MovieSceneScriptingFloatChannel, default=0.15, keys [[8026,0],[8076,0.15],[8322,0]]}]
     }]
   }
  ]
 }
 P_Cattus_Rocket_Impact_Large_08 {
  id=#43
  possessed_class=Emitter
  tracks [
   {
    type=MovieSceneSpawnTrack
    sections [{
      type=MovieSceneSpawnSection
      end_frame=5848
      start_frame=5406
      channels [None {type=MovieSceneScriptingBoolChannel, default=false, keys [[5406,false],[5408,true],[5848,false]]}]
     }]
   }
   {
    type=MovieScene3DTransformTrack
    sections [{
      type=MovieScene3DTransformSection
      end_frame=5848
      start_frame=5408
      channels [
       Location.X {type=MovieSceneScriptingDoubleChannel, default=77489.8125}
       Location.Y {type=MovieSceneScriptingDoubleChannel, default=-39003.1016}
       Location.Z {type=MovieSceneScriptingDoubleChannel, default=3737.9636}
       Rotation.X {type=MovieSceneScriptingDoubleChannel, default=13.5026}
       Rotation.Y {type=MovieSceneScriptingDoubleChannel, default=26.8527}
       Rotation.Z {type=MovieSceneScriptingDoubleChannel, default=104.4541}
       Scale.X {type=MovieSceneScriptingDoubleChannel, default=1.2}
       Scale.Y {type=MovieSceneScriptingDoubleChannel, default=1.2}
       Scale.Z {type=MovieSceneScriptingDoubleChannel, default=1.2}
      ]
     }]
   }
   {
    type=MovieSceneParticleTrack
    display_name=Particle System
    sections [{
      type=MovieSceneParticleSection
      channels [None {type=MovieSceneScriptingParticleChannel, keys [[5408, "<ParticleKey.ACTIVATE: 0>"], [5642, "<ParticleKey.DEACTIVATE: 1>"]]}]
     }]
   }
  ]
 }
 P_Cattus_Rocket_Impact_Large_07 {
  id=#44
  possessed_class=Emitter
  tracks [
   {
    type=MovieSceneSpawnTrack
    sections [{
      type=MovieSceneSpawnSection
      end_frame=5848
      start_frame=5406
      channels [None {type=MovieSceneScriptingBoolChannel, default=false, keys [[5406,false],[5408,true],[5848,false]]}]
     }]
   }
   {
    type=MovieScene3DTransformTrack
    sections [{
      type=MovieScene3DTransformSection
      end_frame=5848
      start_frame=5408
      channels [
       Location.X {type=MovieSceneScriptingDoubleChannel, default=71742.9844}
       Location.Y {type=MovieSceneScriptingDoubleChannel, default=-44565.9844}
       Location.Z {type=MovieSceneScriptingDoubleChannel, default=12519.8984}
       Rotation.X {type=MovieSceneScriptingDoubleChannel, default=6.6881}
       Rotation.Y {type=MovieSceneScriptingDoubleChannel, default=29.1373}
       Rotation.Z {type=MovieSceneScriptingDoubleChannel, default=90}
       Scale.X {type=MovieSceneScriptingDoubleChannel, default=1.2}
       Scale.Y {type=MovieSceneScriptingDoubleChannel, default=1.2}
       Scale.Z {type=MovieSceneScriptingDoubleChannel, default=1.2}
      ]
     }]
   }
   {
    type=MovieSceneParticleTrack
    display_name=Particle System
    sections [{
      type=MovieSceneParticleSection
      channels [None {type=MovieSceneScriptingParticleChannel, keys [[5408, "<ParticleKey.ACTIVATE: 0>"], [5642, "<ParticleKey.DEACTIVATE: 1>"]]}]
     }]
   }
  ]
 }
 P_Doggus_Step_FX {
  id=#45
  possessed_class=Emitter
  tracks [
   {
    type=MovieSceneSpawnTrack
    sections [{
      type=MovieSceneSpawnSection
      end_frame=7522
      start_frame=7220
      channels [None {type=MovieSceneScriptingBoolChannel, default=false, keys [[7220,false],[7226,true],[7522,false]]}]
     }]
   }
   {
    type=MovieScene3DTransformTrack
    sections [{
      type=MovieScene3DTransformSection
      end_frame=7522
      start_frame=7220
      channels [
       Location.X {type=MovieSceneScriptingDoubleChannel, default=60905.5781}
       Location.Y {type=MovieSceneScriptingDoubleChannel, default=21165.6973}
       Location.Z {type=MovieSceneScriptingDoubleChannel, default=2588.0244}
       Rotation.X {type=MovieSceneScriptingDoubleChannel, default=0.6495}
       Rotation.Y {type=MovieSceneScriptingDoubleChannel, default=-0.9318}
       Rotation.Z {type=MovieSceneScriptingDoubleChannel, default=9123756}
       Scale.X {type=MovieSceneScriptingDoubleChannel, default=1}
       Scale.Y {type=MovieSceneScriptingDoubleChannel, default=1}
       Scale.Z {type=MovieSceneScriptingDoubleChannel, default=1}
      ]
     }]
   }
   {
    type=MovieSceneParticleTrack
    display_name=Particle System
    sections [{
      type=MovieSceneParticleSection
      channels [None {type=MovieSceneScriptingParticleChannel, keys [[7226, "<ParticleKey.ACTIVATE: 0>"], [7288, "<ParticleKey.DEACTIVATE: 1>"]]}]
     }]
   }
  ]
 }
 P_Doggus_Trail_02 {
  id=#46
  possessed_class=Emitter
  tracks [
   {
    type=MovieSceneSpawnTrack
    sections [{
      type=MovieSceneSpawnSection
      end_frame=5656
      start_frame=5098
      channels [None {type=MovieSceneScriptingBoolChannel, default=false, keys [[5098,false],[5106,true],[5656,false]]}]
     }]
   }
   {
    type=MovieScene3DTransformTrack
    sections [{
      type=MovieScene3DTransformSection
      end_frame=5656
      start_frame=5098
      channels [
       Location.X {type=MovieSceneScriptingDoubleChannel, default=0}
       Location.Y {type=MovieSceneScriptingDoubleChannel, default=0}
       Location.Z {type=MovieSceneScriptingDoubleChannel, default=0}
       Rotation.X {type=MovieSceneScriptingDoubleChannel, default=0}
       Rotation.Y {type=MovieSceneScriptingDoubleChannel, default=180}
       Rotation.Z {type=MovieSceneScriptingDoubleChannel, default=0}
       Scale.X {type=MovieSceneScriptingDoubleChannel, default=0.01}
       Scale.Y {type=MovieSceneScriptingDoubleChannel, default=0.01}
       Scale.Z {type=MovieSceneScriptingDoubleChannel, default=0.01}
      ]
     }]
   }
   {
    type=MovieSceneParticleTrack
    display_name=Particle System
    sections [{
      type=MovieSceneParticleSection
      channels [None {type=MovieSceneScriptingParticleChannel, keys [[5106, "<ParticleKey.ACTIVATE: 0>"], [5306, "<ParticleKey.DEACTIVATE: 1>"]]}]
     }]
   }
   {type=MovieScene3DAttachTrack, sections [{type=MovieScene3DAttachSection, end_frame=5656, start_frame=5106}]}
  ]
 }
 P_Cattus_Rocket_Impact_12 {
  id=#47
  possessed_class=Emitter
  tracks [
   {
    type=MovieSceneSpawnTrack
    sections [{
      type=MovieSceneSpawnSection
      end_frame=6964
      start_frame=6508
      channels [None {type=MovieSceneScriptingBoolChannel, default=false, keys [[6508,false],[6510,true],[6964,false]]}]
     }]
   }
   {
    type=MovieSceneParticleTrack
    display_name=Particle System
    sections [{
      type=MovieSceneParticleSection
      channels [None {type=MovieSceneScriptingParticleChannel, keys [[6510, "<ParticleKey.TRIGGER: 2>"], [6802, "<ParticleKey.DEACTIVATE: 1>"]]}]
     }]
   }
  ]
 }
 P_Castle_Destruction {
  id=#48
  possessed_class=Emitter
  tracks [
   {
    type=MovieSceneSpawnTrack
    sections [{
      type=MovieSceneSpawnSection
      end_frame=5656
      start_frame=5224
      channels [None {type=MovieSceneScriptingBoolChannel, default=false, keys [[5226,false],[5228,true],[5656,false]]}]
     }]
   }
   {
    type=MovieScene3DTransformTrack
    sections [{
      type=MovieScene3DTransformSection
      end_frame=5656
      start_frame=5226
      channels [
       Location.X {type=MovieSceneScriptingDoubleChannel, default=67499.6484, const [70721.3359,1]}
       Location.Y {type=MovieSceneScriptingDoubleChannel, default=-29754.1289, const [-25845.1992,1]}
       Location.Z {type=MovieSceneScriptingDoubleChannel, default=16453.834, const [15887.0508,1]}
       Rotation.X {type=MovieSceneScriptingDoubleChannel, default=11.6584, keys [[5228,16.4064],[5234,3.9393],[5246,-2.5338]]}
       Rotation.Y {type=MovieSceneScriptingDoubleChannel, default=0.6278, thinned_from=3, keys [[5228,-1.9466],[5246,-1.9465]]}
       Rotation.Z {type=MovieSceneScriptingDoubleChannel, default=56.2094, thinned_from=3, keys [[5228,17.816],[5246,17.8166]]}
       Scale.X {type=MovieSceneScriptingDoubleChannel, default=1, const [1,2]}
       Scale.Y {type=MovieSceneScriptingDoubleChannel, default=1, const [1,2]}
       Scale.Z {type=MovieSceneScriptingDoubleChannel, default=1, const [1,2]}
      ]
     }]
   }
   {
    type=MovieSceneParticleTrack
    display_name=Particle System
    sections [{
      type=MovieSceneParticleSection
      channels [None {type=MovieSceneScriptingParticleChannel, keys [[5228, "<ParticleKey.ACTIVATE: 0>"], [5256, "<ParticleKey.DEACTIVATE: 1>"]]}]
     }]
   }
  ]
 }
 P_Doggus_Trail {
  id=#49
  possessed_class=Emitter
  tracks [
   {
    type=MovieSceneSpawnTrack
    sections [{
      type=MovieSceneSpawnSection
      end_frame=6842
      start_frame=6354
      channels [None {type=MovieSceneScriptingBoolChannel, default=false, keys [[6354,false],[6360,true],[6842,false]]}]
     }]
   }
   {
    type=MovieScene3DTransformTrack
    sections [{
      type=MovieScene3DTransformSection
      end_frame=6842
      start_frame=6360
      channels [
       Location.X {type=MovieSceneScriptingDoubleChannel, default=0}
       Location.Y {type=MovieSceneScriptingDoubleChannel, default=0}
       Location.Z {type=MovieSceneScriptingDoubleChannel, default=0}
       Rotation.X {type=MovieSceneScriptingDoubleChannel, default=0}
       Rotation.Y {type=MovieSceneScriptingDoubleChannel, default=180}
       Rotation.Z {type=MovieSceneScriptingDoubleChannel, default=0}
       Scale.X {type=MovieSceneScriptingDoubleChannel, default=0.01}
       Scale.Y {type=MovieSceneScriptingDoubleChannel, default=0.01}
       Scale.Z {type=MovieSceneScriptingDoubleChannel, default=0.01}
      ]
     }]
   }
   {
    type=MovieSceneParticleTrack
    display_name=Particle System
    sections [{
      type=MovieSceneParticleSection
      channels [None {type=MovieSceneScriptingParticleChannel, keys [[6360, "<ParticleKey.ACTIVATE: 0>"], [6545, "<ParticleKey.DEACTIVATE: 1>"]]}]
     }]
   }
   {type=MovieScene3DAttachTrack, sections [{type=MovieScene3DAttachSection, end_frame=6842, start_frame=6360}]}
  ]
 }
 P_Doggus_Step_FX_Light {
  id=#50
  possessed_class=Emitter
  tracks [
   {
    type=MovieScene3DTransformTrack
    sections [{
      type=MovieScene3DTransformSection
      end_frame=4802
      start_frame=4530
      channels [
       Location.X {type=MovieSceneScriptingDoubleChannel, default=58664.6523}
       Location.Y {type=MovieSceneScriptingDoubleChannel, default=38633.9883}
       Location.Z {type=MovieSceneScriptingDoubleChannel, default=2648.9136}
       Rotation.X {type=MovieSceneScriptingDoubleChannel, default=0}
       Rotation.Y {type=MovieSceneScriptingDoubleChannel, default=0}
       Rotation.Z {type=MovieSceneScriptingDoubleChannel, default=11340}
       Scale.X {type=MovieSceneScriptingDoubleChannel, default=1}
       Scale.Y {type=MovieSceneScriptingDoubleChannel, default=1}
       Scale.Z {type=MovieSceneScriptingDoubleChannel, default=1}
      ]
     }]
   }
   {
    type=MovieSceneParticleTrack
    display_name=Particle System
    sections [{
      type=MovieSceneParticleSection
      channels [None {type=MovieSceneScriptingParticleChannel, keys [[4530, "<ParticleKey.ACTIVATE: 0>"], [4596, "<ParticleKey.DEACTIVATE: 1>"]]}]
     }]
   }
   {
    type=MovieSceneSpawnTrack
    sections [{
      type=MovieSceneSpawnSection
      end_frame=4802
      start_frame=4528
      channels [None {type=MovieSceneScriptingBoolChannel, default=false, keys [[4528,false],[4530,true],[4802,false]]}]
     }]
   }
  ]
 }
 P_Cattus_Rocket_Impact_16 {
  id=#51
  possessed_class=Emitter
  tracks [
   {
    type=MovieSceneSpawnTrack
    sections [{
      type=MovieSceneSpawnSection
      end_frame=7196
      start_frame=6628
      channels [None {type=MovieSceneScriptingBoolChannel, default=false, keys [[6628,false],[6630,true],[7196,false]]}]
     }]
   }
   {
    type=MovieSceneParticleTrack
    display_name=Particle System
    sections [{
      type=MovieSceneParticleSection
      channels [None {type=MovieSceneScriptingParticleChannel, keys [[6630, "<ParticleKey.TRIGGER: 2>"], [7166, "<ParticleKey.DEACTIVATE: 1>"]]}]
     }]
   }
   {
    type=MovieScene3DTransformTrack
    sections [{
      type=MovieScene3DTransformSection
      end_frame=7196
      start_frame=6630
      channels [
       Location.X {type=MovieSceneScriptingDoubleChannel, default=66461.5859}
       Location.Y {type=MovieSceneScriptingDoubleChannel, default=-43732.3008}
       Location.Z {type=MovieSceneScriptingDoubleChannel, default=8809.2539}
       Rotation.X {type=MovieSceneScriptingDoubleChannel, default=10.6103}
       Rotation.Y {type=MovieSceneScriptingDoubleChannel, default=30.6275}
       Rotation.Z {type=MovieSceneScriptingDoubleChannel, default=155.9483}
       Scale.X {type=MovieSceneScriptingDoubleChannel, default=1.2}
       Scale.Y {type=MovieSceneScriptingDoubleChannel, default=1.2}
       Scale.Z {type=MovieSceneScriptingDoubleChannel, default=1.2}
      ]
     }]
   }
  ]
 }
 P_Cattus_Rocket_Impact_3 {
  id=#52
  possessed_class=Emitter
  tracks [
   {
    type=MovieSceneSpawnTrack
    sections [{
      type=MovieSceneSpawnSection
      end_frame=6722
      start_frame=6346
      channels [None {type=MovieSceneScriptingBoolChannel, default=false, keys [[6346,false],[6348,true],[6722,false]]}]
     }]
   }
   {
    type=MovieSceneParticleTrack
    display_name=Particle System
    sections [{
      type=MovieSceneParticleSection
      channels [None {type=MovieSceneScriptingParticleChannel, keys [[6348, "<ParticleKey.TRIGGER: 2>"], [6722, "<ParticleKey.DEACTIVATE: 1>"]]}]
     }]
   }
  ]
 }
 P_Doggus_Trail {
  id=#53
  possessed_class=Emitter
  tracks [
   {
    type=MovieSceneSpawnTrack
    sections [{
      type=MovieSceneSpawnSection
      end_frame=5424
      start_frame=5058
      channels [None {type=MovieSceneScriptingBoolChannel, default=false, keys [[5058,false],[5064,true],[5424,false]]}]
     }]
   }
   {
    type=MovieScene3DTransformTrack
    sections [{
      type=MovieScene3DTransformSection
      end_frame=5424
      start_frame=5058
      channels [
       Location.X {type=MovieSceneScriptingDoubleChannel, default=0}
       Location.Y {type=MovieSceneScriptingDoubleChannel, default=0}
       Location.Z {type=MovieSceneScriptingDoubleChannel, default=0}
       Rotation.X {type=MovieSceneScriptingDoubleChannel, default=540}
       Rotation.Y {type=MovieSceneScriptingDoubleChannel, default=0}
       Rotation.Z {type=MovieSceneScriptingDoubleChannel, default=540}
       Scale.X {type=MovieSceneScriptingDoubleChannel, default=0.01}
       Scale.Y {type=MovieSceneScriptingDoubleChannel, default=0.01}
       Scale.Z {type=MovieSceneScriptingDoubleChannel, default=0.01}
      ]
     }]
   }
   {
    type=MovieSceneParticleTrack
    display_name=Particle System
    sections [{
      type=MovieSceneParticleSection
      channels [None {type=MovieSceneScriptingParticleChannel, keys [[5064, "<ParticleKey.ACTIVATE: 0>"], [5234, "<ParticleKey.DEACTIVATE: 1>"]]}]
     }]
   }
   {type=MovieScene3DAttachTrack, sections [{type=MovieScene3DAttachSection, end_frame=5424, start_frame=5058}]}
  ]
 }
 P Doggus Entry Smoke {
  id=#54
  possessed_class=Emitter
  tracks [
   {
    type=MovieSceneSpawnTrack
    sections [{
      type=MovieSceneSpawnSection
      end_frame=4322
      start_frame=2878
      channels [None {type=MovieSceneScriptingBoolChannel, default=false, keys [[2878,false],[2880,true],[4322,false]]}]
     }]
   }
   {
    type=MovieScene3DTransformTrack
    sections [{
      type=MovieScene3DTransformSection
      end_frame=4384
      start_frame=2880
      channels [
       Location.X {type=MovieSceneScriptingDoubleChannel, default=83428, const [83428,1]}
       Location.Y {type=MovieSceneScriptingDoubleChannel, default=37252, const [37252,1]}
       Location.Z {type=MovieSceneScriptingDoubleChannel, default=10844, const [10844,1]}
       Rotation.X {type=MovieSceneScriptingDoubleChannel, default=0, const [0,1]}
       Rotation.Y {type=MovieSceneScriptingDoubleChannel, default=0, const [0,1]}
       Rotation.Z {type=MovieSceneScriptingDoubleChannel, default=0, const [0,1]}
       Scale.X {type=MovieSceneScriptingDoubleChannel, default=1, const [1,1]}
       Scale.Y {type=MovieSceneScriptingDoubleChannel, default=1, const [1,1]}
       Scale.Z {type=MovieSceneScriptingDoubleChannel, default=1, const [1,1]}
      ]
     }]
   }
   {
    type=MovieSceneParticleTrack
    display_name=Particle System
    sections [{
      type=MovieSceneParticleSection
      channels [None {type=MovieSceneScriptingParticleChannel, keys [[2880, "<ParticleKey.ACTIVATE: 0>"], [3786, "<ParticleKey.DEACTIVATE: 1>"]]}]
     }]
   }
  ]
 }
 P_Doggus_Trail_02 {
  id=#55
  possessed_class=Emitter
  tracks [
   {
    type=MovieSceneSpawnTrack
    sections [{
      type=MovieSceneSpawnSection
      end_frame=5656
      start_frame=5072
      channels [None {type=MovieSceneScriptingBoolChannel, default=false, keys [[5072,false],[5078,true],[5656,false]]}]
     }]
   }
   {
    type=MovieScene3DTransformTrack
    sections [{
      type=MovieScene3DTransformSection
      end_frame=5656
      start_frame=5072
      channels [
       Location.X {type=MovieSceneScriptingDoubleChannel, default=0}
       Location.Y {type=MovieSceneScriptingDoubleChannel, default=0}
       Location.Z {type=MovieSceneScriptingDoubleChannel, default=0}
       Rotation.X {type=MovieSceneScriptingDoubleChannel, default=0}
       Rotation.Y {type=MovieSceneScriptingDoubleChannel, default=180}
       Rotation.Z {type=MovieSceneScriptingDoubleChannel, default=0}
       Scale.X {type=MovieSceneScriptingDoubleChannel, default=0.01}
       Scale.Y {type=MovieSceneScriptingDoubleChannel, default=0.01}
       Scale.Z {type=MovieSceneScriptingDoubleChannel, default=0.01}
      ]
     }]
   }
   {
    type=MovieSceneParticleTrack
    display_name=Particle System
    sections [{
      type=MovieSceneParticleSection
      channels [None {type=MovieSceneScriptingParticleChannel, keys [[5078, "<ParticleKey.ACTIVATE: 0>"], [5285, "<ParticleKey.DEACTIVATE: 1>"]]}]
     }]
   }
   {type=MovieScene3DAttachTrack, sections [{type=MovieScene3DAttachSection, end_frame=5656, start_frame=5072}]}
  ]
 }
 P_Cattus_Rocket_Impact_Large_08 {
  id=#56
  possessed_class=Emitter
  tracks [
   {
    type=MovieSceneSpawnTrack
    sections [{
      type=MovieSceneSpawnSection
      end_frame=5885
      start_frame=5406
      channels [None {type=MovieSceneScriptingBoolChannel, default=false, keys [[5436,false],[5438,true],[5878,false]]}]
     }]
   }
   {
    type=MovieScene3DTransformTrack
    sections [{
      type=MovieScene3DTransformSection
      end_frame=5878
      start_frame=5438
      channels [
       Location.X {type=MovieSceneScriptingDoubleChannel, default=74004.9531}
       Location.Y {type=MovieSceneScriptingDoubleChannel, default=-36963.3672}
       Location.Z {type=MovieSceneScriptingDoubleChannel, default=4065.4624}
       Rotation.X {type=MovieSceneScriptingDoubleChannel, default=13.5026}
       Rotation.Y {type=MovieSceneScriptingDoubleChannel, default=26.8527}
       Rotation.Z {type=MovieSceneScriptingDoubleChannel, default=104.4541}
       Scale.X {type=MovieSceneScriptingDoubleChannel, default=1.2}
       Scale.Y {type=MovieSceneScriptingDoubleChannel, default=1.2}
       Scale.Z {type=MovieSceneScriptingDoubleChannel, default=1.2}
      ]
     }]
   }
   {
    type=MovieSceneParticleTrack
    display_name=Particle System
    sections [{
      type=MovieSceneParticleSection
      channels [None {type=MovieSceneScriptingParticleChannel, keys [[5438, "<ParticleKey.ACTIVATE: 0>"], [5672, "<ParticleKey.DEACTIVATE: 1>"]]}]
     }]
   }
  ]
 }
 P_Doggus_Trail {
  id=#57
  possessed_class=Emitter
  tracks [
   {
    type=MovieSceneSpawnTrack
    sections [{
      type=MovieSceneSpawnSection
      end_frame=6682
      start_frame=6196
      channels [None {type=MovieSceneScriptingBoolChannel, default=false, keys [[6196,false],[6202,true],[6682,false]]}]
     }]
   }
   {
    type=MovieScene3DTransformTrack
    sections [{
      type=MovieScene3DTransformSection
      end_frame=6682
      start_frame=6202
      channels [
       Location.X {type=MovieSceneScriptingDoubleChannel, default=0}
       Location.Y {type=MovieSceneScriptingDoubleChannel, default=0}
       Location.Z {type=MovieSceneScriptingDoubleChannel, default=0}
       Rotation.X {type=MovieSceneScriptingDoubleChannel, default=0}
       Rotation.Y {type=MovieSceneScriptingDoubleChannel, default=180}
       Rotation.Z {type=MovieSceneScriptingDoubleChannel, default=0}
       Scale.X {type=MovieSceneScriptingDoubleChannel, default=0.01}
       Scale.Y {type=MovieSceneScriptingDoubleChannel, default=0.01}
       Scale.Z {type=MovieSceneScriptingDoubleChannel, default=0.01}
      ]
     }]
   }
   {
    type=MovieSceneParticleTrack
    display_name=Particle System
    sections [{
      type=MovieSceneParticleSection
      channels [None {type=MovieSceneScriptingParticleChannel, keys [[6202, "<ParticleKey.ACTIVATE: 0>"], [6402, "<ParticleKey.DEACTIVATE: 1>"]]}]
     }]
   }
   {type=MovieScene3DAttachTrack, sections [{type=MovieScene3DAttachSection, end_frame=6682, start_frame=6202}]}
  ]
 }
 P_Doggus_Trail {
  id=#58
  possessed_class=Emitter
  tracks [
   {
    type=MovieSceneSpawnTrack
    sections [{
      type=MovieSceneSpawnSection
      end_frame=6936
      start_frame=6452
      channels [None {type=MovieSceneScriptingBoolChannel, default=false, keys [[6452,false],[6458,true],[6936,false]]}]
     }]
   }
   {
    type=MovieScene3DTransformTrack
    sections [{
      type=MovieScene3DTransformSection
      end_frame=6936
      start_frame=6458
      channels [
       Location.X {type=MovieSceneScriptingDoubleChannel, default=0}
       Location.Y {type=MovieSceneScriptingDoubleChannel, default=0}
       Location.Z {type=MovieSceneScriptingDoubleChannel, default=0}
       Rotation.X {type=MovieSceneScriptingDoubleChannel, default=-180}
       Rotation.Y {type=MovieSceneScriptingDoubleChannel, default=360}
       Rotation.Z {type=MovieSceneScriptingDoubleChannel, default=-180}
       Scale.X {type=MovieSceneScriptingDoubleChannel, default=0.01}
       Scale.Y {type=MovieSceneScriptingDoubleChannel, default=0.01}
       Scale.Z {type=MovieSceneScriptingDoubleChannel, default=0.01}
      ]
     }]
   }
   {
    type=MovieSceneParticleTrack
    display_name=Particle System
    sections [{
      type=MovieSceneParticleSection
      channels [None {type=MovieSceneScriptingParticleChannel, keys [[6458, "<ParticleKey.ACTIVATE: 0>"], [6631, "<ParticleKey.DEACTIVATE: 1>"]]}]
     }]
   }
   {type=MovieScene3DAttachTrack, sections [{type=MovieScene3DAttachSection, end_frame=6936, start_frame=6458}]}
  ]
 }
 P_Doggus_Smoke_Simmer {
  id=#59
  possessed_class=Emitter
  tracks [
   {
    type=MovieSceneSpawnTrack
    sections [{
      type=MovieSceneSpawnSection
      end_frame=7166
      start_frame=5228
      channels [None {type=MovieSceneScriptingBoolChannel, default=false, keys [[5448,false],[5454,true],[7166,false]]}]
     }]
   }
   {
    type=MovieScene3DTransformTrack
    sections [{
      type=MovieScene3DTransformSection
      end_frame=7166
      start_frame=5454
      channels [
       Location.X {type=MovieSceneScriptingDoubleChannel, default=14.4344}
       Location.Y {type=MovieSceneScriptingDoubleChannel, default=-12.0072}
       Location.Z {type=MovieSceneScriptingDoubleChannel, default=5.6072}
       Rotation.X {type=MovieSceneScriptingDoubleChannel, default=-28.7759}
       Rotation.Y {type=MovieSceneScriptingDoubleChannel, default=0.0003}
       Rotation.Z {type=MovieSceneScriptingDoubleChannel, default=0.0004}
       Scale.X {type=MovieSceneScriptingDoubleChannel, default=0.01}
       Scale.Y {type=MovieSceneScriptingDoubleChannel, default=0.01}
       Scale.Z {type=MovieSceneScriptingDoubleChannel, default=0.01}
      ]
     }]
   }
   {
    type=MovieSceneParticleTrack
    display_name=Particle System
    sections [{
      type=MovieSceneParticleSection
      channels [None {type=MovieSceneScriptingParticleChannel, keys [[5456, "<ParticleKey.ACTIVATE: 0>"], [5854, "<ParticleKey.DEACTIVATE: 1>"]]}]
     }]
   }
   {type=MovieScene3DAttachTrack, sections [{type=MovieScene3DAttachSection, end_frame=7166, start_frame=5228}]}
  ]
 }
 P Impact Shockwave Dust Radial {
  id=#60
  possessed_class=Emitter
  tracks [
   {
    type=MovieSceneSpawnTrack
    sections [{
      type=MovieSceneSpawnSection
      end_frame=7300
      start_frame=5288
      channels [None {type=MovieSceneScriptingBoolChannel, default=false, keys [[5288,false],[5290,true],[5798,false],[6346,true],[7300,false]]}]
     }]
   }
   {
    type=MovieScene3DTransformTrack
    sections [{
      type=MovieScene3DTransformSection
      end_frame=7740
      start_frame=5288
      channels [
       Location.X {
        type=MovieSceneScriptingDoubleChannel
        default=70229.1016
        keys [[5290,64350.1094],[5356,71063.6094],[6346,71413.4297],[6566,71402.7344],[6680,70590.7578]]
       }
       Location.Y {
        type=MovieSceneScriptingDoubleChannel
        default=-39522.7031
        keys [[5290,-26501.5957],[5356,-42247.4844],[6346,-28388.1387],[6566,-43547.5664],[6680,-41737.4062]]
       }
       Location.Z {
        type=MovieSceneScriptingDoubleChannel
        default=-32.4805
        keys [[5290,462.4092],[5356,133.8158],[6346,133.8158],[6566,-396.7379],[6680,-511.0151]]
       }
       Rotation.X {type=MovieSceneScriptingDoubleChannel, default=0}
       Rotation.Y {type=MovieSceneScriptingDoubleChannel, default=0}
       Rotation.Z {type=MovieSceneScriptingDoubleChannel, default=0}
       Scale.X {type=MovieSceneScriptingDoubleChannel, default=1, const [1,5]}
       Scale.Y {type=MovieSceneScriptingDoubleChannel, default=1, const [1,5]}
       Scale.Z {type=MovieSceneScriptingDoubleChannel, default=1, const [1,5]}
      ]
     }]
   }
   {
    type=MovieSceneParticleTrack
    display_name=Particle System
    sections [{type=MovieSceneParticleSection, channels [None {type=MovieSceneScriptingParticleChannel, const ["<ParticleKey.TRIGGER: 2>", 5]}]}]
   }
  ]
 }
 P Doggus Foot Jets_L {
  id=#61
  possessed_class=Emitter
  tracks [
   {
    type=MovieSceneSpawnTrack
    sections [{
      type=MovieSceneSpawnSection
      end_frame=8462
      start_frame=8216
      channels [None {type=MovieSceneScriptingBoolChannel, default=false, keys [[8216,true],[8462,false]]}]
     }]
   }
   {
    type=MovieScene3DTransformTrack
    sections [{
      type=MovieScene3DTransformSection
      end_frame=8462
      start_frame=8216
      channels [
       Location.X {type=MovieSceneScriptingDoubleChannel, default=0}
       Location.Y {type=MovieSceneScriptingDoubleChannel, default=0}
       Location.Z {type=MovieSceneScriptingDoubleChannel, default=0}
       Rotation.X {type=MovieSceneScriptingDoubleChannel, default=0}
       Rotation.Y {type=MovieSceneScriptingDoubleChannel, default=0}
       Rotation.Z {type=MovieSceneScriptingDoubleChannel, default=0}
       Scale.X {type=MovieSceneScriptingDoubleChannel, default=1}
       Scale.Y {type=MovieSceneScriptingDoubleChannel, default=1}
       Scale.Z {type=MovieSceneScriptingDoubleChannel, default=1}
      ]
     }]
   }
   {type=MovieScene3DAttachTrack, sections [{type=MovieScene3DAttachSection, end_frame=8462, start_frame=8216}]}
   {
    type=MovieSceneParticleTrack
    display_name=Particle System
    sections [{
      type=MovieSceneParticleSection
      channels [None {type=MovieSceneScriptingParticleChannel, keys [[8216, "<ParticleKey.ACTIVATE: 0>"], [8340, "<ParticleKey.DEACTIVATE: 1>"]]}]
     }]
   }
   {
    type=MovieSceneParticleParameterTrack
    display_name=Particle Parameter
    sections [{
      type=MovieSceneParameterSection
      end_frame=8462
      start_frame=8216
      channels [
       MeshVelocity.X {type=MovieSceneScriptingFloatChannel, default=-0.5, const [-0.5,1]}
       MeshVelocity.Y {type=MovieSceneScriptingFloatChannel, default=4, const [4,1]}
       MeshVelocity.Z {type=MovieSceneScriptingFloatChannel, default=-0.82, const [-0.82,1]}
       SizeScale.X {type=MovieSceneScriptingFloatChannel, default=2, keys [[8216,0],[8236,2],[8308,1]]}
       SizeScale.Y {type=MovieSceneScriptingFloatChannel, default=1, keys [[8216,3],[8308,1]]}
       SizeScale.Z {type=MovieSceneScriptingFloatChannel, default=1, keys [[8216,3],[8308,1]]}
       SpriteVelocity.X {type=MovieSceneScriptingFloatChannel, default=-78.3298, const [-78.3298,1]}
       SpriteVelocity.Y {type=MovieSceneScriptingFloatChannel, default=600, const [600,1]}
       SpriteVelocity.Z {type=MovieSceneScriptingFloatChannel, default=-124.9804, const [-124.9804,1]}
      ]
     }]
   }
  ]
 }
 P_Doggus_Step_FX {
  id=#62
  possessed_class=Emitter
  tracks [
   {
    type=MovieSceneSpawnTrack
    sections [{
      type=MovieSceneSpawnSection
      end_frame=4688
      start_frame=4378
      channels [None {type=MovieSceneScriptingBoolChannel, default=false, keys [[4378,false],[4384,true],[4688,false]]}]
     }]
   }
   {
    type=MovieScene3DTransformTrack
    sections [{
      type=MovieScene3DTransformSection
      end_frame=4688
      start_frame=4384
      channels [
       Location.X {type=MovieSceneScriptingDoubleChannel, default=68124.0781}
       Location.Y {type=MovieSceneScriptingDoubleChannel, default=37216.5938}
       Location.Z {type=MovieSceneScriptingDoubleChannel, default=4922.7988}
       Rotation.X {type=MovieSceneScriptingDoubleChannel, default=0}
       Rotation.Y {type=MovieSceneScriptingDoubleChannel, default=0}
       Rotation.Z {type=MovieSceneScriptingDoubleChannel, default=180}
       Scale.X {type=MovieSceneScriptingDoubleChannel, default=1}
       Scale.Y {type=MovieSceneScriptingDoubleChannel, default=1}
       Scale.Z {type=MovieSceneScriptingDoubleChannel, default=1}
      ]
     }]
   }
   {
    type=MovieSceneParticleTrack
    display_name=Particle System
    sections [{
      type=MovieSceneParticleSection
      channels [None {type=MovieSceneScriptingParticleChannel, keys [[4384, "<ParticleKey.ACTIVATE: 0>"], [4448, "<ParticleKey.DEACTIVATE: 1>"]]}]
     }]
   }
  ]
 }
 P_Doggus_Trail {
  id=#63
  possessed_class=Emitter
  tracks [
   {
    type=MovieSceneSpawnTrack
    sections [{
      type=MovieSceneSpawnSection
      end_frame=6760
      start_frame=6274
      channels [None {type=MovieSceneScriptingBoolChannel, default=false, keys [[6274,false],[6280,true],[6760,false]]}]
     }]
   }
   {
    type=MovieScene3DTransformTrack
    sections [{
      type=MovieScene3DTransformSection
      end_frame=6760
      start_frame=6280
      channels [
       Location.X {type=MovieSceneScriptingDoubleChannel, default=0}
       Location.Y {type=MovieSceneScriptingDoubleChannel, default=0}
       Location.Z {type=MovieSceneScriptingDoubleChannel, default=0}
       Rotation.X {type=MovieSceneScriptingDoubleChannel, default=0}
       Rotation.Y {type=MovieSceneScriptingDoubleChannel, default=180}
       Rotation.Z {type=MovieSceneScriptingDoubleChannel, default=0}
       Scale.X {type=MovieSceneScriptingDoubleChannel, default=0.01}
       Scale.Y {type=MovieSceneScriptingDoubleChannel, default=0.01}
       Scale.Z {type=MovieSceneScriptingDoubleChannel, default=0.01}
      ]
     }]
   }
   {
    type=MovieSceneParticleTrack
    display_name=Particle System
    sections [{
      type=MovieSceneParticleSection
      channels [None {type=MovieSceneScriptingParticleChannel, keys [[6280, "<ParticleKey.ACTIVATE: 0>"], [6468, "<ParticleKey.DEACTIVATE: 1>"]]}]
     }]
   }
   {type=MovieScene3DAttachTrack, sections [{type=MovieScene3DAttachSection, end_frame=6758, start_frame=6280}]}
  ]
 }
 P_Cattus_Rocket_Impact_7 {
  id=#64
  possessed_class=Emitter
  tracks [
   {
    type=MovieSceneSpawnTrack
    sections [{
      type=MovieSceneSpawnSection
      end_frame=6832
      start_frame=6422
      channels [None {type=MovieSceneScriptingBoolChannel, default=false, keys [[6422,false],[6424,true],[6832,false]]}]
     }]
   }
   {
    type=MovieSceneParticleTrack
    display_name=Particle System
    sections [{
      type=MovieSceneParticleSection
      channels [None {type=MovieSceneScriptingParticleChannel, keys [[6424, "<ParticleKey.TRIGGER: 2>"], [6830, "<ParticleKey.DEACTIVATE: 1>"]]}]
     }]
   }
  ]
 }
 P_Doggus_Trail_02_R {
  id=#65
  possessed_class=Emitter
  tracks [
   {
    type=MovieSceneSpawnTrack
    sections [{
      type=MovieSceneSpawnSection
      end_frame=5748
      start_frame=5132
      channels [None {type=MovieSceneScriptingBoolChannel, default=false, keys [[5132,false],[5138,true],[5748,false]]}]
     }]
   }
   {
    type=MovieScene3DTransformTrack
    sections [{
      type=MovieScene3DTransformSection
      end_frame=5748
      start_frame=5132
      channels [
       Location.X {type=MovieSceneScriptingDoubleChannel, default=0}
       Location.Y {type=MovieSceneScriptingDoubleChannel, default=0}
       Location.Z {type=MovieSceneScriptingDoubleChannel, default=0}
       Rotation.X {type=MovieSceneScriptingDoubleChannel, default=0}
       Rotation.Y {type=MovieSceneScriptingDoubleChannel, default=180}
       Rotation.Z {type=MovieSceneScriptingDoubleChannel, default=0}
       Scale.X {type=MovieSceneScriptingDoubleChannel, default=0.01}
       Scale.Y {type=MovieSceneScriptingDoubleChannel, default=0.01}
       Scale.Z {type=MovieSceneScriptingDoubleChannel, default=0.01}
      ]
     }]
   }
   {
    type=MovieSceneParticleTrack
    display_name=Particle System
    sections [{
      type=MovieSceneParticleSection
      channels [None {type=MovieSceneScriptingParticleChannel, keys [[5138, "<ParticleKey.ACTIVATE: 0>"], [5337, "<ParticleKey.DEACTIVATE: 1>"]]}]
     }]
   }
   {type=MovieScene3DAttachTrack, sections [{type=MovieScene3DAttachSection, end_frame=5748, start_frame=5132}]}
  ]
 }
 P_Cattus_Rocket_Impact_15 {
  id=#66
  possessed_class=Emitter
  tracks [
   {
    type=MovieSceneSpawnTrack
    sections [{
      type=MovieSceneSpawnSection
      end_frame=7062
      start_frame=6566
      channels [None {type=MovieSceneScriptingBoolChannel, default=false, keys [[6566,false],[6568,true],[7062,false]]}]
     }]
   }
   {
    type=MovieSceneParticleTrack
    display_name=Particle System
    sections [{
      type=MovieSceneParticleSection
      channels [None {type=MovieSceneScriptingParticleChannel, keys [[6568, "<ParticleKey.TRIGGER: 2>"], [6990, "<ParticleKey.DEACTIVATE: 1>"]]}]
     }]
   }
  ]
 }
 P_Doggus_Step_FX_Light {
  id=#67
  possessed_class=Emitter
  tracks [
   {
    type=MovieScene3DTransformTrack
    sections [{
      type=MovieScene3DTransformSection
      end_frame=5094
      start_frame=4798
      channels [
       Location.X {type=MovieSceneScriptingDoubleChannel, default=59297.4219}
       Location.Y {type=MovieSceneScriptingDoubleChannel, default=36428.3828}
       Location.Z {type=MovieSceneScriptingDoubleChannel, default=2617.4067}
       Rotation.X {type=MovieSceneScriptingDoubleChannel, default=-3.647}
       Rotation.Y {type=MovieSceneScriptingDoubleChannel, default=-3.1859}
       Rotation.Z {type=MovieSceneScriptingDoubleChannel, default=1474488.375}
       Scale.X {type=MovieSceneScriptingDoubleChannel, default=1}
       Scale.Y {type=MovieSceneScriptingDoubleChannel, default=1}
       Scale.Z {type=MovieSceneScriptingDoubleChannel, default=1}
      ]
     }]
   }
   {
    type=MovieSceneParticleTrack
    display_name=Particle System
    sections [{
      type=MovieSceneParticleSection
      channels [None {type=MovieSceneScriptingParticleChannel, keys [[4798, "<ParticleKey.ACTIVATE: 0>"], [4864, "<ParticleKey.DEACTIVATE: 1>"]]}]
     }]
   }
   {
    type=MovieSceneSpawnTrack
    sections [{
      type=MovieSceneSpawnSection
      end_frame=5094
      start_frame=4796
      channels [None {type=MovieSceneScriptingBoolChannel, default=false, keys [[4796,false],[4798,true],[5094,false]]}]
     }]
   }
  ]
 }
 P_Doggus_Trail {
  id=#68
  possessed_class=Emitter
  tracks [
   {
    type=MovieSceneSpawnTrack
    sections [{
      type=MovieSceneSpawnSection
      end_frame=6632
      start_frame=6118
      channels [None {type=MovieSceneScriptingBoolChannel, default=false, keys [[6118,false],[6124,true],[6632,false]]}]
     }]
   }
   {
    type=MovieScene3DTransformTrack
    sections [{
      type=MovieScene3DTransformSection
      end_frame=6632
      start_frame=6124
      channels [
       Location.X {type=MovieSceneScriptingDoubleChannel, default=0}
       Location.Y {type=MovieSceneScriptingDoubleChannel, default=0}
       Location.Z {type=MovieSceneScriptingDoubleChannel, default=0}
       Rotation.X {type=MovieSceneScriptingDoubleChannel, default=-1980}
       Rotation.Y {type=MovieSceneScriptingDoubleChannel, default=1080}
       Rotation.Z {type=MovieSceneScriptingDoubleChannel, default=-540}
       Scale.X {type=MovieSceneScriptingDoubleChannel, default=0.01}
       Scale.Y {type=MovieSceneScriptingDoubleChannel, default=0.01}
       Scale.Z {type=MovieSceneScriptingDoubleChannel, default=0.01}
      ]
     }]
   }
   {
    type=MovieSceneParticleTrack
    display_name=Particle System
    sections [{
      type=MovieSceneParticleSection
      channels [None {type=MovieSceneScriptingParticleChannel, keys [[6124, "<ParticleKey.ACTIVATE: 0>"], [6323, "<ParticleKey.DEACTIVATE: 1>"]]}]
     }]
   }
   {type=MovieScene3DAttachTrack, sections [{type=MovieScene3DAttachSection, end_frame=6632, start_frame=6124}]}
  ]
 }
 P_Doggus_Trail {
  id=#69
  possessed_class=Emitter
  tracks [
   {
    type=MovieSceneSpawnTrack
    sections [{
      type=MovieSceneSpawnSection
      end_frame=6720
      start_frame=6228
      channels [None {type=MovieSceneScriptingBoolChannel, default=false, keys [[6228,false],[6234,true],[6720,false]]}]
     }]
   }
   {
    type=MovieScene3DTransformTrack
    sections [{
      type=MovieScene3DTransformSection
      end_frame=6720
      start_frame=6234
      channels [
       Location.X {type=MovieSceneScriptingDoubleChannel, default=0}
       Location.Y {type=MovieSceneScriptingDoubleChannel, default=0}
       Location.Z {type=MovieSceneScriptingDoubleChannel, default=0}
       Rotation.X {type=MovieSceneScriptingDoubleChannel, default=0}
       Rotation.Y {type=MovieSceneScriptingDoubleChannel, default=180}
       Rotation.Z {type=MovieSceneScriptingDoubleChannel, default=0}
       Scale.X {type=MovieSceneScriptingDoubleChannel, default=0.01}
       Scale.Y {type=MovieSceneScriptingDoubleChannel, default=0.01}
       Scale.Z {type=MovieSceneScriptingDoubleChannel, default=0.01}
      ]
     }]
   }
   {
    type=MovieSceneParticleTrack
    display_name=Particle System
    sections [{
      type=MovieSceneParticleSection
      channels [None {type=MovieSceneScriptingParticleChannel, keys [[6234, "<ParticleKey.ACTIVATE: 0>"], [6424, "<ParticleKey.DEACTIVATE: 1>"]]}]
     }]
   }
   {type=MovieScene3DAttachTrack, sections [{type=MovieScene3DAttachSection, end_frame=6726, start_frame=6234}]}
  ]
 }
 P_Doggus_Step_Large_Directional {
  id=#70
  possessed_class=Emitter
  tracks [
   {
    type=MovieSceneSpawnTrack
    sections [{
      type=MovieSceneSpawnSection
      end_frame=8364
      start_frame=8054
      channels [None {type=MovieSceneScriptingBoolChannel, default=false, keys [[8054,false],[8060,true],[8364,false]]}]
     }]
   }
   {
    type=MovieScene3DTransformTrack
    sections [{
      type=MovieScene3DTransformSection
      end_frame=8364
      start_frame=8054
      channels [
       Location.X {type=MovieSceneScriptingDoubleChannel, default=50015.8789}
       Location.Y {type=MovieSceneScriptingDoubleChannel, default=35170.668}
       Location.Z {type=MovieSceneScriptingDoubleChannel, default=2904.3582}
       Rotation.X {type=MovieSceneScriptingDoubleChannel, default=5.859, const [5.859,1]}
       Rotation.Y {type=MovieSceneScriptingDoubleChannel, default=3.2654, const [3.2654,1]}
       Rotation.Z {type=MovieSceneScriptingDoubleChannel, default=-80, const [-80,1]}
       Scale.X {type=MovieSceneScriptingDoubleChannel, default=1}
       Scale.Y {type=MovieSceneScriptingDoubleChannel, default=1}
       Scale.Z {type=MovieSceneScriptingDoubleChannel, default=1}
      ]
     }]
   }
   {
    type=MovieSceneParticleTrack
    display_name=Particle System
    sections [{
      type=MovieSceneParticleSection
      channels [None {type=MovieSceneScriptingParticleChannel, keys [[8060, "<ParticleKey.ACTIVATE: 0>"], [8126, "<ParticleKey.DEACTIVATE: 1>"]]}]
     }]
   }
  ]
 }
 P_Doggus_Step_FX_Light {
  id=#71
  possessed_class=Emitter
  tracks [
   {
    type=MovieScene3DTransformTrack
    sections [{
      type=MovieScene3DTransformSection
      end_frame=4938
      start_frame=4710
      channels [
       Location.X {type=MovieSceneScriptingDoubleChannel, default=54489.2539}
       Location.Y {type=MovieSceneScriptingDoubleChannel, default=36314.3047}
       Location.Z {type=MovieSceneScriptingDoubleChannel, default=2799.8386}
       Rotation.X {type=MovieSceneScriptingDoubleChannel, default=5.6386}
       Rotation.Y {type=MovieSceneScriptingDoubleChannel, default=0}
       Rotation.Z {type=MovieSceneScriptingDoubleChannel, default=184214}
       Scale.X {type=MovieSceneScriptingDoubleChannel, default=1}
       Scale.Y {type=MovieSceneScriptingDoubleChannel, default=1}
       Scale.Z {type=MovieSceneScriptingDoubleChannel, default=1}
      ]
     }]
   }
   {
    type=MovieSceneParticleTrack
    display_name=Particle System
    sections [{
      type=MovieSceneParticleSection
      channels [None {type=MovieSceneScriptingParticleChannel, keys [[4710, "<ParticleKey.ACTIVATE: 0>"], [4776, "<ParticleKey.DEACTIVATE: 1>"]]}]
     }]
   }
   {
    type=MovieSceneSpawnTrack
    sections [{
      type=MovieSceneSpawnSection
      end_frame=4938
      start_frame=4708
      channels [None {type=MovieSceneScriptingBoolChannel, default=false, keys [[4708,false],[4710,true],[4938,false]]}]
     }]
   }
  ]
 }
 P_Doggus_Step_FX {
  id=#72
  possessed_class=Emitter
  tracks [
   {
    type=MovieSceneSpawnTrack
    sections [{
      type=MovieSceneSpawnSection
      end_frame=8398
      start_frame=8084
      channels [None {type=MovieSceneScriptingBoolChannel, default=false, keys [[8084,false],[8090,true],[8398,false]]}]
     }]
   }
   {
    type=MovieScene3DTransformTrack
    sections [{
      type=MovieScene3DTransformSection
      end_frame=8398
      start_frame=8084
      channels [
       Location.X {type=MovieSceneScriptingDoubleChannel, default=52011.9336}
       Location.Y {type=MovieSceneScriptingDoubleChannel, default=21041.6875}
       Location.Z {type=MovieSceneScriptingDoubleChannel, default=2732.7937}
       Rotation.X {type=MovieSceneScriptingDoubleChannel, default=0.9348}
       Rotation.Y {type=MovieSceneScriptingDoubleChannel, default=-5.6774}
       Rotation.Z {type=MovieSceneScriptingDoubleChannel, default=-76.5156}
       Scale.X {type=MovieSceneScriptingDoubleChannel, default=1}
       Scale.Y {type=MovieSceneScriptingDoubleChannel, default=1}
       Scale.Z {type=MovieSceneScriptingDoubleChannel, default=1}
      ]
     }]
   }
   {
    type=MovieSceneParticleTrack
    display_name=Particle System
    sections [{
      type=MovieSceneParticleSection
      channels [None {type=MovieSceneScriptingParticleChannel, keys [[8090, "<ParticleKey.ACTIVATE: 0>"], [8156, "<ParticleKey.DEACTIVATE: 1>"]]}]
     }]
   }
  ]
 }
 P_Doggus_Smoke_Simmer {
  id=#73
  possessed_class=Emitter
  tracks [
   {
    type=MovieSceneSpawnTrack
    sections [{
      type=MovieSceneSpawnSection
      end_frame=7166
      start_frame=5228
      channels [None {type=MovieSceneScriptingBoolChannel, default=false, keys [[5466,false],[5472,true],[7166,false]]}]
     }]
   }
   {
    type=MovieScene3DTransformTrack
    sections [{
      type=MovieScene3DTransformSection
      end_frame=7166
      start_frame=5472
      channels [
       Location.X {type=MovieSceneScriptingDoubleChannel, default=49.0105}
       Location.Y {type=MovieSceneScriptingDoubleChannel, default=-6.4877}
       Location.Z {type=MovieSceneScriptingDoubleChannel, default=5.3052}
       Rotation.X {type=MovieSceneScriptingDoubleChannel, default=0}
       Rotation.Y {type=MovieSceneScriptingDoubleChannel, default=0}
       Rotation.Z {type=MovieSceneScriptingDoubleChannel, default=0}
       Scale.X {type=MovieSceneScriptingDoubleChannel, default=0.01}
       Scale.Y {type=MovieSceneScriptingDoubleChannel, default=0.01}
       Scale.Z {type=MovieSceneScriptingDoubleChannel, default=0.01}
      ]
     }]
   }
   {
    type=MovieSceneParticleTrack
    display_name=Particle System
    sections [{
      type=MovieSceneParticleSection
      channels [None {type=MovieSceneScriptingParticleChannel, keys [[5474, "<ParticleKey.ACTIVATE: 0>"], [5786, "<ParticleKey.DEACTIVATE: 1>"]]}]
     }]
   }
   {type=MovieScene3DAttachTrack, sections [{type=MovieScene3DAttachSection, end_frame=7166, start_frame=5228}]}
  ]
 }
 P_Doggus_Fire_01 {
  id=#74
  possessed_class=Emitter
  tracks [
   {
    type=MovieSceneParticleTrack
    display_name=Particle System
    sections [{type=MovieSceneParticleSection, channels [None {type=MovieSceneScriptingParticleChannel, const ["<ParticleKey.TRIGGER: 2>", 12]}]}]
   }
   {
    type=MovieSceneSpawnTrack
    sections [{
      type=MovieSceneSpawnSection
      end_frame=5798
      start_frame=5012
      channels [None {type=MovieSceneScriptingBoolChannel, default=false, keys [[5012,false],[5014,true],[5796,false]]}]
     }]
   }
   {
    type=MovieScene3DTransformTrack
    sections [{
      type=MovieScene3DTransformSection
      end_frame=5796
      start_frame=5012
      channels [
       Location.X {
        type=MovieSceneScriptingDoubleChannel
        default=55475.3125
        keys [[5064,55475.3125],[5072,58667.1172],[5086,55475.3125],[5096,58758.5391],[5106,55640.8867],[5122,58667.1172],[5138,55314.9531],[5162,58460.3945],[5194,58731.9414],[5256,55628.7188],[5276,55330.3828],[5324,58507.9375]]
       }
       Location.Y {
        type=MovieSceneScriptingDoubleChannel
        default=35308.3281
        keys [[5064,35308.3281],[5072,35490.0117],[5086,35308.3281],[5096,35809.4648],[5106,35256.9453],[5122,35490.0117],[5138,35475.7617],[5162,35285.0117],[5194,35423.3203],[5256,34860.2422],[5276,35155.4414],[5324,34817.9766]]
       }
       Location.Z {
        type=MovieSceneScriptingDoubleChannel
        default=14686.4824
        keys [[5064,14686.4824],[5072,14753.9443],[5086,14686.4824],[5096,14450.9424],[5106,14271.1357],[5122,14753.9443],[5138,14331.6104],[5162,14311.3535],[5194,14392.5566],[5256,14280.9893],[5276,14370.4551],[5324,14318.6338]]
       }
       Rotation.X {
        type=MovieSceneScriptingDoubleChannel
        default=0.3529
        keys [[5064,0.3529],[5072,-8.0909],[5086,0.3529],[5096,-8.1583],[5106,0.3529],[5122,-8.0909],[5138,0.3529],[5162,-8.0909],[5194,-8.0909],[5256,0.3529],[5276,0.3529],[5324,-8.0909]]
       }
       Rotation.Y {
        type=MovieSceneScriptingDoubleChannel
        default=-8.2624
        keys [[5064,-8.2624],[5072,-1.7173],[5086,-8.2624],[5096,-6.8021],[5106,-8.2624],[5122,-1.7173],[5138,-8.2624],[5162,-1.7173],[5194,-1.7173],[5256,-8.2624],[5276,-8.2624],[5324,-1.7173]]
       }
       Rotation.Z {
        type=MovieSceneScriptingDoubleChannel
        default=-125.1689
        keys [[5064,-125.1688],[5072,-44.6192],[5086,-125.1688],[5096,-43.4483],[5106,-125.1688],[5122,-44.6192],[5138,-125.1688],[5162,-44.6192],[5194,-44.6192],[5256,-125.1688],[5276,-125.1688],[5324,-44.6192]]
       }
       Scale.X {type=MovieSceneScriptingDoubleChannel, default=1.25}
       Scale.Y {type=MovieSceneScriptingDoubleChannel, default=1.25}
       Scale.Z {type=MovieSceneScriptingDoubleChannel, default=1.25}
      ]
     }]
   }
   {type=MovieScene3DAttachTrack, sections []}
  ]
 }
 P_Cattus_Rocket_Impact_Large_04 {
  id=#75
  possessed_class=Emitter
  tracks [
   {
    type=MovieSceneSpawnTrack
    sections [{
      type=MovieSceneSpawnSection
      end_frame=5826
      start_frame=5304
      channels [None {type=MovieSceneScriptingBoolChannel, default=false, keys [[5304,false],[5306,true],[5826,false]]}]
     }]
   }
   {
    type=MovieScene3DTransformTrack
    sections [{
      type=MovieScene3DTransformSection
      end_frame=5826
      start_frame=5306
      channels [
       Location.X {type=MovieSceneScriptingDoubleChannel, default=66918.0703}
       Location.Y {type=MovieSceneScriptingDoubleChannel, default=-35170.9219}
       Location.Z {type=MovieSceneScriptingDoubleChannel, default=9056.5391}
       Rotation.X {type=MovieSceneScriptingDoubleChannel, default=0}
       Rotation.Y {type=MovieSceneScriptingDoubleChannel, default=-2.9841}
       Rotation.Z {type=MovieSceneScriptingDoubleChannel, default=258785376}
       Scale.X {type=MovieSceneScriptingDoubleChannel, default=1.2}
       Scale.Y {type=MovieSceneScriptingDoubleChannel, default=1.2}
       Scale.Z {type=MovieSceneScriptingDoubleChannel, default=1.2}
      ]
     }]
   }
   {
    type=MovieSceneParticleTrack
    display_name=Particle System
    sections [{
      type=MovieSceneParticleSection
      channels [None {type=MovieSceneScriptingParticleChannel, keys [[5306, "<ParticleKey.ACTIVATE: 0>"], [5542, "<ParticleKey.DEACTIVATE: 1>"]]}]
     }]
   }
  ]
 }
 P_Doggus_Step_FX {
  id=#76
  possessed_class=Emitter
  tracks [
   {
    type=MovieSceneSpawnTrack
    sections [{
      type=MovieSceneSpawnSection
      end_frame=8484
      start_frame=8160
      channels [None {type=MovieSceneScriptingBoolChannel, default=false, keys [[8160,false],[8166,true],[8484,false]]}]
     }]
   }
   {
    type=MovieScene3DTransformTrack
    sections [{
      type=MovieScene3DTransformSection
      end_frame=8484
      start_frame=8160
      channels [
       Location.X {type=MovieSceneScriptingDoubleChannel, default=59817}
       Location.Y {type=MovieSceneScriptingDoubleChannel, default=-12382.2764}
       Location.Z {type=MovieSceneScriptingDoubleChannel, default=-369.0144}
       Rotation.X {type=MovieSceneScriptingDoubleChannel, default=0.9348}
       Rotation.Y {type=MovieSceneScriptingDoubleChannel, default=-5.6774}
       Rotation.Z {type=MovieSceneScriptingDoubleChannel, default=-70}
       Scale.X {type=MovieSceneScriptingDoubleChannel, default=1}
       Scale.Y {type=MovieSceneScriptingDoubleChannel, default=1}
       Scale.Z {type=MovieSceneScriptingDoubleChannel, default=1}
      ]
     }]
   }
   {
    type=MovieSceneParticleTrack
    display_name=Particle System
    sections [{
      type=MovieSceneParticleSection
      channels [None {type=MovieSceneScriptingParticleChannel, keys [[8166, "<ParticleKey.ACTIVATE: 0>"], [8232, "<ParticleKey.DEACTIVATE: 1>"]]}]
     }]
   }
  ]
 }
 P_Doggus_Trail {
  id=#77
  possessed_class=Emitter
  tracks [
   {
    type=MovieSceneSpawnTrack
    sections [{
      type=MovieSceneSpawnSection
      end_frame=6632
      start_frame=6118
      channels [None {type=MovieSceneScriptingBoolChannel, default=false, keys [[6118,false],[6124,true],[6632,false]]}]
     }]
   }
   {
    type=MovieScene3DTransformTrack
    sections [{
      type=MovieScene3DTransformSection
      end_frame=6632
      start_frame=6124
      channels [
       Location.X {type=MovieSceneScriptingDoubleChannel, default=0}
       Location.Y {type=MovieSceneScriptingDoubleChannel, default=0}
       Location.Z {type=MovieSceneScriptingDoubleChannel, default=0}
       Rotation.X {type=MovieSceneScriptingDoubleChannel, default=-1980}
       Rotation.Y {type=MovieSceneScriptingDoubleChannel, default=1080}
       Rotation.Z {type=MovieSceneScriptingDoubleChannel, default=-540}
       Scale.X {type=MovieSceneScriptingDoubleChannel, default=0.01}
       Scale.Y {type=MovieSceneScriptingDoubleChannel, default=0.01}
       Scale.Z {type=MovieSceneScriptingDoubleChannel, default=0.01}
      ]
     }]
   }
   {
    type=MovieSceneParticleTrack
    display_name=Particle System
    sections [{
      type=MovieSceneParticleSection
      channels [None {type=MovieSceneScriptingParticleChannel, keys [[6124, "<ParticleKey.ACTIVATE: 0>"], [6325, "<ParticleKey.DEACTIVATE: 1>"]]}]
     }]
   }
   {type=MovieScene3DAttachTrack, sections [{type=MovieScene3DAttachSection, end_frame=6632, start_frame=6124}]}
  ]
 }
 P_Cattus_Rocket_Impact_Large_03 {
  id=#78
  possessed_class=Emitter
  tracks [
   {
    type=MovieSceneSpawnTrack
    sections [{
      type=MovieSceneSpawnSection
      end_frame=5934
      start_frame=5294
      channels [None {type=MovieSceneScriptingBoolChannel, default=false, keys [[5292,false],[5294,true],[5934,false]]}]
     }]
   }
   {
    type=MovieScene3DTransformTrack
    sections [{
      type=MovieScene3DTransformSection
      end_frame=5934
      start_frame=5294
      channels [
       Location.X {type=MovieSceneScriptingDoubleChannel, default=73671.1953}
       Location.Y {type=MovieSceneScriptingDoubleChannel, default=-33306.2578}
       Location.Z {type=MovieSceneScriptingDoubleChannel, default=4734.7129}
       Rotation.X {type=MovieSceneScriptingDoubleChannel, default=-93.3599}
       Rotation.Y {type=MovieSceneScriptingDoubleChannel, default=12.8763}
       Rotation.Z {type=MovieSceneScriptingDoubleChannel, default=119439440}
       Scale.X {type=MovieSceneScriptingDoubleChannel, default=1}
       Scale.Y {type=MovieSceneScriptingDoubleChannel, default=1}
       Scale.Z {type=MovieSceneScriptingDoubleChannel, default=1}
      ]
     }]
   }
   {
    type=MovieSceneParticleTrack
    display_name=Particle System
    sections [{
      type=MovieSceneParticleSection
      channels [None {type=MovieSceneScriptingParticleChannel, keys [[5294, "<ParticleKey.ACTIVATE: 0>"], [5528, "<ParticleKey.DEACTIVATE: 1>"]]}]
     }]
   }
  ]
 }
 P_Doggus_Trail {
  id=#79
  possessed_class=Emitter
  tracks [
   {
    type=MovieSceneSpawnTrack
    sections [{
      type=MovieSceneSpawnSection
      end_frame=6672
      start_frame=6150
      channels [None {type=MovieSceneScriptingBoolChannel, default=false, keys [[6150,false],[6156,true],[6672,false]]}]
     }]
   }
   {
    type=MovieScene3DTransformTrack
    sections [{
      type=MovieScene3DTransformSection
      end_frame=6672
      start_frame=6156
      channels [
       Location.X {type=MovieSceneScriptingDoubleChannel, default=0}
       Location.Y {type=MovieSceneScriptingDoubleChannel, default=0}
       Location.Z {type=MovieSceneScriptingDoubleChannel, default=0}
       Rotation.X {type=MovieSceneScriptingDoubleChannel, default=0}
       Rotation.Y {type=MovieSceneScriptingDoubleChannel, default=0}
       Rotation.Z {type=MovieSceneScriptingDoubleChannel, default=0}
       Scale.X {type=MovieSceneScriptingDoubleChannel, default=0.01}
       Scale.Y {type=MovieSceneScriptingDoubleChannel, default=0.01}
       Scale.Z {type=MovieSceneScriptingDoubleChannel, default=0.01}
      ]
     }]
   }
   {
    type=MovieSceneParticleTrack
    display_name=Particle System
    sections [{
      type=MovieSceneParticleSection
      channels [None {type=MovieSceneScriptingParticleChannel, keys [[6156, "<ParticleKey.ACTIVATE: 0>"], [6348, "<ParticleKey.DEACTIVATE: 1>"]]}]
     }]
   }
   {type=MovieScene3DAttachTrack, sections [{type=MovieScene3DAttachSection, end_frame=6672, start_frame=6156}]}
  ]
 }
 P_Cattus_Rocket_Impact_Large_01 {
  id=#80
  possessed_class=Emitter
  tracks [
   {
    type=MovieSceneSpawnTrack
    sections [{
      type=MovieSceneSpawnSection
      end_frame=5966
      start_frame=5284
      channels [None {type=MovieSceneScriptingBoolChannel, default=false, keys [[5282,false],[5284,true],[5966,false]]}]
     }]
   }
   {
    type=MovieScene3DTransformTrack
    sections [{
      type=MovieScene3DTransformSection
      end_frame=5966
      start_frame=5284
      channels [
       Location.X {type=MovieSceneScriptingDoubleChannel, default=65775.5156}
       Location.Y {type=MovieSceneScriptingDoubleChannel, default=-30919.9023}
       Location.Z {type=MovieSceneScriptingDoubleChannel, default=6827.1265}
       Rotation.X {type=MovieSceneScriptingDoubleChannel, default=-93.4142}
       Rotation.Y {type=MovieSceneScriptingDoubleChannel, default=2.1468}
       Rotation.Z {type=MovieSceneScriptingDoubleChannel, default=544416.1875}
       Scale.X {type=MovieSceneScriptingDoubleChannel, default=1}
       Scale.Y {type=MovieSceneScriptingDoubleChannel, default=1}
       Scale.Z {type=MovieSceneScriptingDoubleChannel, default=1}
      ]
     }]
   }
   {
    type=MovieSceneParticleTrack
    display_name=Particle System
    sections [{
      type=MovieSceneParticleSection
      channels [None {type=MovieSceneScriptingParticleChannel, keys [[5284, "<ParticleKey.ACTIVATE: 0>"], [5518, "<ParticleKey.DEACTIVATE: 1>"]]}]
     }]
   }
  ]
 }
 P_Doggus_Smoke_Simmer {
  id=#81
  possessed_class=Emitter
  tracks [
   {
    type=MovieSceneSpawnTrack
    sections [{
      type=MovieSceneSpawnSection
      end_frame=8378
      start_frame=6660
      channels [None {type=MovieSceneScriptingBoolChannel, default=false, keys [[6660,false],[6666,true],[8378,false]]}]
     }]
   }
   {
    type=MovieScene3DTransformTrack
    sections [{
      type=MovieScene3DTransformSection
      end_frame=8378
      start_frame=6440
      channels [
       Location.X {type=MovieSceneScriptingDoubleChannel, default=-1.7196}
       Location.Y {type=MovieSceneScriptingDoubleChannel, default=-26.9248}
       Location.Z {type=MovieSceneScriptingDoubleChannel, default=-12.1711}
       Rotation.X {type=MovieSceneScriptingDoubleChannel, default=-0.0002}
       Rotation.Y {type=MovieSceneScriptingDoubleChannel, default=-20.6294}
       Rotation.Z {type=MovieSceneScriptingDoubleChannel, default=-94.155}
       Scale.X {type=MovieSceneScriptingDoubleChannel, default=0.01}
       Scale.Y {type=MovieSceneScriptingDoubleChannel, default=0.01}
       Scale.Z {type=MovieSceneScriptingDoubleChannel, default=0.01}
      ]
     }]
   }
   {
    type=MovieSceneParticleTrack
    display_name=Particle System
    sections [{
      type=MovieSceneParticleSection
      channels [None {type=MovieSceneScriptingParticleChannel, keys [[6668, "<ParticleKey.ACTIVATE: 0>"], [7110, "<ParticleKey.DEACTIVATE: 1>"]]}]
     }]
   }
   {type=MovieScene3DAttachTrack, sections [{type=MovieScene3DAttachSection, end_frame=8378, start_frame=5928}]}
  ]
 }
 P_Doggus_Trail_01_R {
  id=#82
  possessed_class=Emitter
  tracks [
   {
    type=MovieSceneSpawnTrack
    sections [{
      type=MovieSceneSpawnSection
      end_frame=5656
      start_frame=5060
      channels [None {type=MovieSceneScriptingBoolChannel, default=false, keys [[5060,false],[5072,true],[5656,false]]}]
     }]
   }
   {
    type=MovieScene3DTransformTrack
    sections [{
      type=MovieScene3DTransformSection
      end_frame=5656
      start_frame=5060
      channels [
       Location.X {type=MovieSceneScriptingDoubleChannel, default=0}
       Location.Y {type=MovieSceneScriptingDoubleChannel, default=0}
       Location.Z {type=MovieSceneScriptingDoubleChannel, default=0}
       Rotation.X {type=MovieSceneScriptingDoubleChannel, default=0}
       Rotation.Y {type=MovieSceneScriptingDoubleChannel, default=180}
       Rotation.Z {type=MovieSceneScriptingDoubleChannel, default=0}
       Scale.X {type=MovieSceneScriptingDoubleChannel, default=0.01}
       Scale.Y {type=MovieSceneScriptingDoubleChannel, default=0.01}
       Scale.Z {type=MovieSceneScriptingDoubleChannel, default=0.01}
      ]
     }]
   }
   {
    type=MovieSceneParticleTrack
    display_name=Particle System
    sections [{
      type=MovieSceneParticleSection
      channels [None {type=MovieSceneScriptingParticleChannel, keys [[5072, "<ParticleKey.ACTIVATE: 0>"], [5280, "<ParticleKey.DEACTIVATE: 1>"]]}]
     }]
   }
   {type=MovieScene3DAttachTrack, sections [{type=MovieScene3DAttachSection, end_frame=5656, start_frame=5060}]}
  ]
 }
 P_Cattus_Rocket_Impact_13 {
  id=#83
  possessed_class=Emitter
  tracks [
   {
    type=MovieSceneSpawnTrack
    sections [{
      type=MovieSceneSpawnSection
      end_frame=6964
      start_frame=6534
      channels [None {type=MovieSceneScriptingBoolChannel, default=false, keys [[6534,false],[6536,true],[6964,false]]}]
     }]
   }
   {
    type=MovieSceneParticleTrack
    display_name=Particle System
    sections [{type=MovieSceneParticleSection, channels [None {type=MovieSceneScriptingParticleChannel, const ["<ParticleKey.TRIGGER: 2>", 1]}]}]
   }
  ]
 }
 P_Cattus_Rocket_Impact_5 {
  id=#84
  possessed_class=Emitter
  tracks [
   {
    type=MovieSceneSpawnTrack
    sections [{
      type=MovieSceneSpawnSection
      end_frame=6780
      start_frame=6394
      channels [None {type=MovieSceneScriptingBoolChannel, default=false, keys [[6394,false],[6396,true],[6780,false]]}]
     }]
   }
   {
    type=MovieSceneParticleTrack
    display_name=Particle System
    sections [{
      type=MovieSceneParticleSection
      channels [None {type=MovieSceneScriptingParticleChannel, keys [[6396, "<ParticleKey.TRIGGER: 2>"], [6778, "<ParticleKey.DEACTIVATE: 1>"]]}]
     }]
   }
  ]
 }
 P_Doggus_Trail {
  id=#85
  possessed_class=Emitter
  tracks [
   {
    type=MovieSceneSpawnTrack
    sections [{
      type=MovieSceneSpawnSection
      end_frame=6872
      start_frame=6380
      channels [None {type=MovieSceneScriptingBoolChannel, default=false, keys [[6380,false],[6386,true],[6872,false]]}]
     }]
   }
   {
    type=MovieScene3DTransformTrack
    sections [{
      type=MovieScene3DTransformSection
      end_frame=6872
      start_frame=6386
      channels [
       Location.X {type=MovieSceneScriptingDoubleChannel, default=0}
       Location.Y {type=MovieSceneScriptingDoubleChannel, default=0}
       Location.Z {type=MovieSceneScriptingDoubleChannel, default=0}
       Rotation.X {type=MovieSceneScriptingDoubleChannel, default=0}
       Rotation.Y {type=MovieSceneScriptingDoubleChannel, default=180}
       Rotation.Z {type=MovieSceneScriptingDoubleChannel, default=0}
       Scale.X {type=MovieSceneScriptingDoubleChannel, default=0.01}
       Scale.Y {type=MovieSceneScriptingDoubleChannel, default=0.01}
       Scale.Z {type=MovieSceneScriptingDoubleChannel, default=0.01}
      ]
     }]
   }
   {
    type=MovieSceneParticleTrack
    display_name=Particle System
    sections [{
      type=MovieSceneParticleSection
      channels [None {type=MovieSceneScriptingParticleChannel, keys [[6386, "<ParticleKey.ACTIVATE: 0>"], [6568, "<ParticleKey.DEACTIVATE: 1>"]]}]
     }]
   }
   {type=MovieScene3DAttachTrack, sections [{type=MovieScene3DAttachSection, end_frame=6872, start_frame=6386}]}
  ]
 }
 P Directional Dust {
  id=#86
  possessed_class=Emitter
  tracks [
   {
    type=MovieSceneSpawnTrack
    sections [{
      type=MovieSceneSpawnSection
      end_frame=8500
      start_frame=8024
      channels [None {type=MovieSceneScriptingBoolChannel, default=false, keys [[8024,false],[8026,true],[8500,false]]}]
     }]
   }
   {
    type=MovieScene3DTransformTrack
    sections [{
      type=MovieScene3DTransformSection
      end_frame=8664
      start_frame=8024
      channels [
       Location.X {type=MovieSceneScriptingDoubleChannel, default=49394.5195, keys [[8026,47529.75],[8092,47739.5898]]}
       Location.Y {type=MovieSceneScriptingDoubleChannel, default=51882.9961, keys [[8026,44966.0625],[8092,43234.7891]]}
       Location.Z {type=MovieSceneScriptingDoubleChannel, default=2618.2417, const [2618.2417,2]}
       Rotation.X {type=MovieSceneScriptingDoubleChannel, default=0, const [0,1]}
       Rotation.Y {type=MovieSceneScriptingDoubleChannel, default=0, const [0,1]}
       Rotation.Z {type=MovieSceneScriptingDoubleChannel, default=96.9109, const [96.9109,1]}
       Scale.X {type=MovieSceneScriptingDoubleChannel, default=1, keys [[8026,0.75],[8092,1]]}
       Scale.Y {type=MovieSceneScriptingDoubleChannel, default=0.8, keys [[8026,0.75],[8092,0.8]]}
       Scale.Z {type=MovieSceneScriptingDoubleChannel, default=1, keys [[8026,0.75],[8092,1]]}
      ]
     }]
   }
   {
    type=MovieSceneParticleTrack
    display_name=Particle System
    sections [{
      type=MovieSceneParticleSection
      channels [None {type=MovieSceneScriptingParticleChannel, keys [[8026, "<ParticleKey.ACTIVATE: 0>"], [8092, "<ParticleKey.DEACTIVATE: 1>"]]}]
     }]
   }
  ]
 }
 P_Cattus_Rocket_Impact_Large_02 {
  id=#87
  possessed_class=Emitter
  tracks [
   {
    type=MovieSceneSpawnTrack
    sections [{
      type=MovieSceneSpawnSection
      end_frame=5966
      start_frame=5272
      channels [None {type=MovieSceneScriptingBoolChannel, default=false, keys [[5272,false],[5290,true],[5966,false]]}]
     }]
   }
   {
    type=MovieScene3DTransformTrack
    sections [{
      type=MovieScene3DTransformSection
      end_frame=5966
      start_frame=5272
      channels [
       Location.X {type=MovieSceneScriptingDoubleChannel, default=71811.1562}
       Location.Y {type=MovieSceneScriptingDoubleChannel, default=-32083.748}
       Location.Z {type=MovieSceneScriptingDoubleChannel, default=9094.4775}
       Rotation.X {type=MovieSceneScriptingDoubleChannel, default=-93.0463}
       Rotation.Y {type=MovieSceneScriptingDoubleChannel, default=15.13}
       Rotation.Z {type=MovieSceneScriptingDoubleChannel, default=44789848}
       Scale.X {type=MovieSceneScriptingDoubleChannel, default=1}
       Scale.Y {type=MovieSceneScriptingDoubleChannel, default=1}
       Scale.Z {type=MovieSceneScriptingDoubleChannel, default=1}
      ]
     }]
   }
   {
    type=MovieSceneParticleTrack
    display_name=Particle System
    sections [{
      type=MovieSceneParticleSection
      channels [None {type=MovieSceneScriptingParticleChannel, keys [[5290, "<ParticleKey.ACTIVATE: 0>"], [5506, "<ParticleKey.DEACTIVATE: 1>"]]}]
     }]
   }
  ]
 }
 P_Cattus_Rocket_Impact_Large_06 {
  id=#88
  possessed_class=Emitter
  tracks [
   {
    type=MovieSceneSpawnTrack
    sections [{
      type=MovieSceneSpawnSection
      end_frame=5826
      start_frame=5362
      channels [None {type=MovieSceneScriptingBoolChannel, default=false, keys [[5362,false],[5364,true],[5826,false]]}]
     }]
   }
   {
    type=MovieScene3DTransformTrack
    sections [{
      type=MovieScene3DTransformSection
      end_frame=5826
      start_frame=5364
      channels [
       Location.X {type=MovieSceneScriptingDoubleChannel, default=66886.6406}
       Location.Y {type=MovieSceneScriptingDoubleChannel, default=-44423.2188}
       Location.Z {type=MovieSceneScriptingDoubleChannel, default=5970.625}
       Rotation.X {type=MovieSceneScriptingDoubleChannel, default=0}
       Rotation.Y {type=MovieSceneScriptingDoubleChannel, default=0}
       Rotation.Z {type=MovieSceneScriptingDoubleChannel, default=90}
       Scale.X {type=MovieSceneScriptingDoubleChannel, default=1.2}
       Scale.Y {type=MovieSceneScriptingDoubleChannel, default=1.2}
       Scale.Z {type=MovieSceneScriptingDoubleChannel, default=1.2}
      ]
     }]
   }
   {
    type=MovieSceneParticleTrack
    display_name=Particle System
    sections [{
      type=MovieSceneParticleSection
      channels [None {type=MovieSceneScriptingParticleChannel, keys [[5364, "<ParticleKey.ACTIVATE: 0>"], [5598, "<ParticleKey.DEACTIVATE: 1>"]]}]
     }]
   }
  ]
 }
 P_Cattus_Rocket_Impact_8 {
  id=#89
  possessed_class=Emitter
  tracks [
   {
    type=MovieSceneSpawnTrack
    sections [{
      type=MovieSceneSpawnSection
      end_frame=6864
      start_frame=6426
      channels [None {type=MovieSceneScriptingBoolChannel, default=false, keys [[6426,false],[6428,true],[6864,false]]}]
     }]
   }
   {
    type=MovieSceneParticleTrack
    display_name=Particle System
    sections [{
      type=MovieSceneParticleSection
      channels [None {type=MovieSceneScriptingParticleChannel, keys [[6428, "<ParticleKey.TRIGGER: 2>"], [6862, "<ParticleKey.DEACTIVATE: 1>"]]}]
     }]
   }
  ]
 }
 P_Cattus_Rocket_Impact_14 {
  id=#90
  possessed_class=Emitter
  tracks [
   {
    type=MovieSceneSpawnTrack
    sections [{
      type=MovieSceneSpawnSection
      end_frame=7016
      start_frame=6544
      channels [None {type=MovieSceneScriptingBoolChannel, default=false, keys [[6544,false],[6546,true],[7016,false]]}]
     }]
   }
   {
    type=MovieSceneParticleTrack
    display_name=Particle System
    sections [{
      type=MovieSceneParticleSection
      channels [None {type=MovieSceneScriptingParticleChannel, keys [[6546, "<ParticleKey.TRIGGER: 2>"], [6848, "<ParticleKey.DEACTIVATE: 1>"]]}]
     }]
   }
  ]
 }
 P Directional Dust {
  id=#91
  possessed_class=Emitter
  tracks [
   {
    type=MovieSceneSpawnTrack
    sections [{
      type=MovieSceneSpawnSection
      end_frame=8178
      start_frame=7394
      channels [None {type=MovieSceneScriptingBoolChannel, default=false, keys [[7394,false],[7396,true],[8178,false]]}]
     }]
   }
   {
    type=MovieScene3DTransformTrack
    sections [{
      type=MovieScene3DTransformSection
      end_frame=8180
      start_frame=7386
      channels [
       Location.X {type=MovieSceneScriptingDoubleChannel, default=54875.2422}
       Location.Y {type=MovieSceneScriptingDoubleChannel, default=50490.4219}
       Location.Z {type=MovieSceneScriptingDoubleChannel, default=1856.1077}
       Rotation.X {type=MovieSceneScriptingDoubleChannel, default=0}
       Rotation.Y {type=MovieSceneScriptingDoubleChannel, default=44.7832}
       Rotation.Z {type=MovieSceneScriptingDoubleChannel, default=89.8902}
       Scale.X {type=MovieSceneScriptingDoubleChannel, default=2, keys [[7386,1],[7782,2]]}
       Scale.Y {type=MovieSceneScriptingDoubleChannel, default=2, keys [[7386,1],[7782,2]]}
       Scale.Z {type=MovieSceneScriptingDoubleChannel, default=2, keys [[7386,1],[7782,2]]}
      ]
     }]
   }
   {
    type=MovieSceneParticleTrack
    display_name=Particle System
    sections [{
      type=MovieSceneParticleSection
      channels [None {type=MovieSceneScriptingParticleChannel, keys [[7396, "<ParticleKey.ACTIVATE: 0>"], [7474, "<ParticleKey.DEACTIVATE: 1>"]]}]
     }]
   }
  ]
 }
 P_Doggus_Trail {
  id=#92
  possessed_class=Emitter
  tracks [
   {
    type=MovieSceneSpawnTrack
    sections [{
      type=MovieSceneSpawnSection
      end_frame=6964
      start_frame=6480
      channels [None {type=MovieSceneScriptingBoolChannel, default=false, keys [[6480,false],[6486,true],[6964,false]]}]
     }]
   }
   {
    type=MovieScene3DTransformTrack
    sections [{
      type=MovieScene3DTransformSection
      end_frame=6964
      start_frame=6486
      channels [
       Location.X {type=MovieSceneScriptingDoubleChannel, default=0}
       Location.Y {type=MovieSceneScriptingDoubleChannel, default=0}
       Location.Z {type=MovieSceneScriptingDoubleChannel, default=0}
       Rotation.X {type=MovieSceneScriptingDoubleChannel, default=0}
       Rotation.Y {type=MovieSceneScriptingDoubleChannel, default=0}
       Rotation.Z {type=MovieSceneScriptingDoubleChannel, default=0}
       Scale.X {type=MovieSceneScriptingDoubleChannel, default=0.01}
       Scale.Y {type=MovieSceneScriptingDoubleChannel, default=0.01}
       Scale.Z {type=MovieSceneScriptingDoubleChannel, default=0.01}
      ]
     }]
   }
   {
    type=MovieSceneParticleTrack
    display_name=Particle System
    sections [{
      type=MovieSceneParticleSection
      channels [None {type=MovieSceneScriptingParticleChannel, keys [[6486, "<ParticleKey.ACTIVATE: 0>"], [6680, "<ParticleKey.DEACTIVATE: 1>"]]}]
     }]
   }
   {type=MovieScene3DAttachTrack, sections [{type=MovieScene3DAttachSection, end_frame=6964, start_frame=6486}]}
  ]
 }
 P_Doggus_Step_FX {
  id=#93
  possessed_class=Emitter
  tracks [
   {
    type=MovieSceneSpawnTrack
    sections [{
      type=MovieSceneSpawnSection
      end_frame=7278
      start_frame=6704
      channels [None {type=MovieSceneScriptingBoolChannel, default=false, keys [[6704,false],[6710,true],[7278,false]]}]
     }]
   }
   {
    type=MovieScene3DTransformTrack
    sections [{
      type=MovieScene3DTransformSection
      end_frame=7278
      start_frame=6704
      channels [
       Location.X {type=MovieSceneScriptingDoubleChannel, default=59819.9844}
       Location.Y {type=MovieSceneScriptingDoubleChannel, default=35852.0625}
       Location.Z {type=MovieSceneScriptingDoubleChannel, default=2572.147}
       Rotation.X {type=MovieSceneScriptingDoubleChannel, default=-3.2043}
       Rotation.Y {type=MovieSceneScriptingDoubleChannel, default=-1.0741}
       Rotation.Z {type=MovieSceneScriptingDoubleChannel, default=94973.1719}
       Scale.X {type=MovieSceneScriptingDoubleChannel, default=1}
       Scale.Y {type=MovieSceneScriptingDoubleChannel, default=1}
       Scale.Z {type=MovieSceneScriptingDoubleChannel, default=1}
      ]
     }]
   }
   {
    type=MovieSceneParticleTrack
    display_name=Particle System
    sections [{
      type=MovieSceneParticleSection
      channels [None {type=MovieSceneScriptingParticleChannel, keys [[6710, "<ParticleKey.ACTIVATE: 0>"], [6780, "<ParticleKey.DEACTIVATE: 1>"]]}]
     }]
   }
  ]
 }
 P Directional Dust {
  id=#94
  possessed_class=Emitter
  tracks [
   {
    type=MovieSceneSpawnTrack
    sections [{
      type=MovieSceneSpawnSection
      end_frame=8042
      start_frame=7386
      channels [None {type=MovieSceneScriptingBoolChannel, default=false, keys [[7386,false],[7388,true],[8042,false]]}]
     }]
   }
   {
    type=MovieScene3DTransformTrack
    sections [{
      type=MovieScene3DTransformSection
      end_frame=8042
      start_frame=7386
      channels [
       Location.X {type=MovieSceneScriptingDoubleChannel, default=54584.6602}
       Location.Y {type=MovieSceneScriptingDoubleChannel, default=44395.8594}
       Location.Z {type=MovieSceneScriptingDoubleChannel, default=2630.5786}
       Rotation.X {type=MovieSceneScriptingDoubleChannel, default=0}
       Rotation.Y {type=MovieSceneScriptingDoubleChannel, default=0}
       Rotation.Z {type=MovieSceneScriptingDoubleChannel, default=89.8902}
       Scale.X {type=MovieSceneScriptingDoubleChannel, default=2, keys [[7386,1],[7696,2]]}
       Scale.Y {type=MovieSceneScriptingDoubleChannel, default=2, keys [[7386,1],[7696,2]]}
       Scale.Z {type=MovieSceneScriptingDoubleChannel, default=2, keys [[7386,1],[7696,2]]}
      ]
     }]
   }
   {
    type=MovieSceneParticleTrack
    display_name=Particle System
    sections [{
      type=MovieSceneParticleSection
      channels [None {type=MovieSceneScriptingParticleChannel, keys [[7388, "<ParticleKey.ACTIVATE: 0>"], [7452, "<ParticleKey.DEACTIVATE: 1>"]]}]
     }]
   }
  ]
 }
 P Doggus Wing Engines L {
  id=#95
  possessed_class=Emitter
  tracks [
   {
    type=MovieSceneSpawnTrack
    sections [{
      type=MovieSceneSpawnSection
      end_frame=8442
      start_frame=8144
      channels [None {type=MovieSceneScriptingBoolChannel, default=false, keys [[8144,false],[8146,true],[8442,false]]}]
     }]
   }
   {
    type=MovieScene3DTransformTrack
    sections [{
      type=MovieScene3DTransformSection
      end_frame=8462
      start_frame=8146
      channels [
       Location.X {type=MovieSceneScriptingDoubleChannel, default=0}
       Location.Y {type=MovieSceneScriptingDoubleChannel, default=0}
       Location.Z {type=MovieSceneScriptingDoubleChannel, default=0}
       Rotation.X {type=MovieSceneScriptingDoubleChannel, default=0}
       Rotation.Y {type=MovieSceneScriptingDoubleChannel, default=0}
       Rotation.Z {type=MovieSceneScriptingDoubleChannel, default=0}
       Scale.X {type=MovieSceneScriptingDoubleChannel, default=1}
       Scale.Y {type=MovieSceneScriptingDoubleChannel, default=1}
       Scale.Z {type=MovieSceneScriptingDoubleChannel, default=1}
      ]
     }]
   }
   {
    type=MovieSceneParticleTrack
    display_name=Particle System
    sections [{
      type=MovieSceneParticleSection
      channels [None {type=MovieSceneScriptingParticleChannel, keys [[8146, "<ParticleKey.ACTIVATE: 0>"], [8416, "<ParticleKey.DEACTIVATE: 1>"]]}]
     }]
   }
   {type=MovieScene3DAttachTrack, sections [{type=MovieScene3DAttachSection, end_frame=8462, start_frame=8146}]}
   {
    type=MovieSceneParticleParameterTrack
    display_name=Particle Parameter
    sections [{
      type=MovieSceneParameterSection
      end_frame=8416
      start_frame=8146
      channels [
       WPO_Scale {type=MovieSceneScriptingFloatChannel, default=0, keys [[8146,-1],[8222,0]]}
       JetScale.X {type=MovieSceneScriptingFloatChannel, default=1, const [1,2]}
       JetScale.Y {type=MovieSceneScriptingFloatChannel, default=1.5, keys [[8146,1.5],[8222,1]]}
       JetScale.Z {type=MovieSceneScriptingFloatChannel, default=1, const [1,2]}
      ]
     }]
   }
  ]
 }
 P_Cattus_Rocket_Impact_2 {
  id=#96
  possessed_class=Emitter
  tracks [
   {
    type=MovieSceneSpawnTrack
    sections [{
      type=MovieSceneSpawnSection
      end_frame=6679
      start_frame=6320
      channels [None {type=MovieSceneScriptingBoolChannel, default=false, keys [[6320,false],[6322,true],[6679.13,false]]}]
     }]
   }
   {
    type=MovieSceneParticleTrack
    display_name=Particle System
    sections [{
      type=MovieSceneParticleSection
      channels [None {type=MovieSceneScriptingParticleChannel, keys [[6324, "<ParticleKey.TRIGGER: 2>"], [6679.13, "<ParticleKey.DEACTIVATE: 1>"]]}]
     }]
   }
  ]
 }
 P Directional Dust {
  id=#97
  possessed_class=Emitter
  tracks [
   {
    type=MovieSceneSpawnTrack
    sections [{
      type=MovieSceneSpawnSection
      end_frame=8042
      start_frame=7376
      channels [None {type=MovieSceneScriptingBoolChannel, default=false, keys [[7376,false],[7378,true],[8042,false]]}]
     }]
   }
   {
    type=MovieScene3DTransformTrack
    sections [{
      type=MovieScene3DTransformSection
      end_frame=8042
      start_frame=7376
      channels [
       Location.X {type=MovieSceneScriptingDoubleChannel, default=56319.3359}
       Location.Y {type=MovieSceneScriptingDoubleChannel, default=40296.1367}
       Location.Z {type=MovieSceneScriptingDoubleChannel, default=2630.5786}
       Rotation.X {type=MovieSceneScriptingDoubleChannel, default=0}
       Rotation.Y {type=MovieSceneScriptingDoubleChannel, default=0}
       Rotation.Z {type=MovieSceneScriptingDoubleChannel, default=89.8902}
       Scale.X {type=MovieSceneScriptingDoubleChannel, default=2, keys [[7378,1],[7680,2]]}
       Scale.Y {type=MovieSceneScriptingDoubleChannel, default=2, keys [[7378,1],[7680,2]]}
       Scale.Z {type=MovieSceneScriptingDoubleChannel, default=2, keys [[7378,1],[7680,2]]}
      ]
     }]
   }
   {
    type=MovieSceneParticleTrack
    display_name=Particle System
    sections [{
      type=MovieSceneParticleSection
      channels [None {type=MovieSceneScriptingParticleChannel, keys [[7378, "<ParticleKey.ACTIVATE: 0>"], [7444, "<ParticleKey.DEACTIVATE: 1>"]]}]
     }]
   }
  ]
 }
 P_Doggus_Trail {
  id=#98
  possessed_class=Emitter
  tracks [
   {
    type=MovieSceneSpawnTrack
    sections [{
      type=MovieSceneSpawnSection
      end_frame=6720
      start_frame=6234
      channels [None {type=MovieSceneScriptingBoolChannel, default=false, keys [[6234,false],[6240,true],[6720,false]]}]
     }]
   }
   {
    type=MovieScene3DTransformTrack
    sections [{
      type=MovieScene3DTransformSection
      end_frame=6720
      start_frame=6240
      channels [
       Location.X {type=MovieSceneScriptingDoubleChannel, default=0}
       Location.Y {type=MovieSceneScriptingDoubleChannel, default=0}
       Location.Z {type=MovieSceneScriptingDoubleChannel, default=0}
       Rotation.X {type=MovieSceneScriptingDoubleChannel, default=0}
       Rotation.Y {type=MovieSceneScriptingDoubleChannel, default=0}
       Rotation.Z {type=MovieSceneScriptingDoubleChannel, default=0}
       Scale.X {type=MovieSceneScriptingDoubleChannel, default=0.01}
       Scale.Y {type=MovieSceneScriptingDoubleChannel, default=0.01}
       Scale.Z {type=MovieSceneScriptingDoubleChannel, default=0.01}
      ]
     }]
   }
   {
    type=MovieSceneParticleTrack
    display_name=Particle System
    sections [{
      type=MovieSceneParticleSection
      channels [None {type=MovieSceneScriptingParticleChannel, keys [[6240, "<ParticleKey.ACTIVATE: 0>"], [6428, "<ParticleKey.DEACTIVATE: 1>"]]}]
     }]
   }
   {type=MovieScene3DAttachTrack, sections [{type=MovieScene3DAttachSection, end_frame=6720, start_frame=6240}]}
  ]
 }
 P_Doggus_Trail_03 {
  id=#99
  possessed_class=Emitter
  tracks [
   {
    type=MovieSceneSpawnTrack
    sections [{
      type=MovieSceneSpawnSection
      end_frame=5684
      start_frame=5186
      channels [None {type=MovieSceneScriptingBoolChannel, default=false, keys [[5186,false],[5194,true],[5684,false]]}]
     }]
   }
   {
    type=MovieScene3DTransformTrack
    sections [{
      type=MovieScene3DTransformSection
      end_frame=5684
      start_frame=5186
      channels [
       Location.X {type=MovieSceneScriptingDoubleChannel, default=0}
       Location.Y {type=MovieSceneScriptingDoubleChannel, default=0}
       Location.Z {type=MovieSceneScriptingDoubleChannel, default=0}
       Rotation.X {type=MovieSceneScriptingDoubleChannel, default=0}
       Rotation.Y {type=MovieSceneScriptingDoubleChannel, default=180}
       Rotation.Z {type=MovieSceneScriptingDoubleChannel, default=0}
       Scale.X {type=MovieSceneScriptingDoubleChannel, default=0.01}
       Scale.Y {type=MovieSceneScriptingDoubleChannel, default=0.01}
       Scale.Z {type=MovieSceneScriptingDoubleChannel, default=0.01}
      ]
     }]
   }
   {
    type=MovieSceneParticleTrack
    display_name=Particle System
    sections [{
      type=MovieSceneParticleSection
      channels [None {type=MovieSceneScriptingParticleChannel, keys [[5194, "<ParticleKey.ACTIVATE: 0>"], [5404, "<ParticleKey.DEACTIVATE: 1>"]]}]
     }]
   }
   {type=MovieScene3DAttachTrack, sections [{type=MovieScene3DAttachSection, end_frame=5684, start_frame=5194}]}
  ]
 }
 P_Doggus_Smoke_Simmer {
  id=#100
  possessed_class=Emitter
  tracks [
   {
    type=MovieSceneSpawnTrack
    sections [{
      type=MovieSceneSpawnSection
      end_frame=8378
      start_frame=6318
      channels [None {type=MovieSceneScriptingBoolChannel, default=false, keys [[6318,false],[6324,true],[8378.0025,false]]}]
     }]
   }
   {
    type=MovieSceneParticleTrack
    display_name=Particle System
    sections [{
      type=MovieSceneParticleSection
      channels [None {type=MovieSceneScriptingParticleChannel, keys [[6326, "<ParticleKey.ACTIVATE: 0>"], [7056, "<ParticleKey.DEACTIVATE: 1>"]]}]
     }]
   }
   {type=MovieScene3DAttachTrack, sections [{type=MovieScene3DAttachSection, end_frame=8378, start_frame=5928}]}
  ]
 }
 P_Doggus_Trail {
  id=#101
  possessed_class=Emitter
  tracks [
   {
    type=MovieSceneSpawnTrack
    sections [{
      type=MovieSceneSpawnSection
      end_frame=6726
      start_frame=6196
      channels [None {type=MovieSceneScriptingBoolChannel, default=false, keys [[6196,false],[6202,true],[6726,false]]}]
     }]
   }
   {
    type=MovieScene3DTransformTrack
    sections [{
      type=MovieScene3DTransformSection
      end_frame=6726
      start_frame=6202
      channels [
       Location.X {type=MovieSceneScriptingDoubleChannel, default=0}
       Location.Y {type=MovieSceneScriptingDoubleChannel, default=0}
       Location.Z {type=MovieSceneScriptingDoubleChannel, default=0}
       Rotation.X {type=MovieSceneScriptingDoubleChannel, default=0}
       Rotation.Y {type=MovieSceneScriptingDoubleChannel, default=180}
       Rotation.Z {type=MovieSceneScriptingDoubleChannel, default=0}
       Scale.X {type=MovieSceneScriptingDoubleChannel, default=0.01}
       Scale.Y {type=MovieSceneScriptingDoubleChannel, default=0.01}
       Scale.Z {type=MovieSceneScriptingDoubleChannel, default=0.01}
      ]
     }]
   }
   {
    type=MovieSceneParticleTrack
    display_name=Particle System
    sections [{
      type=MovieSceneParticleSection
      channels [None {type=MovieSceneScriptingParticleChannel, keys [[6202, "<ParticleKey.ACTIVATE: 0>"], [6404, "<ParticleKey.DEACTIVATE: 1>"]]}]
     }]
   }
   {type=MovieScene3DAttachTrack, sections [{type=MovieScene3DAttachSection, end_frame=6726, start_frame=6202}]}
  ]
 }
 P_Doggus_Step_FX {
  id=#102
  possessed_class=Emitter
  tracks [
   {
    type=MovieSceneSpawnTrack
    sections [{
      type=MovieSceneSpawnSection
      end_frame=7440
      start_frame=7130
      channels [None {type=MovieSceneScriptingBoolChannel, default=false, keys [[7130,false],[7136,true],[7440,false]]}]
     }]
   }
   {
    type=MovieScene3DTransformTrack
    sections [{
      type=MovieScene3DTransformSection
      end_frame=7440
      start_frame=7130
      channels [
       Location.X {type=MovieSceneScriptingDoubleChannel, default=57695.2812}
       Location.Y {type=MovieSceneScriptingDoubleChannel, default=29872.4082}
       Location.Z {type=MovieSceneScriptingDoubleChannel, default=2515.5227}
       Rotation.X {type=MovieSceneScriptingDoubleChannel, default=0.603}
       Rotation.Y {type=MovieSceneScriptingDoubleChannel, default=-0.9625}
       Rotation.Z {type=MovieSceneScriptingDoubleChannel, default=1140400.125}
       Scale.X {type=MovieSceneScriptingDoubleChannel, default=1}
       Scale.Y {type=MovieSceneScriptingDoubleChannel, default=1}
       Scale.Z {type=MovieSceneScriptingDoubleChannel, default=1}
      ]
     }]
   }
   {
    type=MovieSceneParticleTrack
    display_name=Particle System
    sections [{
      type=MovieSceneParticleSection
      channels [None {type=MovieSceneScriptingParticleChannel, keys [[7136, "<ParticleKey.ACTIVATE: 0>"], [7202, "<ParticleKey.DEACTIVATE: 1>"]]}]
     }]
   }
  ]
 }
 P Directional Dust {
  id=#103
  possessed_class=Emitter
  tracks [
   {
    type=MovieSceneSpawnTrack
    sections [{
      type=MovieSceneSpawnSection
      end_frame=8393
      start_frame=8185
      channels [None {type=MovieSceneScriptingBoolChannel, default=false, keys [[8185,false],[8187,true],[8393,false]]}]
     }]
   }
   {
    type=MovieScene3DTransformTrack
    sections [{
      type=MovieScene3DTransformSection
      end_frame=8397
      start_frame=8185
      channels [
       Location.X {type=MovieSceneScriptingDoubleChannel, default=66990.9766}
       Location.Y {type=MovieSceneScriptingDoubleChannel, default=-32963.8711}
       Location.Z {type=MovieSceneScriptingDoubleChannel, default=-609.9226}
       Rotation.X {type=MovieSceneScriptingDoubleChannel, default=0}
       Rotation.Y {type=MovieSceneScriptingDoubleChannel, default=0}
       Rotation.Z {type=MovieSceneScriptingDoubleChannel, default=94.7951}
       Scale.X {type=MovieSceneScriptingDoubleChannel, default=1.1}
       Scale.Y {type=MovieSceneScriptingDoubleChannel, default=1.1}
       Scale.Z {type=MovieSceneScriptingDoubleChannel, default=1.1}
      ]
     }]
   }
   {
    type=MovieSceneParticleTrack
    display_name=Particle System
    sections [{
      type=MovieSceneParticleSection
      channels [None {type=MovieSceneScriptingParticleChannel, keys [[8187, "<ParticleKey.ACTIVATE: 0>"], [8258, "<ParticleKey.DEACTIVATE: 1>"]]}]
     }]
   }
   {
    type=MovieSceneParticleParameterTrack
    display_name=Particle Parameter
    sections [{
      type=MovieSceneParameterSection
      end_frame=8404
      start_frame=8187
      channels [
       ColorMultiplier.X {type=MovieSceneScriptingFloatChannel, default=0.1, const [0.1,1]}
       ColorMultiplier.Y {type=MovieSceneScriptingFloatChannel, default=0.085, const [0.085,1]}
       ColorMultiplier.Z {type=MovieSceneScriptingFloatChannel, default=0.0821, const [0.0821,1]}
      ]
     }]
   }
  ]
 }
]
display_rate {denominator=1, numerator=60}
tick_resolution {denominator=1, numerator=24000}
tracks []
