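# Translucent unlit particle-sprite material in the Cattus prototype FX that blurs the scene behind the sprite.
# SpiralBlur-SceneTexture supplies emissive, with a depth fade as its distance mask.
# Opacity combines the blurred result's boosted blue channel, the function's clamped scene output and a circular UV mask, while world position offset lifts sprites along Z.
# A feature-level switch zeroes emissive and opacity on ES2 and ES3.1; several camera-offset and math nodes are unused.
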Material /Figment_S10/Athena/Prototype/Blueprints/Cattus/FX/M_BlurMat
graph {
 properties {
  bUseMaterialAttributes=true
  bUseTranslucencyVertexFog=false
  bUsedWithParticleSprites=true
  blendMode=BLEND_Translucent
  pixelDepthOffsetMode=PDOM_Legacy
  shadingModel=MSM_Unlit
  translucencyPass=MTP_BeforeDOF
  usedShadingModels=Unlit
 }
 nodes [
  n1 = ScalarParameter() {defaultValue=300, parameterName=Offset}
  n2 = Constant3Vector() {constant=<rgba 0 0 200 0>}
  Add(
   A=MaterialFunctionCall(Offset Amount=@n1) {materialFunction=/Engine/Functions/Engine_MaterialFunctions02/WorldPositionOffset/CameraOffset}.Clamped Camera Offset
   B=@n2
  )
  n3 = Constant() {r=4}
  n4 = DepthFade(FadeDistance=@n1)
  n5 = MaterialFunctionCall!(
   Distance=Constant() {r=0.0025}
   Distance Steps=@n3
   Radial Steps=@n3
   Radial Offset
   TempAA Radial Blur
   TempAA Distance Blur
   Distance Mask=@n4
  ) {materialFunction=/Game/Effects/Fort_Effects/Materials/_Functions/SpiralBlur-SceneTexture}
  n6 = Constant() {r=0.5}
  n7 = SphereMask(A=TextureCoordinate(), B=@n6, Radius=@n6) {hardnessPercent=90}
  Add(A=@n5.SceneColor clamp to 0, B=@n7)
  Constant() {r=300}
  n8 = MaterialFunctionCall(Float2=ViewSize()) {materialFunction=&BreakOutFloat2Components}
  Divide(B=Max(A=@n8.R, B=@n8.R)) {constA=4}
  n9 = Multiply(A=ComponentMask(None=@n5.Result) {b=true}) {constB=128}
  n10 = SetMaterialAttributes!(
   MaterialAttributes
   Emissive Color=@n5.Result
   Opacity=Multiply(A=Saturate(None=@n9), B=Multiply(A=@n5.SceneColor clamp to 0, B=@n7))
   World Position Offset=@n2
  ) {attributeSetTypes [#1, #2, #3]}
  n11 = Constant()
  n12 = SetMaterialAttributes(Emissive Color=@n11, Opacity=@n11) {attributeSetTypes [#1, #2, #3]}
  FeatureLevelSwitch(Default=@n10, ES2=@n12, ES3_1=@n12, SM4_REMOVED=@n10)
  Multiply(A=@n4, B=@n7)
 ]
}
