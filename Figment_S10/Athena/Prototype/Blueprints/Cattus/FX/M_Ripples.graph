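# Translucent, unlit, refractive mesh-particle material for rippling effects, tinted by a color parameter and time-of-day function.
# Panned noise distorts two world-XY electric-noise samples, masked by a water ripple texture and UV falloff to drive opacity and refraction, with depth fade and particle alpha.
# Dynamic parameters scale noise-driven vertical and local-Y vertex offsets.
# Feature-level switches drop depth fade on mobile and give ES2 a duplicate branch without vertex offset.
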Material /Figment_S10/Athena/Prototype/Blueprints/Cattus/FX/M_Ripples
graph {
 properties {
  bUseMaterialAttributes=true
  bUsedWithMeshParticles=true
  blendMode=BLEND_Translucent
  pixelDepthOffsetMode=PDOM_Legacy
  refractionMethod=RM_IndexOfRefraction
  shadingModel=MSM_Unlit
  translucencyPass=MTP_BeforeDOF
  usedShadingModels=Unlit
 }
 nodes [
  n1 = ComponentMask(None=WorldPosition().XYZ) {g=true, r=true}
  n2 = Divide(A=@n1) {constB=2048}
  n3 = TextureSample!(UVs=@n2, Tex, Level, Apply View MipBias) {
   mipValueMode=TMVM_MipLevel
   samplerType=SAMPLERTYPE_LinearColor
   texture=/Game/Effects/Fort_Effects/Gameplay/SmokeDustExplosions/Textures/T_NoisePack_LinearColor
  }
  n4 = Multiply(A=@n3.R, B=Constant3Vector() {constant=<rgba 0 0 1 0>})
  n5 = DynamicParameter() {paramNames [Offset,OffsetLocal,Param3,Param4]}
  n6 = TextureCoordinate() {uTiling=2, vTiling=2}
  n7 = Multiply(A=Transform(None=Constant3Vector() {constant=<rgba 0 1 0 0>}) {transformType=TRANSFORM_Local}, B=ComponentMask(None=@n6) {g=true})
  n8 = DepthFade(FadeDistance=ScalarParameter() {defaultValue=500, parameterName=DF_FD})
  n9 = Constant() {r=1}
  n10 = TextureSample(UVs=Panner(Coordinate=TextureCoordinate() {uTiling=0.25, vTiling=0.25}) {speedX=0.1, speedY=0.05}) {
   mipValueMode=TMVM_MipLevel
   samplerType=SAMPLERTYPE_LinearColor
   texture=&T_NoisePack_LinearColor
  }
  n11 = Add(A=Subtract(A=ComponentMask(None=@n10.RGB) {g=true, r=true}) {constB=0.5}, B=Multiply(A=@n3.B) {constB=0.25})
  n12 = Add(A=@n2, B=@n11)
  n13 = Divide(A=@n1) {constB=4096}
  n14 = Add(A=@n13, B=@n11)
  n15 = OneMinus(None=@n6)
  n16 = MaterialFunctionCall(Float2=OneMinus(None=Abs(None=@n15))) {materialFunction=&BreakOutFloat2Components}
  n17 = Multiply(
   A=Add(
    A=TextureSample!(UVs=@n12, Tex, Apply View MipBias) {samplerType=SAMPLERTYPE_LinearColor, texture=/Game/Characters/Player/Male/Medium/Bodies/M_MED_SRDrift/Textures/T_ElectricNoise}.R
    B=TextureSample(UVs=@n14) {samplerType=SAMPLERTYPE_LinearColor, texture=&T_ElectricNoise}.R
   )
   B=Multiply(
    A=TextureSample() {samplerType=SAMPLERTYPE_LinearColor, texture=/Game/Effects/Fort_Effects/Effects/Water/Textures/WaterDecalRippleMask}.R
    B=Multiply(A=@n16.R, B=@n16.R)
   )
  )
  n18 = Multiply(A=Multiply(A=FeatureLevelSwitch(Default=@n8, ES2=@n9, ES3_1=@n9, SM4_REMOVED=@n8), B=SquareRoot(None=@n17)), B=ParticleColor().A)
  n19 = LinearInterpolate(Alpha=@n3.G) {constA=0.1}
  n20 = TextureSample!(UVs=Divide(A=@n13) {constB=4}, Tex, Level, Apply View MipBias) {mipValueMode=TMVM_MipLevel, texture=/Figment_S10/Effects/Fort_Effects/Textures/Misc/T_Tiling_Noise_01}
  n21 = MakeMaterialAttributes!(
   BaseColor
   Metallic
   Specular
   Roughness
   Anisotropy
   EmissiveColor=Multiply(
    A=MaterialFunctionCall() {materialFunction=&Color_Time_of_Day}.Result
    B=VectorParameter() {parameterName=Color, defaultValue=<rgba 0.2656 0.8828 1.5 0>}.RGB
   )
   Opacity=Subtract(A=Multiply(A=@n18, B=@n19), B=Multiply(A=@n20.R) {constB=0.05})
   OpacityMask
   Normal
   Tangent
   WorldPositionOffset=Add(A=Multiply(A=@n4, B=@n5.Offset), B=Multiply(A=@n7, B=@n5.OffsetLocal) {constB=6})
   SubsurfaceColor
   ClearCoat
   ClearCoatRoughness
   AmbientOcclusion
   Refraction=LinearInterpolate(Alpha=@n18) {constA=1, constB=1.08}
   CustomizedUVs_0
   CustomizedUVs_1
   CustomizedUVs_2
   CustomizedUVs_3
   CustomizedUVs_4
   CustomizedUVs_5
   CustomizedUVs_6
   CustomizedUVs_7
   PixelDepthOffset
   ShadingModel
   Displacement
  )
  n22 = DepthFade(FadeDistance=ScalarParameter() {defaultValue=500, parameterName=DF_FD})
  n23 = Constant() {r=1}
  n24 = ComponentMask(None=WorldPosition().XYZ) {g=true, r=true}
  n25 = Divide(A=@n24) {constB=2048}
  n26 = TextureSample!(UVs=Panner(Coordinate=TextureCoordinate() {uTiling=0.25, vTiling=0.25}) {speedX=0.1, speedY=0.05}, Tex, Apply View MipBias) {samplerType=SAMPLERTYPE_LinearColor, texture=&T_NoisePack_LinearColor}
  n27 = TextureSample(UVs=@n25) {samplerType=SAMPLERTYPE_LinearColor, texture=&T_NoisePack_LinearColor}
  n28 = Add(A=Subtract(A=ComponentMask(None=@n26.RGB) {g=true, r=true}) {constB=0.5}, B=Multiply(A=@n27.B) {constB=0.25})
  n29 = Add(A=@n25, B=@n28)
  n30 = Divide(A=@n24) {constB=4096}
  n31 = Add(A=@n30, B=@n28)
  n32 = TextureCoordinate() {uTiling=2, vTiling=2}
  n33 = OneMinus(None=@n32)
  n34 = MaterialFunctionCall(Float2=OneMinus(None=Abs(None=@n33))) {materialFunction=&BreakOutFloat2Components}
  n35 = Multiply(
   A=Add(
    A=TextureSample(UVs=@n29) {samplerType=SAMPLERTYPE_LinearGrayscale, texture=&T_ElectricNoise}.R
    B=TextureSample(UVs=@n31) {samplerType=SAMPLERTYPE_LinearGrayscale, texture=&T_ElectricNoise}.R
   )
   B=Multiply(
    A=TextureSample() {samplerType=SAMPLERTYPE_LinearColor, texture=&WaterDecalRippleMask}.R
    B=Multiply(A=@n34.R, B=@n34.R)
   )
  )
  n36 = Multiply(A=Multiply(A=FeatureLevelSwitch(Default=@n22, ES2=@n23, ES3_1=@n23, SM4_REMOVED=@n22), B=SquareRoot(None=@n35)), B=ParticleColor().A)
  n37 = Multiply(A=@n36, B=LinearInterpolate(Alpha=@n27.G) {constA=0.1})
  n38 = Multiply(A=TextureSample(UVs=Divide(A=@n30) {constB=4}) {texture=&T_Tiling_Noise_01}.R) {constB=0.05}
  n39 = MaterialFunctionCall() {materialFunction=&Color_Time_of_Day}
  n40 = VectorParameter() {parameterName=Color, defaultValue=<rgba 0.2656 0.8828 1.5 0>}
  FeatureLevelSwitch(
   Default=@n21
   ES2=MakeMaterialAttributes(
    EmissiveColor=Multiply(A=@n39.Result, B=@n40.RGB)
    Opacity=Subtract(A=@n37, B=@n38)
    Refraction=LinearInterpolate(Alpha=@n36) {constA=1, constB=1.08}
   )
   SM4_REMOVED=@n21
  )
  n41 = Multiply(A=@n27.R, B=Constant3Vector() {constant=<rgba 0 0 1 0>})
  n42 = DynamicParameter() {paramNames [Offset,OffsetLocal,Param3,Param4]}
  n43 = Multiply(A=Transform(None=Constant3Vector() {constant=<rgba 0 1 0 0>}) {transformType=TRANSFORM_Local}, B=ComponentMask(None=@n32) {g=true})
  Reroute(None=Add(A=Multiply(A=@n41, B=@n42.Offset), B=Multiply(A=@n43, B=@n42.OffsetLocal) {constB=6}))
 ]
}
