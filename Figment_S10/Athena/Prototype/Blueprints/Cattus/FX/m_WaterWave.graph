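# Masked, lit water wave material for mesh particles in the Cattus prototype, colored by a curve atlas along UV height plus panning world-space noise.
# Opacity combines noise-varied height above the actor, particle alpha and temporal screen-space dither, also driving pixel depth offset.
# World position offset adds dynamic-parameter Z lift and pulls red-painted vertices toward the object center by noise.
# A feature-level switch gives mobile a fully opaque mask without depth offset.
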Material /Figment_S10/Athena/Prototype/Blueprints/Cattus/FX/m_WaterWave
graph {
 properties {
  bUseMaterialAttributes=true
  bUsedWithMeshParticles=true
  blendMode=BLEND_Masked
  pixelDepthOffsetMode=PDOM_Legacy
  usedShadingModels=Default Lit
 }
 nodes [
  n1 = TextureCoordinate()
  n2 = Divide(A=WorldPosition().XYZ) {constB=2048}
  n3 = Add(A=@n2, B=Multiply(A=Time()) {constB=0.25})
  n4 = TextureCoordinate() {uTiling=0.5, vTiling=0.5}
  n5 = TextureSample!(
   UVs=Add(
    A=ComponentMask(None=@n3) {g=true, r=true}
    B=TextureSample(UVs=@n4) {mipValueMode=TMVM_MipLevel, texture=/Game/Effects/Fort_Effects/Gameplay/SmokeDustExplosions/Textures/T_NoisePack_01}.R
   )
   Tex
   Level
   Apply View MipBias
  ) {mipValueMode=TMVM_MipLevel, texture=&T_NoisePack_01}
  n6 = CurveAtlasRowParameter(CurveTime=Add(A=ComponentMask(None=@n1) {g=true}, B=Multiply(A=@n5.G) {constB=0.25})) {atlas=/Game/Athena/LinearColorCurves/TempAthenaCurveAtlas, curve=/Game/Athena/Prototype/Blueprints/Cattus/FX/Curve_Color, defaultValue=3}
  n7 = Constant() {r=0.25}
  n8 = Constant()
  n9 = Subtract(A=WorldPosition().XYZ, B=ActorPositionWS())
  n10 = Saturate(None=Divide(A=ComponentMask(None=@n9) {b=true}, B=LinearInterpolate(Alpha=@n5.R) {constA=64, constB=256}) {constB=512})
  n11 = ParticleColor()
  n12 = Custom!(None) {
   code=View.TemporalAAParams
   desc="\n*** Original source before expression upgrade ***\nView.TemporalAAParams\n*** Original source before expression upgrade ***\nFrame.TemporalAAParams"
   description=TemporalAA SampleID
   outputType=CMOT_Float2
  }
  MaterialExpressionAdd_2 = Add(
   A=MaterialFunctionCall!(TextureResolution=Constant2Vector() {g=1, r=1}) {materialFunction=/Engine/Functions/Engine_MaterialFunctions02/Texturing/ScreenAlignedPixelToPixelUVs}.Result
   B=ComponentMask(None=@n12) {r=true}
  )
  n13 = Custom!(p=@MaterialExpressionAdd_2) {
   code="((uint)(p.x) + 2 * (uint)(p.y)) % 5"
   outputType=CMOT_Float1
   inputs [{inputName=p, input {expression=&self:MaterialExpressionAdd_2}}]
  }
  n14 = TextureSample(UVs=MaterialFunctionCall(TextureResolution=Constant2Vector() {g=64, r=64}) {materialFunction=&ScreenAlignedPixelToPixelUVs}.Result) {mipValueMode=TMVM_MipLevel, samplerType=SAMPLERTYPE_LinearColor, texture=/Game/Effects/Fort_Effects/Textures/Tile/LinerNoise64x64}
  n15 = Add(A=@n13, B=@n14.R) {constB=0}
  n16 = Add(A=Add(A=Add(A=@n10, B=@n11.A), B=Divide(A=@n15) {constB=6}) {constB=0.5}) {constB=-0.5}
  n17 = VertexColor()
  n18 = TextureSample(UVs=ComponentMask(None=@n2) {g=true, r=true}) {mipValueMode=TMVM_MipLevel, texture=&T_NoisePack_01}
  n19 = Subtract(A=WorldPosition().XYZ, B=ObjectPositionWS())
  n20 = Add(
   A=Multiply(A=DynamicParameter() {paramNames [Offset,Param2,Param3,Param4]}.Offset, B=Constant3Vector() {constant=<rgba 0 0 1 0>})
   B=Multiply(A=Multiply(A=@n17.R, B=@n18.B), B=Multiply(A=@n19) {constB=-0.5})
  )
  n21 = MakeMaterialAttributes!(
   BaseColor=@n6
   Metallic
   Specular=@n7
   Roughness=@n8
   Anisotropy
   EmissiveColor
   Opacity
   OpacityMask=@n16
   Normal
   Tangent
   WorldPositionOffset=@n20
   SubsurfaceColor
   ClearCoat
   ClearCoatRoughness
   AmbientOcclusion
   Refraction
   CustomizedUVs_0
   CustomizedUVs_1
   CustomizedUVs_2
   CustomizedUVs_3
   CustomizedUVs_4
   CustomizedUVs_5
   CustomizedUVs_6
   CustomizedUVs_7
   PixelDepthOffset=Multiply(A=@n16) {constB=512}
   ShadingModel
   Displacement
  )
  n22 = TextureCoordinate()
  n23 = Add(A=Divide(A=WorldPosition().XYZ) {constB=2048}, B=Multiply(A=Time()) {constB=0.25})
  n24 = TextureCoordinate() {uTiling=0.5, vTiling=0.5}
  n25 = TextureSample!(
   UVs=Add(
    A=ComponentMask(None=@n23) {g=true, r=true}
    B=TextureSample(UVs=@n24) {texture=&T_NoisePack_01}.R
   )
   Tex
   Apply View MipBias
  ) {texture=&T_NoisePack_01}
  n26 = CurveAtlasRowParameter(CurveTime=Add(A=ComponentMask(None=@n22) {g=true}, B=Multiply(A=@n25.G) {constB=0.25})) {
   atlas=&TempAthenaCurveAtlas
   curve=&Curve_Color
   defaultValue=0.0137
  }
  n27 = MakeMaterialAttributes(BaseColor=@n26, Specular=@n7, Roughness=@n8, OpacityMask=Constant() {r=1}, WorldPositionOffset=@n20)
  FeatureLevelSwitch(Default=@n21, ES2=@n27, ES3_1=@n27, SM4_REMOVED=@n21)
 ]
}
