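; Prototype gameplay ability that knocks its avatar player away from an overlapping Cattus/Doggus skeletal mesh actor.
; Activation runs KnockMe, which reschedules itself by timer while overlap lasts and otherwise calls RemoveGrantedbyEffect.
; Each pass launches the pawn by the mesh's physics velocity at the closest physics-asset point plus a normal-scaled push, clamped in size.
; A dot product test doubles the pre-clamp launch when the player is behind that surface.
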(event Ability|EventOnEndAbility
  (params (bWasCancelled :type Boolean)))

(event Custom|KnockMe
  (bind _baseknockbackstrength (Variables|Default|GetBaseKnockbackStrength))
  (bind _knockedplayer (Variables|Default|GetKnockedPlayer))
  (bind _returnvalue (Transformation|GetActorLocation _knockedplayer))
  (bind _mesh (Variables|Default|GetMesh))
  (bind _skeletalmeshcomponent (Variables|/Script/Engine.SkeletalMeshActor|GetSkeletalMeshComponent _mesh))
  (bind (_closestworldposition _normal _bonename) (Components|SkeletalMesh|GetClosestPointOnPhysicsAsset _skeletalmeshcomponent _returnvalue))
  (bind _returnvalue_2 (* _normal _baseknockbackstrength))
  (bind _knockedplayer_2 (Variables|Default|GetKnockedPlayer))
  (bind _overlappingactors (Collision|GetOverlappingActors _knockedplayer_2 "/Figment_S10/Athena/Prototype/Blueprints/Cattus/BP_CattusDoggus_SkelMesh_gameplay.BP_CattusDoggus_SkelMesh_gameplay_C" :!class /Script/Engine.Actor))
  (bind _knockedplayer_3 (Variables|Default|GetKnockedPlayer))
  (bind _maxknock (Variables|Default|GetMaxKnock))
  (bind _minknock (Variables|Default|GetMinKnock))
  (if (> (Utilities|Array|Length _overlappingactors) 0)
    (bind _asbp_cattus_doggus_skel_mesh_gameplay (Utilities|Casting|CastToBP_CattusDoggus_SkelMesh_gameplay (Utilities|Array|Get(acopy) _overlappingactors)))
    (Variables|Default|SetMesh _asbp_cattus_doggus_skel_mesh_gameplay)
    (Utilities|FlowControl|Sequence
      (:then_0
        (bind _returnvalue_3 (Physics|GetPhysicsLinearVelocityatPoint _skeletalmeshcomponent _closestworldposition _bonename))
        (if (> (Math|Vector|DotProduct _normal (- (Transformation|GetActorLocation _knockedplayer_3) _closestworldposition)) 0.0)
          (bind _asfort_pawn (Utilities|Casting|CastToFortPawn _knockedplayer_3))
          (Pawn|Character|LaunchCharacterJump _asfort_pawn (Math|Vector|ClampVectorSize (+ _returnvalue_3 _returnvalue_2) _minknock _maxknock) true true true true)
          (else
            (bind _asfort_pawn_2 (Utilities|Casting|CastToFortPawn _knockedplayer_3))
            (Pawn|Character|LaunchCharacterJump _asfort_pawn_2 (Math|Vector|ClampVectorSize (* (+ _returnvalue_3 _returnvalue_2) 2.0) _minknock _maxknock) true true true true))))
      (:then_1
        (Utilities|Time|SetTimerbyEvent (EventDispatchers|CreateEvent :!event Custom|KnockMe) (Variables|Default|GetLoopTime))))
    (else
      (Ability|RemoveGrantedbyEffect))))

(event Ability|EventActivateAbility
  (bind _returnvalue (Ability|GetAvatarActorfromActorInfo))
  (Utilities|IsValid _returnvalue
    (:Is%20Valid
      (Development|PrintString "VALID PLAYER")
      (Variables|Default|SetKnockedPlayer _returnvalue)
      (Utilities|Time|SetTimerbyEvent (EventDispatchers|CreateEvent :!event Custom|KnockMe) (Variables|Default|GetLoopTime))
      (CallFunction|KnockMe))
    (:Is%20Not%20Valid
      (Development|PrintString (Utilities|String|Append "INVALID TARGET" (Utilities|GetDisplayName _returnvalue))))))
