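; Prototype Blueprint actor for the Cattus/Doggus giant-creature sequence, providing custom events for beats such as footsteps, lasers, impacts, flight and sword draw.
; Each event plays local camera shakes, some delayed, plus in-range force feedback on non-dedicated-server instances.
; CameraShakeScaleFromDistance returns the local controller, a power-curve falloff from view-target distance and an in-range flag.
; Footsteps also launch nearby walking pawns upward with squared falloff, on both server and local client.
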(event Custom|TestFootSTep
  ;[1 1 1] Cheat commands - cheat ke* <function name>
  (CallFunction|ReplicateFootStep))

(event Custom|ReplicateFootStep
  (bind _returnvalue (Transformation|GetActorLocation))
  (bind _returnvalue_2 (Transformation|GetActorLocation))
  (bind _knockup_maxdistance (Variables|Default|GetKnockUp_MaxDistance))
  (bind _knockup_maxdistance_2 (Variables|Default|GetKnockUp_MaxDistance))
  ;[1 1 1] Knock players up on big guy walk
  (if (Networking|IsDedicatedServer)
    ;;[1 1 1] Handle server side calculations and movement
    (bind _outfortplayerpawns (Fortnite|GetFortPlayerPawns))
    (Development|PrintString (Utilities|String|Append (Variables|Default|GetDebugTextPrefix) "Players In match: " (Utilities|String|ToString(Integer) (Utilities|Array|Length (Variables|/Script/Engine.GameStateBase|GetPlayerArray (Game|GetGameState))))) (Variables|Default|GetbShouldDebugText) (Variables|Default|GetbShouldDebugText))
    (for _array_element _outfortplayerpawns
      (if (<= (Math|Vector|Distance(Vector) (Transformation|GetActorLocation _array_element) _returnvalue) _knockup_maxdistance_2)
        (if (Pawn|Components|CharacterMovement|IsWalking (Variables|/Script/Engine.Character|GetCharacterMovement _array_element))
          (Pawn|Character|LaunchCharacterJump _array_element (Math|Vector|MakeVector 0.0 0.0 (* (Math|Float|Power (Math|Float|MapRangeClamped (Math|Vector|Distance(Vector) (Transformation|GetActorLocation _array_element) _returnvalue) 0.0 _knockup_maxdistance_2 1.0 (Variables|Default|GetMinLaunch)) 2.0) (Variables|Default|GetKnockUp_BaseStrength))) false false true true))))
    (else
      ;;[1 1 1] Footsteps local player camera shake
      (bind (_output _scale _distanceff) (CallFunction|CameraShakeScalefromDistance (Variables|Default|GetFootStepCameraShakeDistance)))
      (Game|Feedback|ClientStartCameraShake :self _output :Shake "/Figment_S10/Athena/Prototype/Blueprints/Cattus/CameraShake/CameraShake_DoggusCattus_FootStep.CameraShake_DoggusCattus_FootStep_C" :Scale _scale :PlaySpace "CameraLocal")
      (if _distanceff
        (Game|Feedback|ClientPlayForceFeedback _output "/Figment_S10/Athena/Prototype/Blueprints/Cattus/ForceFeedback/FF_DoggusCattus_Footsteps.FF_DoggusCattus_Footsteps" "None" false false false)
        (else
          _))
      (bind _asfort_player_pawn (Utilities|Casting|CastToFortPlayerPawn (Pawn|GetControlledPawn _output)))
      (if (<= (Math|Vector|Distance(Vector) (Transformation|GetActorLocation (Pawn|GetControlledPawn _output)) _returnvalue_2) _knockup_maxdistance)
        (if (Pawn|Components|CharacterMovement|IsWalking (Variables|/Script/Engine.Character|GetCharacterMovement _asfort_player_pawn))
          (Pawn|Character|LaunchCharacterJump _asfort_player_pawn (Math|Vector|MakeVector 0.0 0.0 (* (Math|Float|Power (Math|Float|MapRangeClamped (Math|Vector|Distance(Vector) (Transformation|GetActorLocation (Pawn|GetControlledPawn _output)) _returnvalue_2) 0.0 _knockup_maxdistance 1.0 (Variables|Default|GetMinLaunch)) 2.0) (Variables|Default|GetKnockUp_BaseStrength))) false false true true))))))

(event Custom|IntenseScream
  ;[1 1 1] Intense Scream local player shake and feedback
  (if (not (Networking|IsDedicatedServer))
    (bind (_output _scale _distanceff) (CallFunction|CameraShakeScalefromDistance (Variables|Default|GetScreamCameraShakeDistance)))
    (Game|Feedback|ClientStartCameraShake :self _output :Shake "/Figment_S10/Athena/Prototype/Blueprints/Cattus/CameraShake/CameraShake_DoggusCattus_Scream.CameraShake_DoggusCattus_Scream_C" :Scale _scale :PlaySpace "CameraLocal")
    (if _distanceff
      (Game|Feedback|ClientPlayForceFeedback _output "/Figment_S10/Athena/Prototype/Blueprints/Cattus/ForceFeedback/FF_DoggusCattus_Scream.FF_DoggusCattus_Scream" "None" false false false))))

(event Custom|BigAssLaser
  ;[1 1 1] Laser local player shake and feedback
  (if (not (Networking|IsDedicatedServer))
    (bind (_output _scale _distanceff) (CallFunction|CameraShakeScalefromDistance (Variables|Default|GetLaserCameraShakeDistance)))
    (Game|Feedback|ClientStartCameraShake :self _output :Shake "/Figment_S10/Athena/Prototype/Blueprints/Cattus/CameraShake/CameraShake_DoggusCattus_LaserStart.CameraShake_DoggusCattus_LaserStart_C" :Scale _scale :PlaySpace "CameraLocal")
    (if _distanceff
      (Game|Feedback|ClientPlayForceFeedback _output "/Figment_S10/Athena/Prototype/Blueprints/Cattus/ForceFeedback/FF_DoggusCattus_ChargeUpLaser.FF_DoggusCattus_ChargeUpLaser" "None" false false false)
      (else
        _))
    (Utilities|FlowControl|Delay 2.0)
    (Game|Feedback|ClientStartCameraShake :self _output :Shake "/Figment_S10/Athena/Prototype/Blueprints/Cattus/CameraShake/CameraShake_DoggusCattus_Laser.CameraShake_DoggusCattus_Laser_C" :Scale _scale :PlaySpace "CameraLocal")))

(event Custom|MassiveImpact
  ;[1 1 1] Massive Impact local player shake and feedback
  (if (not (Networking|IsDedicatedServer))
    (bind (_output _scale _distanceff) (CallFunction|CameraShakeScalefromDistance (Variables|Default|GetMassiveImpactCameraShakeDistance)))
    (Game|Feedback|ClientStartCameraShake :self _output :Shake "/Figment_S10/Athena/Prototype/Blueprints/Cattus/CameraShake/CameraShake_DoggusCattus_MassiveStart.CameraShake_DoggusCattus_MassiveStart_C" :Scale _scale :PlaySpace "CameraLocal")
    (Game|Feedback|ClientStartCameraShake :self _output :Shake "/Figment_S10/Athena/Prototype/Blueprints/Cattus/CameraShake/CameraShake_DoggusCattus_MassiveImpact.CameraShake_DoggusCattus_MassiveImpact_C" :Scale _scale :PlaySpace "CameraLocal")
    (if _distanceff
      (Game|Feedback|ClientPlayForceFeedback _output "/Figment_S10/Athena/Prototype/Blueprints/Cattus/ForceFeedback/FF_DoggusCattus_MassiveImpact.FF_DoggusCattus_MassiveImpact" "None" false false false))))

(event Custom|Missle%20Barrage
  ;[1 1 1] Missle Barrage local player shake and feedback
  (if (not (Networking|IsDedicatedServer))
    (bind (_output _scale _distanceff) (CallFunction|CameraShakeScalefromDistance (Variables|Default|GetMissleBarrageCameraShakeDistance)))
    (Game|Feedback|ClientStartCameraShake :self _output :Shake "/Figment_S10/Athena/Prototype/Blueprints/Cattus/CameraShake/CameraShake_DoggusCattus_Scream.CameraShake_DoggusCattus_Scream_C" :Scale _scale :PlaySpace "CameraLocal")
    (if _distanceff
      (Game|Feedback|ClientPlayForceFeedback _output "/Figment_S10/Athena/Prototype/Blueprints/Cattus/ForceFeedback/FF_DoggusCattus_MissleBarrage.FF_DoggusCattus_MissleBarrage" "None" false false false))))

(event Custom|IMPACTIMPACT
  ;[1 1 1] IMPACT IMPACT local player shake and feedback
  (if (not (Networking|IsDedicatedServer))
    (bind (_output _scale _distanceff) (CallFunction|CameraShakeScalefromDistance (Variables|Default|GetImpactImpactCameraShakeDistance)))
    (Game|Feedback|ClientStartCameraShake :self _output :Shake "/Figment_S10/Athena/Prototype/Blueprints/Cattus/CameraShake/CameraShake_DoggusCattus_ImpactImpact.CameraShake_DoggusCattus_ImpactImpact_C" :Scale _scale :PlaySpace "CameraLocal")
    (Game|Feedback|ClientStartCameraShake :self _output :Shake "/Figment_S10/Athena/Prototype/Blueprints/Cattus/CameraShake/CameraShake_DoggusCattus_ImpactImpactEnd.CameraShake_DoggusCattus_ImpactImpactEnd_C" :Scale _scale :PlaySpace "CameraLocal")
    (if _distanceff
      (Game|Feedback|ClientPlayForceFeedback _output "/Figment_S10/Athena/Prototype/Blueprints/Cattus/ForceFeedback/FF_DoggusCattus_IMPACTIMPACT.FF_DoggusCattus_IMPACTIMPACT" "None" false false false))))

(event Custom|TakeOff
  ;[1 1 1] Take off local player shake and feedback
  (if (not (Networking|IsDedicatedServer))
    (bind (_output _scale _distanceff) (CallFunction|CameraShakeScalefromDistance (Variables|Default|GetTakeOffCameraShakeDistance)))
    (Game|Feedback|ClientStartCameraShake :self _output :Shake "/Figment_S10/Athena/Prototype/Blueprints/Cattus/CameraShake/CameraShake_DoggusCattus_FlightRumble.CameraShake_DoggusCattus_FlightRumble_C" :Scale _scale :PlaySpace "CameraLocal")
    (if _distanceff
      (Game|Feedback|ClientPlayForceFeedback _output "/Figment_S10/Athena/Prototype/Blueprints/Cattus/ForceFeedback/FF_DoggusCattus_Flight.FF_DoggusCattus_Flight" "None" false false false))))

(event Custom|FlightEnd
  ;[1 1 1] Flight ending local player shake and feedback
  (if (not (Networking|IsDedicatedServer))
    (bind (_output _scale _distanceff) (CallFunction|CameraShakeScalefromDistance (Variables|Default|GetFlightEndCameraShakeDistance)))
    (Game|Feedback|ClientStartCameraShake :self _output :Shake "/Figment_S10/Athena/Prototype/Blueprints/Cattus/CameraShake/CameraShake_DoggusCattus_FlightEnd.CameraShake_DoggusCattus_FlightEnd_C" :Scale _scale :PlaySpace "CameraLocal")
    (if _distanceff
      (Game|Feedback|ClientPlayForceFeedback _output "/Figment_S10/Athena/Prototype/Blueprints/Cattus/ForceFeedback/FF_DoggusCattus_Flight.FF_DoggusCattus_Flight" "None" false false false))))

(event Custom|Destruction
  ;[1 1 1] Destruction local player shake and feedback
  (if (not (Networking|IsDedicatedServer))
    (bind (_output _scale _distanceff) (CallFunction|CameraShakeScalefromDistance (Variables|Default|GetDestructionCameraShakeDistance)))
    (Game|Feedback|ClientStartCameraShake :self _output :Shake "/Figment_S10/Athena/Prototype/Blueprints/Cattus/CameraShake/CameraShake_DoggusCattus_Destruction.CameraShake_DoggusCattus_Destruction_C" :Scale _scale :PlaySpace "CameraLocal")
    (if _distanceff
      (Game|Feedback|ClientPlayForceFeedback _output "/Figment_S10/Athena/Prototype/Blueprints/Cattus/ForceFeedback/FF_DoggusCattus_Destruction.FF_DoggusCattus_Destruction" "None" false false false))))

(event Custom|PullSwordOut
  ;[1 1 1] Pull sword out local player shake and feedback
  (if (not (Networking|IsDedicatedServer))
    (bind (_output _scale _distanceff) (CallFunction|CameraShakeScalefromDistance (Variables|Default|GetPullSwordOutCameraShakeDistance)))
    (Game|Feedback|ClientStartCameraShake :self _output :Shake "/Figment_S10/Athena/Prototype/Blueprints/Cattus/CameraShake/CameraShake_DoggusCattus_PullSwordOut.CameraShake_DoggusCattus_PullSwordOut_C" :Scale _scale :PlaySpace "CameraLocal")
    (if _distanceff
      (Game|Feedback|ClientPlayForceFeedback _output "/Figment_S10/Athena/Prototype/Blueprints/Cattus/ForceFeedback/FF_DoggusCattus_PullSwordOut.FF_DoggusCattus_PullSwordOut" "None" false false false))))

(event Custom|SwordFullyExtended
  ;[1 1 1] Sword fully extends local player shake and feedback
  (if (not (Networking|IsDedicatedServer))
    (bind (_output _scale _distanceff) (CallFunction|CameraShakeScalefromDistance (Variables|Default|GetPullSwordOutCameraShakeDistance)))
    (Game|Feedback|ClientStartCameraShake :self _output :Shake "/Figment_S10/Athena/Prototype/Blueprints/Cattus/CameraShake/CameraShake_DoggusCattus_PullSwordOut.CameraShake_DoggusCattus_PullSwordOut_C" :Scale _scale :PlaySpace "CameraLocal")
    (if _distanceff
      (Game|Feedback|ClientPlayForceFeedback _output "/Figment_S10/Athena/Prototype/Blueprints/Cattus/ForceFeedback/FF_DoggusCattus_SwordExtend.FF_DoggusCattus_SwordExtend" "None" false false false))))

(event Custom|BreakParts
  ;[1 1 1] Break parts local player shake and feedback
  (if (not (Networking|IsDedicatedServer))
    (bind (_output _scale _distanceff) (CallFunction|CameraShakeScalefromDistance (Variables|Default|GetBreakPartsCameraShakeDistance)))
    (Game|Feedback|ClientStartCameraShake :self _output :Shake "/Figment_S10/Athena/Prototype/Blueprints/Cattus/CameraShake/CameraShake_DoggusCattus_BreakParts.CameraShake_DoggusCattus_BreakParts_C" :Scale _scale :PlaySpace "CameraLocal")
    (if _distanceff
      (Game|Feedback|ClientPlayForceFeedback _output "/Figment_S10/Athena/Prototype/Blueprints/Cattus/ForceFeedback/FF_DoggusCattus_BreakParts.FF_DoggusCattus_BreakParts" "None" false false false))))

(event Custom|PowerSurge
  ;[1 1 1] PowerSurge local player shake and feedback
  (if (not (Networking|IsDedicatedServer))
    (bind (_output _scale _distanceff) (CallFunction|CameraShakeScalefromDistance (Variables|Default|GetPowerSurgeCameraShakeDistance)))
    (Game|Feedback|ClientStartCameraShake :self _output :Shake "/Figment_S10/Athena/Prototype/Blueprints/Cattus/CameraShake/CameraShake_DoggusCattus_PowerSurgeStart.CameraShake_DoggusCattus_PowerSurgeStart_C" :Scale _scale :PlaySpace "CameraLocal")
    (if _distanceff
      (Game|Feedback|ClientPlayForceFeedback _output "/Figment_S10/Athena/Prototype/Blueprints/Cattus/ForceFeedback/FF_DoggusCattus_PowerSurge.FF_DoggusCattus_PowerSurge" "None" false false false)
      (else
        _))
    (Utilities|FlowControl|Delay 0.2)
    (Game|Feedback|ClientStartCameraShake :self _output :Shake "/Figment_S10/Athena/Prototype/Blueprints/Cattus/CameraShake/CameraShake_DoggusCattus_PowerSurgeContinuous.CameraShake_DoggusCattus_PowerSurgeContinuous_C" :Scale _scale :PlaySpace "CameraLocal")))

(event Custom|ElevatorRising
  ;[1 1 1] Elevator Rising local player camera shake and feedback
  (if (not (Networking|IsDedicatedServer))
    (bind (_output _scale _distanceff) (CallFunction|CameraShakeScalefromDistance (Variables|Default|GetElevatorRisingCameraShakeDistance)))
    (Game|Feedback|ClientStartCameraShake :self _output :Shake "/Figment_S10/Athena/Prototype/Blueprints/Cattus/CameraShake/CameraShake_DoggusCattus_ElevatorRumble.CameraShake_DoggusCattus_ElevatorRumble_C" :Scale _scale :PlaySpace "CameraLocal")
    (if _distanceff
      (Game|Feedback|ClientPlayForceFeedback _output "/Figment_S10/Athena/Prototype/Blueprints/Cattus/ForceFeedback/FF_DoggusCattus_ElevatorRumble.FF_DoggusCattus_ElevatorRumble" "None" false false false))))

(event Custom|First%20Laser
  ;[1 1 1] First Laser local player shake and feedback
  (if (not (Networking|IsDedicatedServer))
    (bind (_output _scale _distanceff) (CallFunction|CameraShakeScalefromDistance (Variables|Default|GetLaserCameraShakeDistance)))
    (Game|Feedback|ClientStartCameraShake :self _output :Shake "/Figment_S10/Athena/Prototype/Blueprints/Cattus/CameraShake/CameraShake_DoggusCattus_FirstLaserStart.CameraShake_DoggusCattus_FirstLaserStart_C" :Scale _scale :PlaySpace "CameraLocal")
    (if _distanceff
      (Game|Feedback|ClientPlayForceFeedback _output "/Figment_S10/Athena/Prototype/Blueprints/Cattus/ForceFeedback/FF_DoggusCattus_FirstLaser.FF_DoggusCattus_FirstLaser" "None" false false false))
    (Utilities|FlowControl|Delay 2.2)
    (Game|Feedback|ClientStartCameraShake :self _output :Shake "/Figment_S10/Athena/Prototype/Blueprints/Cattus/CameraShake/CameraShake_DoggusCattus_FirstLaser.CameraShake_DoggusCattus_FirstLaser_C" :Scale _scale :PlaySpace "CameraLocal")))

(fn CameraShakeScaleFromDistance
  (params (InRangeB :type Float%20%28double-precision%29))
  (LocalVariables|SetLocalPlayerController (Utilities|Array|Get(acopy) (Fortnite|GetLocalFortPlayerControllers true true)))
  (return (LocalVariables|GetLocalPlayerController) (Math|Float|Power (Math|Float|MapRangeClamped (Math|Vector|Distance(Vector) (Transformation|GetActorLocation) (Transformation|GetActorLocation (Pawn|GetViewTarget (LocalVariables|GetLocalPlayerController)))) 0.0 InRangeB 1.0) (Variables|Default|GetCameraShakePower)) (<= (Math|Vector|Distance(Vector) (Transformation|GetActorLocation) (Transformation|GetActorLocation (Pawn|GetViewTarget (LocalVariables|GetLocalPlayerController)))) InRangeB)))

(fn LaunchPawnFromLocation
  (params (TargetPawn :type Actor%20Object%20Reference)))

(fn ConstructionScript)
